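# Control Rig for the PaperBlaze male medium body, applying corrective offsets to its skirt, loincloth, shoulder pad and accessory bones.
# A shared DeformRig function runs first, then sequenced spherical pose readers on thighs, clavicle, upper arm and head weight additive transform tweaks.
# Movement-state variables scale these weights, with AlphaInterp smoothing some toggles.
# Left shoulder pad correctives branch by front end, skydiving, crouching and weapon type, with separate tuning per case.
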ControlRigBlueprint /BRCosmetics/Characters/Player/Male/Medium/Bodies/M_MED_PaperBlaze/Meshes/M_MED_PaperBlaze_CtrlRig
graphs [
 Multiply_1_1_1_ContainedGraph {nodes [
   n1 = RigVMFunctionEntryNode()
   RigVMFunctionReturnNode(Result=&RigVMFunction_MathDoubleMul(B=@n1.C, A=&RigVMFunction_MathDoubleMul(A=@n1.A, B=@n1.B).Result).Result)
  ]}
 Multiply_5_1_1_ContainedGraph {nodes [
   n1 = RigVMFunctionEntryNode()
   RigVMFunctionReturnNode(Result=&RigVMFunction_MathDoubleMul(B=@n1.C, A=&RigVMFunction_MathDoubleMul(A=@n1.A, B=@n1.B).Result).Result)
  ]}
 Multiply_5_1_1_ContainedGraph {nodes [
   n1 = RigVMFunctionEntryNode()
   RigVMFunctionReturnNode(Result=&RigVMFunction_MathDoubleMul(B=@n1.C, A=&RigVMFunction_MathDoubleMul(A=@n1.A, B=@n1.B).Result).Result)
  ]}
 RigVMFunctionLibrary {links [], nodes []}
 RigVMFunction_Sequence_1_1_ContainedGraph {nodes [
   n1 = RigVMFunctionEntryNode()
   n2 = &RigVMFunction_Sequence(ExecuteContext=@n1.ExecuteContext)
   n3 = &RigVMFunction_Sequence(ExecuteContext=@n2.B)
   RigVMFunctionReturnNode(A=@n2.A, B=@n3.A, C=@n3.B)
  ]}
 RigVMFunction_Sequence_1_ContainedGraph {nodes [
   n1 = RigVMFunctionEntryNode()
   RigVMFunction_Sequence = &RigVMFunction_Sequence(ExecuteContext=@n1.ExecuteContext)
   n2 = &RigVMFunction_Sequence(ExecuteContext=@RigVMFunction_Sequence.B)
   n3 = &RigVMFunction_Sequence(ExecuteContext=@n2.B)
   n4 = &RigVMFunction_Sequence(ExecuteContext=@n3.B)
   n5 = &RigVMFunction_Sequence(ExecuteContext=@n4.B)
   RigVMFunctionReturnNode(A=@RigVMFunction_Sequence.A, B=@n2.A, C=@n3.A, D=@n4.A, E=@n5.A, F=@n5.B)
  ]}
 RigVMModel {nodes [
   RigUnit_BeginExecution = &RigUnit_BeginExecution()
   FN_DeformRig_FNC = RigVMFunctionReferenceNode(
    DeformRig_ Curve="(deformRig_l_upperarm_up_alpha_72_#1=False,deformRig_r_upperarm_up_alpha_73_#2=False,deformRig_l_upperarm_fr_alpha_74_#3=False,deformRig_r_upperarm_fr_alpha_75_CB51B8 ...(truncated)"
    DeformRig_Bool="(upperarm_1_#4=True,elbow_3_#5=false,wrist_7_#6=false,groin_8_#7=true,glute_17_#8=true,knee_9_40C408324CA2 ...(truncated)"
    DeformRig_Debug=(upperarm_debug_18_#4=False,wrist_debug_20_#6=False,lat_debug_26_#9=False,groinGlute_debug_27_#7=False)
    DeformRig_Value="(upperarm_up_18_#10=(Rotation=(X=0,Y=-0.4226,Z=0,W=0.9063),Translation=(X=6,Y=5,Z=10),Scale3D=(X=0.7,Y=1,Z=1)),upperarm_fwd_fr_19_#11=(Rotation=(X=0,Y=-0.4226,Z=0,W=0.9063),Translation=(X=-10,Y=10,Z ...(truncated)"
    ExecuteContext=@RigUnit_BeginExecution.ExecutePin
   )
   > RigVMFunction_Sequence = RigVMAggregateNode()
   n1 = &RigUnit_SphericalPoseReader(
    ActiveRegionScaleFactors=(PositiveWidth=1,NegativeWidth=1,PositiveHeight=2,NegativeHeight=2)
    ActiveRegionSize=0.1
    Debug=(bDrawDebug=true,bDraw2D=false,DebugScale=25,DebugSegments=20,DebugThickness=0.25)
    DriverAxis=(X=-1,Y=0,Z=0)
    DriverItem=(Type=Bone,Name="thigh_l")
    FalloffRegionScaleFactors=(PositiveWidth=1,NegativeWidth=1,PositiveHeight=2,NegativeHeight=2)
    FalloffSize=0.3
    OptionalParentItem=(Type=Bone,Name="pelvis")
    RotationOffset=(X=118.397,Y=-4.8963,Z=-8.7838)
    ExecutePin=@RigVMFunction_Sequence.A
   )
   n2 = RigVMVariableNode() {variable=IsCrouching}
   Remap = &RigVMFunction_MathDoubleRemap(SourceMaximum=1, SourceMinimum=0, TargetMaximum=0.8, TargetMinimum=0, bClamp=False, Value=@n2.Value)
   > &RigUnit_ParentConstraint(
    AdvancedSettings=(InterpolationType=Average,RotationOrderForFilter=XZY)
    Child=(Type=Bone,Name="dyn_left_loin_3")
    Parents=((Item=(Type=Bone,Name="pelvis"),Weight=1))
    Weight=@Remap.Result
   )
   > &RigUnit_ModifyTransforms(
    ItemToModify="((Item=(Type=Bone,Name=\"dyn_left_loin_3\"),Transform=(Rotation=(X=0,Y=-0.0493,Z=0,W=0.9988),Translation=(X=-1.8853,Y=0.0214,Z=0.2747),Scale3D=(X=1,Y=1,Z=1))),(Item=(Type=Bone,Name=\"dyn_left_loin_1\"),Transform=(Rotation=(X=0.0216,Y=0.0007,Z=0,W=0.9998),Trans ...(truncated)"
    WeightMinimum=0.5
    Weight=&RigVMFunction_MathDoubleMul(A=@n1.OutputParam, B=@n2.Value).Result
   )
   n3 = RigVMVariableNode() {variable=SkyDiving}
   Multiply_5_1_1 = RigVMAggregateNode(
    C=&RigVMFunction_MathDoubleSub(A=1, B=&RigVMFunction_MathBoolToFloat(Value=@n3.Value).Result).Result
    A=RigVMVariableNode() {variable=IsCrouching}.Value
    B=RigVMVariableNode() {variable=IsSliding}.Value
   )
   n4 = &RigVMFunction_MathDoubleSub(A=1, B=@Multiply_5_1_1.Result)
   n5 = RigVMVariableNode() {variable=IsJumping}
   n6 = RigVMVariableNode() {variable=IsFalling}
   n7 = &RigVMFunction_MathDoubleSub(
    A=1
    B=&RigVMFunction_MathDoubleMul(A=&RigVMFunction_MathBoolToFloat(Value=@n5.Value).Result, B=&RigVMFunction_MathBoolToFloat(Value=@n6.Value).Result).Result
   )
   n8 = &RigVMFunction_MathDoubleSub(A=1, B=&RigVMFunction_MathBoolToFloat(Value=RigVMVariableNode() {variable=IsOnGround}.Value).Result)
   > &RigUnit_ModifyTransforms(
    ItemToModify="((Item=(Type=Bone,Name=\"dyn_skirt_l\"),Transform=(Rotation=(X=0.3873,Y=-0.1833,Z=-0.0789,W=0.9001),Translation=(X=0,Y=0,Z=0),Scale3D=(X=1,Y=1,Z=1))),(Item=(Type=Bone,Name=\"dyn_skirt_main_l\"),Transform=(Rotation=(X=0.2754,Y=0.2151,Z=-0.0391,W=0.9361),Transla ...(truncated)"
    WeightMinimum=0
    Weight=&RigVMFunction_MathDoubleMul(
     A=&RigVMFunction_MathDoubleMul(A=@n1.OutputParam, B=@n4.Result).Result
     B=&RigVMFunction_MathDoubleMul(A=@n7.Result, B=@n8.Result).Result
    ).Result
   )
   n9 = &RigUnit_SphericalPoseReader(
    ActiveRegionScaleFactors=(PositiveWidth=1.1,NegativeWidth=1,PositiveHeight=1,NegativeHeight=1)
    ActiveRegionSize=0.3
    Debug=(bDrawDebug=false,bDraw2D=false,DebugScale=25,DebugSegments=20,DebugThickness=0.25)
    DriverAxis=(X=1,Y=0,Z=0)
    DriverItem=(Type=Bone,Name="thigh_r")
    FalloffRegionScaleFactors=(PositiveWidth=3,NegativeWidth=1,PositiveHeight=1,NegativeHeight=1)
    FalloffSize=0.5
    OptionalParentItem=(Type=Bone,Name="pelvis")
    RotationOffset=(X=-80,Y=0,Z=-20)
    ExecutePin=@RigVMFunction_Sequence.B
   )
   n10 = RigVMVariableNode() {variable=IsBaseballSliding}
   n11 = &RigVMFunction_MathDoubleMul(
    B=&RigVMFunction_MathDoubleMul(B=&RigVMFunction_MathDoubleSub(A=1, B=@n10.Value).Result, A=@n2.Value).Result
    A=@n9.OutputParam
   )
   > &RigUnit_ModifyTransforms(
    ItemToModify=((Item=(Type=Bone,Name="dyn_right_loin_correct"),Transform=(Rotation=(X=0.0615,Y=-0.1494,Z=0.2062,W=0.9651),Translation=(X=-4.6541,Y=-1.0962,Z=5.0171),Scale3D=(X=1,Y=1,Z=1))))
    Weight=@n11.Result
   )
   n12 = &RigVMFunction_MathDoubleMul(
    B=&RigVMFunction_MathDoubleMul(A=&RigVMFunction_MathDoubleSub(A=1, B=@n2.Value).Result, B=&RigVMFunction_MathDoubleSub(A=1, B=@n10.Value).Result).Result
    A=@n9.OutputParam
   )
   > &RigUnit_ModifyTransforms(
    ItemToModify=((Item=(Type=Bone,Name="dyn_right_loin_correct"),Transform=(Rotation=(X=0.0799,Y=-0.044,Z=0.0465,W=0.9947),Translation=(X=-6.0021,Y=-0.3035,Z=3.4518),Scale3D=(X=1,Y=1,Z=1))))
    Weight=@n12.Result
   )
   > &RigUnit_ModifyTransforms(
    ItemToModify=((Item=(Type=Bone,Name="dyn_skirt_main_r"),Transform=(Rotation=(X=-0.0005,Y=0.0242,Z=0.0235,W=0.9994),Translation=(X=2.8418,Y=1.0643,Z=-0.1129),Scale3D=(X=1,Y=1,Z=1))))
    Weight=@n12.Result
   )
   > &RigUnit_ModifyTransforms(
    ItemToModify=((Item=(Type=Bone,Name="dyn_skirt_main_r"),Transform=(Rotation=(X=0.0909,Y=0.1246,Z=-0.0236,W=0.9878),Translation=(X=0.2524,Y=0,Z=0.6819),Scale3D=(X=1,Y=1,Z=1))))
    Weight=@n11.Result
   )
   > &RigUnit_ModifyTransforms(
    ItemToModify=((Item=(Type=Bone,Name="dyn_skirt_main_bk_r"),Transform=(Rotation=(X=0,Y=0.0436,Z=0,W=0.999),Translation=(X=0,Y=0,Z=0),Scale3D=(X=1,Y=1,Z=1))))
    Weight=@n11.Result
   )
   n13 = RigVMVariableNode() {variable=IsSliding}
   &RigUnit_ModifyTransforms(
    ItemToModify="((Item=(Type=Bone,Name=\"dyn_skilt_bk_l\"),Transform=(Rotation=(X=0,Y=0,Z=0,W=1),Translation=(X=0,Y=0,Z=0),Scale3D=(X=0.7,Y=1,Z=1))),(Item=(Type=Bone,Name=\"dyn_skirt_bk_r\"),Transform=(Rotation=(X=0,Y=0,Z=0,W=1),Translation=(X=0,Y=0,Z=0),Scale3D=(X=0.7,Y=1,Z= ...(truncated)"
    ExecutePin=@RigVMFunction_Sequence.C
    Weight=@n13.Value
   )
   > &RigUnit_ModifyTransforms(
    ItemToModify="((Item=(Type=Bone,Name=\"dyn_skilt_bk_l\"),Transform=(Rotation=(X=0.0333,Y=-0.0064,Z=-0.0307,W=0.999),Translation=(X=0,Y=0,Z=0),Scale3D=(X=0.7,Y=1,Z=1))),(Item=(Type=Bone,Name=\"dyn_skirt_bk_r\"),Transform=(Rotation=(X=-0.069,Y=0,Z=0,W=0.9976),Translation=(X=0 ...(truncated)"
    Weight=@n10.Value
   )
   > n14 = &RigUnit_SphericalPoseReader(
    ActiveRegionScaleFactors=(PositiveWidth=2,NegativeWidth=2,PositiveHeight=1,NegativeHeight=1)
    ActiveRegionSize=0.1
    Debug=(bDrawDebug=false,bDraw2D=false,DebugScale=25,DebugSegments=20,DebugThickness=0.25)
    DriverAxis=(X=1,Y=0,Z=0)
    DriverItem=(Type=Bone,Name="thigh_r")
    FalloffRegionScaleFactors=(PositiveWidth=1,NegativeWidth=1,PositiveHeight=1,NegativeHeight=1)
    FalloffSize=0.3
    OptionalParentItem=(Type=Bone,Name="pelvis")
    RotationOffset=(X=0,Y=-60,Z=40)
   )
   > &RigUnit_ModifyTransforms(
    ItemToModify=((Item=(Type=Bone,Name="dyn_skirt_main_r"),Transform=(Rotation=(X=0.0443,Y=0.0011,Z=-0.0249,W=0.9987),Translation=(X=4.4059,Y=0.2633,Z=-5.4407),Scale3D=(X=1,Y=1,Z=1))))
    Weight=&RigVMFunction_MathDoubleMul(A=@n14.OutputParam, B=@n13.Value).Result
   )
   > n15 = &RigUnit_SphericalPoseReader(
    ActiveRegionScaleFactors=(PositiveWidth=2,NegativeWidth=2,PositiveHeight=1,NegativeHeight=1)
    ActiveRegionSize=0.1
    Debug=(bDrawDebug=false,bDraw2D=false,DebugScale=25,DebugSegments=20,DebugThickness=0.25)
    DriverAxis=(X=1,Y=0,Z=0)
    DriverItem=(Type=Bone,Name="thigh_r")
    FalloffRegionScaleFactors=(PositiveWidth=1,NegativeWidth=1,PositiveHeight=1,NegativeHeight=1)
    FalloffSize=0.3
    OptionalParentItem=(Type=Bone,Name="pelvis")
    RotationOffset=(X=0,Y=-60,Z=30)
   )
   Multiply_1_1 = RigVMAggregateNode(
    A=@n15.OutputParam
    B=&RigVMFunction_MathDoubleSub(A=1, B=@n13.Value).Result
    C=&RigVMFunction_MathDoubleSub(A=1, B=RigVMVariableNode() {variable=IsBaseballSliding}.Value).Result
   )
   > &RigUnit_ModifyTransforms(
    ItemToModify=((Item=(Type=Bone,Name="dyn_skirt_main_bk_r"),Transform=(Rotation=(X=0,Y=0,Z=0,W=1),Translation=(X=4.7245,Y=0,Z=0),Scale3D=(X=1,Y=1,Z=1))))
    Weight=@Multiply_1_1.Result
   )
   &RigUnit_SphericalPoseReader(
    ActiveRegionScaleFactors=(PositiveWidth=1,NegativeWidth=1,PositiveHeight=1,NegativeHeight=1)
    ActiveRegionSize=0.05
    Debug=(bDrawDebug=false,bDraw2D=false,DebugScale=25,DebugSegments=20,DebugThickness=0.25)
    DriverAxis=(X=0,Y=0,Z=1)
    DriverItem=(Type=Bone,Name="hand_r")
    FalloffRegionScaleFactors=(PositiveWidth=1,NegativeWidth=1,PositiveHeight=1,NegativeHeight=1)
    FalloffSize=0.6
    OptionalParentItem=(Type=Bone,Name="pelvis")
    RotationOffset=(X=0,Y=-40,Z=-25)
    ExecutePin=@RigVMFunction_Sequence.D
   )
   > &RigUnit_ModifyTransforms(
    ItemToModify="((Item=(Type=Bone,Name=\"dyn_bracelet_r_c\"),Transform=(Rotation=(X=-0.0021,Y=0.1681,Z=-0.0966,W=0.981),Translation=(X=0.5914,Y=-0.1041,Z=-0.0948),Scale3D=(X=1,Y=1,Z=1))),(Item=(Type=Bone,Name=\"None\"),Transform=(Rotation=(X=-0.2181,Y=0.2583,Z=-0.106,W=0.9351 ...(truncated)"
    Weight=0
   )
   RigVMFunction_Sequence_1 = RigVMAggregateNode(ExecuteContext=@RigVMFunction_Sequence.E)
   &RigUnit_SphericalPoseReader(
    ActiveRegionScaleFactors=(PositiveWidth=1,NegativeWidth=1,PositiveHeight=1,NegativeHeight=1)
    ActiveRegionSize=0.1
    Debug=(bDrawDebug=false,bDraw2D=false,DebugScale=25,DebugSegments=20,DebugThickness=0.25)
    DriverAxis=(X=0,Y=1,Z=0)
    DriverItem=(Type=Bone,Name="thigh_l")
    FalloffRegionScaleFactors=(PositiveWidth=1,NegativeWidth=1,PositiveHeight=1,NegativeHeight=1)
    FalloffSize=0.3
    OptionalParentItem=(Type=Bone,Name="pelvis")
    RotationOffset=(X=90,Y=0,Z=-60)
    ExecutePin=@RigVMFunction_Sequence_1.A
   )
   > n16 = &RigUnit_SphericalPoseReader(
    ActiveRegionScaleFactors=(PositiveWidth=1,NegativeWidth=1,PositiveHeight=1,NegativeHeight=1)
    ActiveRegionSize=0.2
    Debug=(bDrawDebug=true,bDraw2D=false,DebugScale=25,DebugSegments=20,DebugThickness=0.25)
    DriverAxis=(X=1,Y=0,Z=0)
    DriverItem=(Type=Bone,Name="thigh_r")
    FalloffRegionScaleFactors=(PositiveWidth=1,NegativeWidth=1,PositiveHeight=1,NegativeHeight=1)
    FalloffSize=0.35
    OptionalParentItem=(Type=Bone,Name="pelvis")
    RotationOffset=(X=0,Y=0,Z=0)
   )
   Subtract = &RigVMFunction_MathDoubleSub(A=1, B=&RigVMFunction_MathBoolToFloat(Value=RigVMVariableNode() {variable=SkyDiving}.Value).Result)
   Multiply_5_1 = RigVMAggregateNode(C=@Subtract.Result, A=RigVMVariableNode() {variable=IsCrouching}.Value, B=RigVMVariableNode() {variable=IsSliding}.Value)
   > &RigUnit_ModifyTransforms(
    ItemToModify="((Item=(Type=Bone,Name=\"dyn_skirt_main_bk_r\"),Transform=(Rotation=(X=0,Y=-0.1408,Z=0,W=0.99),Translation=(X=0,Y=0,Z=0),Scale3D=(X=1,Y=1,Z=1))),(Item=(Type=Bone,Name=\"dyn_skirt_main_r\"),Transform=(Rotation=(X=-0.0665,Y=-0.3568,Z=-0.1948,W=0.9112),Translatio ...(truncated)"
    WeightMinimum=0
    Weight=&RigVMFunction_MathDoubleMul(A=@n16.OutputParam, B=&RigVMFunction_MathDoubleSub(A=1, B=@Multiply_5_1.Result).Result).Result
   )
   &RigUnit_ModifyTransforms(
    ItemToModify=((Item=(Type=Bone,Name="deform_bracelet_l"),Transform=(Rotation=(X=0.0872,Y=0,Z=0,W=0.9962),Translation=(X=-5,Y=0,Z=0),Scale3D=(X=1,Y=1,Z=1))))
    Weight=1
    ExecutePin=@RigVMFunction_Sequence_1.C
   )
   n17 = &RigVMFunction_AlphaInterp(
    InterpSpeedDecreasing=5
    InterpSpeedIncreasing=5
    Scale=1
    bClampResult=False
    Value=&RigVMFunction_MathBoolToFloat(Value=RigVMVariableNode() {variable=TacticalMovement}.Value).Result
   )
   > &RigUnit_ModifyTransforms(
    ItemToModify="((Item=(Type=Bone,Name=\"dyn_knife_chain_b\"),Transform=(Rotation=(X=0,Y=0,Z=0,W=1),Translation=(X=0,Y=3,Z=-1),Scale3D=(X=1,Y=1,Z=1))),(Item=(Type=Bone,Name=\"dyn_knife_chain_f\"),Transform=(Rotation=(X=0,Y=0,Z=0,W=1),Translation=(X=-1,Y=1,Z=0),Scale3D=(X=1,Y= ...(truncated)"
    Weight=@n17.Result
   )
   n18 = &RigVMFunction_ControlFlowBranch(Condition=RigVMVariableNode() {variable=IsInFrontEnd}.Value, ExecuteContext=@RigVMFunction_Sequence_1.B)
   n19 = &RigVMFunction_ControlFlowBranch(ExecuteContext=@n18.False, Condition=RigVMVariableNode() {variable=SkyDiving}.Value)
   n20 = &RigVMFunction_AlphaInterp(
    InterpSpeedDecreasing=10
    InterpSpeedIncreasing=5
    Scale=1
    bClampResult=False
    Value=&RigVMFunction_MathBoolToFloat(Value=RigVMVariableNode() {variable=TacticalMovement}.Value).Result
   )
   n21 = &RigUnit_ModifyTransforms(
    ItemToModify=((Item=(Type=Bone,Name="dyn_shoulderpad_l"),Transform=(Rotation=(X=-0.1736,Y=0,Z=0,W=0.9848),Translation=(X=0,Y=0,Z=0),Scale3D=(X=1,Y=1,Z=1))))
    Weight=@n20.Result
    ExecutePin=@n19.False
   )
   n22 = &RigVMFunction_ControlFlowBranch(ExecuteContext=@n21.ExecutePin, Condition=RigVMVariableNode() {variable=IsCrouchingBool}.Value)
   n23 = &RigVMFunction_ControlFlowBranch(
    Condition=RigVMDispatchNode(B=2, A=RigVMVariableNode() {variable=MelleePistolOtherAssaultRifle}.Value) {template="DISPATCH_RigVMDispatch_CoreEquals(in A,in B,out Result)"}.Result
    ExecuteContext=@n22.False
   )
   n24 = &RigUnit_SphericalPoseReader(
    ActiveRegionScaleFactors=(PositiveWidth=1,NegativeWidth=1,PositiveHeight=1,NegativeHeight=1)
    ActiveRegionSize=0.05
    Debug=(bDrawDebug=false,bDraw2D=false,DebugScale=25,DebugSegments=20,DebugThickness=0.25)
    DriverAxis=(X=0,Y=0,Z=1)
    DriverItem=(Type=Bone,Name="upperarm_twist_01_l")
    FalloffRegionScaleFactors=(PositiveWidth=1,NegativeWidth=1,PositiveHeight=1,NegativeHeight=1)
    FalloffSize=0.2
    OptionalParentItem=(Type=Bone,Name="dyn_sleeve_l_main")
    RotationOffset=(X=0,Y=0,Z=0)
    ExecutePin=@n23.False
   )
   > n25 = &RigUnit_SphericalPoseReader(
    ActiveRegionScaleFactors=(PositiveWidth=0.2,NegativeWidth=0.2,PositiveHeight=0.2,NegativeHeight=1)
    ActiveRegionSize=0.2
    Debug=(bDrawDebug=false,bDraw2D=false,DebugScale=25,DebugSegments=20,DebugThickness=0.25)
    DriverAxis=(X=1,Y=0,Z=0)
    DriverItem=(Type=Bone,Name="clavicle_l")
    FalloffRegionScaleFactors=(PositiveWidth=0.5,NegativeWidth=0.7,PositiveHeight=1,NegativeHeight=1)
    FalloffSize=0.1
    OptionalParentItem=(Type=Bone,Name="pelvis")
    RotationOffset=(X=10,Y=-60,Z=0)
   )
   > n26 = &RigUnit_SphericalPoseReader(
    ActiveRegionScaleFactors=(PositiveWidth=3,NegativeWidth=3,PositiveHeight=1,NegativeHeight=3)
    ActiveRegionSize=0.1
    Debug=(bDrawDebug=false,bDraw2D=false,DebugScale=25,DebugSegments=20,DebugThickness=0.25)
    DriverAxis=(X=0,Y=0,Z=1)
    DriverItem=(Type=Bone,Name="upperarm_twist_01_l")
    FalloffRegionScaleFactors=(PositiveWidth=1,NegativeWidth=1,PositiveHeight=1.4,NegativeHeight=1)
    FalloffSize=0.5
    OptionalParentItem=(Type=Bone,Name="upperarm_l")
    RotationOffset=(X=-40,Y=0,Z=0)
   )
   > n27 = &RigUnit_SphericalPoseReader(
    ActiveRegionScaleFactors=(PositiveWidth=1,NegativeWidth=1,PositiveHeight=1,NegativeHeight=1)
    ActiveRegionSize=0.1
    Debug=(bDrawDebug=false,bDraw2D=false,DebugScale=25,DebugSegments=20,DebugThickness=0.25)
    DriverAxis=(X=1,Y=0,Z=0)
    DriverItem=(Type=Bone,Name="head")
    FalloffRegionScaleFactors=(PositiveWidth=1,NegativeWidth=1,PositiveHeight=1,NegativeHeight=1)
    FalloffSize=0.2
    OptionalParentItem=(Type=Bone,Name="pelvis")
    RotationOffset=(X=-40,Y=0,Z=50)
   )
   > &RigUnit_ModifyTransforms(
    ItemToModify=((Item=(Type=Bone,Name="dyn_shoulderpad_l"),Transform=(Rotation=(X=0.1251,Y=0.0449,Z=-0.0256,W=0.9908),Translation=(X=0.0326,Y=0.5088,Z=0.1291),Scale3D=(X=1,Y=1,Z=1.1))))
    Weight=@n26.OutputParam
   )
   n28 = &RigVMFunction_MathDoubleSub(A=1, B=@n26.OutputParam)
   > &RigUnit_ModifyTransforms(
    ItemToModify=((Item=(Type=Bone,Name="dyn_shoulderpad_l"),Transform=(Rotation=(X=-0.0296,Y=0.0373,Z=0.0507,W=0.9976),Translation=(X=0.5541,Y=-1.5521,Z=-0.0257),Scale3D=(X=1,Y=1,Z=1.1228))))
    Weight=@n28.Result
   )
   > &RigUnit_ModifyTransforms(
    ItemToModify=((Item=(Type=Bone,Name="dyn_shoulderpad_l"),Transform=(Rotation=(X=-0.0855,Y=-0.0367,Z=-0.0403,W=0.9948),Translation=(X=-0.048,Y=-0.5504,Z=0),Scale3D=(X=1,Y=1,Z=1))))
    Weight=@n25.OutputParam
   )
   > &RigUnit_ModifyTransforms(
    ItemToModify=((Item=(Type=Bone,Name="dyn_shoulderpad_l"),Transform=(Rotation=(X=0,Y=0,Z=-0.0547,W=0.9985),Translation=(X=0,Y=0,Z=0),Scale3D=(X=1,Y=1,Z=1))))
    Weight=@n24.OutputParam
   )
   > n29 = &RigUnit_ModifyTransforms(
    ItemToModify=((Item=(Type=Bone,Name="dyn_shoulderpad_l"),Transform=(Rotation=(X=-0.0541,Y=0,Z=0,W=0.9985),Translation=(X=0,Y=-0.043,Z=-0.3676),Scale3D=(X=1,Y=1,Z=1))))
    Weight=@n27.OutputParam
   )
   n30 = &RigVMFunction_ControlFlowBranch(
    Condition=RigVMDispatchNode(B=3, A=RigVMVariableNode() {variable=MelleePistolOtherAssaultRifle}.Value) {template="DISPATCH_RigVMDispatch_CoreEquals(in A,in B,out Result)"}.Result
    ExecuteContext=@n29.ExecutePin
   )
   n31 = &RigVMFunction_ControlFlowBranch(
    Condition=RigVMDispatchNode(B=0, A=RigVMVariableNode() {variable=MelleePistolOtherAssaultRifle}.Value) {template="DISPATCH_RigVMDispatch_CoreEquals(in A,in B,out Result)"}.Result
    ExecuteContext=@n30.False
   )
   n32 = &RigVMFunction_ControlFlowBranch(
    Condition=RigVMDispatchNode(B=1, A=RigVMVariableNode() {variable=MelleePistolOtherAssaultRifle}.Value) {template="DISPATCH_RigVMDispatch_CoreEquals(in A,in B,out Result)"}.Result
    ExecuteContext=@n31.False
   )
   n33 = RigVMVariableNode() {variable=Targeting}
   n34 = &RigVMFunction_AlphaInterp(
    InterpSpeedDecreasing=5
    InterpSpeedIncreasing=5
    Scale=1
    bClampResult=False
    Value=&RigVMFunction_MathDoubleSub(A=1, B=&RigVMFunction_MathBoolToFloat(Value=@n33.Value).Result).Result
   )
   &RigUnit_ModifyTransforms(
    ItemToModify=((Item=(Type=Bone,Name="dyn_shoulderpad_l"),Transform=(Rotation=(X=0.1393,Y=0.0003,Z=0.0856,W=0.9865),Translation=(X=0,Y=-0.6759,Z=0),Scale3D=(X=1,Y=1,Z=1))))
    Weight=@n34.Result
    ExecutePin=@n32.True
   )
   &RigUnit_ModifyTransforms(
    ItemToModify=((Item=(Type=Bone,Name="dyn_shoulderpad_l"),Transform=(Rotation=(X=0.1698,Y=-0.0108,Z=0.0627,W=0.9834),Translation=(X=0.5292,Y=-0.135,Z=-0.0476),Scale3D=(X=1,Y=1,Z=1))))
    ExecutePin=@n31.True
    Weight=@n26.OutputParam
   )
   &RigUnit_ModifyTransforms(
    ItemToModify=((Item=(Type=Bone,Name="dyn_shoulderpad_l"),Transform=(Rotation=(X=0.2647,Y=0.0027,Z=0.055,W=0.9628),Translation=(X=1.1145,Y=-0.3555,Z=-0.1993),Scale3D=(X=1,Y=1,Z=1))))
    Weight=&RigVMFunction_MathDoubleRemap(SourceMaximum=0.65, SourceMinimum=0.47, TargetMaximum=1, TargetMinimum=0, Value=@n28.Result).Result
    ExecutePin=@n30.True
   )
   n35 = &RigUnit_SphericalPoseReader(
    ActiveRegionScaleFactors=(PositiveWidth=1,NegativeWidth=1,PositiveHeight=1,NegativeHeight=1)
    ActiveRegionSize=0.05
    Debug=(bDrawDebug=false,bDraw2D=false,DebugScale=25,DebugSegments=20,DebugThickness=0.25)
    DriverAxis=(X=0,Y=0,Z=1)
    DriverItem=(Type=Bone,Name="upperarm_twist_01_l")
    FalloffRegionScaleFactors=(PositiveWidth=1,NegativeWidth=1,PositiveHeight=1,NegativeHeight=1)
    FalloffSize=0.2
    OptionalParentItem=(Type=Bone,Name="dyn_sleeve_l_main")
    RotationOffset=(X=0,Y=0,Z=0)
    ExecutePin=@n23.True
   )
   > n36 = &RigUnit_SphericalPoseReader(
    ActiveRegionScaleFactors=(PositiveWidth=0.2,NegativeWidth=0.2,PositiveHeight=0.2,NegativeHeight=1)
    ActiveRegionSize=0.2
    Debug=(bDrawDebug=false,bDraw2D=false,DebugScale=25,DebugSegments=20,DebugThickness=0.25)
    DriverAxis=(X=1,Y=0,Z=0)
    DriverItem=(Type=Bone,Name="clavicle_l")
    FalloffRegionScaleFactors=(PositiveWidth=0.5,NegativeWidth=0.7,PositiveHeight=1,NegativeHeight=1)
    FalloffSize=0.1
    OptionalParentItem=(Type=Bone,Name="pelvis")
    RotationOffset=(X=40,Y=-60,Z=0)
   )
   > n37 = &RigUnit_SphericalPoseReader(
    ActiveRegionScaleFactors=(PositiveWidth=3,NegativeWidth=3,PositiveHeight=1,NegativeHeight=3)
    ActiveRegionSize=0.1
    Debug=(bDrawDebug=false,bDraw2D=false,DebugScale=25,DebugSegments=20,DebugThickness=0.25)
    DriverAxis=(X=0,Y=0,Z=1)
    DriverItem=(Type=Bone,Name="upperarm_twist_01_l")
    FalloffRegionScaleFactors=(PositiveWidth=1,NegativeWidth=1,PositiveHeight=1.4,NegativeHeight=1)
    FalloffSize=0.5
    OptionalParentItem=(Type=Bone,Name="upperarm_l")
    RotationOffset=(X=-40,Y=0,Z=0)
   )
   > n38 = &RigUnit_SphericalPoseReader(
    ActiveRegionScaleFactors=(PositiveWidth=1,NegativeWidth=1,PositiveHeight=1,NegativeHeight=1)
    ActiveRegionSize=0.1
    Debug=(bDrawDebug=true,bDraw2D=false,DebugScale=25,DebugSegments=20,DebugThickness=0.25)
    DriverAxis=(X=1,Y=0,Z=0)
    DriverItem=(Type=Bone,Name="head")
    FalloffRegionScaleFactors=(PositiveWidth=1,NegativeWidth=1,PositiveHeight=1,NegativeHeight=1)
    FalloffSize=0.2
    OptionalParentItem=(Type=Bone,Name="pelvis")
    RotationOffset=(X=-60,Y=0,Z=40)
   )
   > &RigUnit_ModifyTransforms(
    ItemToModify=((Item=(Type=Bone,Name="dyn_shoulderpad_l"),Transform=(Rotation=(X=0.1251,Y=0.0449,Z=-0.0256,W=0.9908),Translation=(X=0.0326,Y=0.5088,Z=0.1291),Scale3D=(X=1,Y=1,Z=1))))
    Weight=@n37.OutputParam
   )
   > &RigUnit_ModifyTransforms(
    ItemToModify=((Item=(Type=Bone,Name="dyn_shoulderpad_l"),Transform=(Rotation=(X=-0.2561,Y=0.0056,Z=0.054,W=0.9651),Translation=(X=-0.0986,Y=-1.2521,Z=-0.0972),Scale3D=(X=1,Y=1,Z=1.0905))))
    Weight=&RigVMFunction_MathDoubleSub(A=1, B=@n37.OutputParam).Result
   )
   > &RigUnit_ModifyTransforms(
    ItemToModify=((Item=(Type=Bone,Name="dyn_shoulderpad_l"),Transform=(Rotation=(X=-0.0855,Y=-0.0367,Z=-0.0403,W=0.9948),Translation=(X=-0.048,Y=-0.5504,Z=0),Scale3D=(X=1,Y=1,Z=1))))
    Weight=@n36.OutputParam
   )
   > &RigUnit_ModifyTransforms(
    ItemToModify=((Item=(Type=Bone,Name="dyn_shoulderpad_l"),Transform=(Rotation=(X=-0.0986,Y=0.0054,Z=-0.0544,W=0.9936),Translation=(X=0,Y=0,Z=0),Scale3D=(X=1,Y=1,Z=1))))
    Weight=@n35.OutputParam
   )
   > &RigUnit_ModifyTransforms(
    ItemToModify=((Item=(Type=Bone,Name="dyn_shoulderpad_l"),Transform=(Rotation=(X=0.0344,Y=0.0214,Z=-0.001,W=0.9992),Translation=(X=0.8335,Y=-0.0581,Z=-0.203),Scale3D=(X=1,Y=1,Z=1))))
    Weight=@n38.OutputParam
   )
   n39 = RigVMVariableNode() {variable=Targeting}
   n40 = &RigVMFunction_AlphaInterp(
    InterpSpeedDecreasing=5
    InterpSpeedIncreasing=5
    Scale=1
    bClampResult=False
    Value=&RigVMFunction_MathDoubleSub(A=1, B=&RigVMFunction_MathBoolToFloat(Value=@n39.Value).Result).Result
   )
   > &RigUnit_ModifyTransforms(
    ItemToModify=((Item=(Type=Bone,Name="dyn_shoulderpad_l"),Transform=(Rotation=(X=0.0562,Y=0.0011,Z=0.0187,W=0.9982),Translation=(X=0,Y=-0.6759,Z=0),Scale3D=(X=1,Y=1,Z=1))))
    Weight=@n40.Result
   )
   n41 = &RigVMFunction_ControlFlowBranch(
    Condition=RigVMDispatchNode(B=2, A=RigVMVariableNode() {variable=MelleePistolOtherAssaultRifle}.Value) {template="DISPATCH_RigVMDispatch_CoreEquals(in A,in B,out Result)"}.Result
    ExecuteContext=@n22.True
   )
   n42 = &RigUnit_SphericalPoseReader(
    ActiveRegionScaleFactors=(PositiveWidth=1,NegativeWidth=1,PositiveHeight=1,NegativeHeight=1)
    ActiveRegionSize=0.05
    Debug=(bDrawDebug=false,bDraw2D=false,DebugScale=25,DebugSegments=20,DebugThickness=0.25)
    DriverAxis=(X=0,Y=0,Z=1)
    DriverItem=(Type=Bone,Name="upperarm_twist_01_l")
    FalloffRegionScaleFactors=(PositiveWidth=1,NegativeWidth=1,PositiveHeight=1,NegativeHeight=1)
    FalloffSize=0.2
    OptionalParentItem=(Type=Bone,Name="dyn_sleeve_l_main")
    RotationOffset=(X=0,Y=0,Z=0)
    ExecutePin=@n41.False
   )
   > n43 = &RigUnit_SphericalPoseReader(
    ActiveRegionScaleFactors=(PositiveWidth=0.2,NegativeWidth=0.2,PositiveHeight=0.2,NegativeHeight=1)
    ActiveRegionSize=0.2
    Debug=(bDrawDebug=false,bDraw2D=false,DebugScale=25,DebugSegments=20,DebugThickness=0.25)
    DriverAxis=(X=1,Y=0,Z=0)
    DriverItem=(Type=Bone,Name="clavicle_l")
    FalloffRegionScaleFactors=(PositiveWidth=0.5,NegativeWidth=0.7,PositiveHeight=1,NegativeHeight=1)
    FalloffSize=0.1
    OptionalParentItem=(Type=Bone,Name="pelvis")
    RotationOffset=(X=10,Y=-60,Z=0)
   )
   > n44 = &RigUnit_SphericalPoseReader(
    ActiveRegionScaleFactors=(PositiveWidth=3,NegativeWidth=3,PositiveHeight=1,NegativeHeight=3)
    ActiveRegionSize=0.1
    Debug=(bDrawDebug=true,bDraw2D=false,DebugScale=25,DebugSegments=20,DebugThickness=0.25)
    DriverAxis=(X=0,Y=0,Z=1)
    DriverItem=(Type=Bone,Name="upperarm_twist_01_l")
    FalloffRegionScaleFactors=(PositiveWidth=1,NegativeWidth=1,PositiveHeight=1.4,NegativeHeight=1)
    FalloffSize=0.5
    OptionalParentItem=(Type=Bone,Name="upperarm_l")
    RotationOffset=(X=-40,Y=0,Z=0)
   )
   > n45 = &RigUnit_SphericalPoseReader(
    ActiveRegionScaleFactors=(PositiveWidth=1,NegativeWidth=1,PositiveHeight=1,NegativeHeight=1)
    ActiveRegionSize=0.1
    Debug=(bDrawDebug=false,bDraw2D=false,DebugScale=25,DebugSegments=20,DebugThickness=0.25)
    DriverAxis=(X=1,Y=0,Z=0)
    DriverItem=(Type=Bone,Name="head")
    FalloffRegionScaleFactors=(PositiveWidth=1,NegativeWidth=1,PositiveHeight=1,NegativeHeight=1)
    FalloffSize=0.2
    OptionalParentItem=(Type=Bone,Name="pelvis")
    RotationOffset=(X=-40,Y=0,Z=50)
   )
   > &RigUnit_ModifyTransforms(
    ItemToModify=((Item=(Type=Bone,Name="dyn_shoulderpad_l"),Transform=(Rotation=(X=0.2647,Y=0.0405,Z=-0.0239,W=0.9632),Translation=(X=0.0392,Y=0.0296,Z=0.3968),Scale3D=(X=1,Y=1.1,Z=1))))
    Weight=@n44.OutputParam
   )
   n46 = &RigVMFunction_MathDoubleSub(A=1, B=@n44.OutputParam)
   > &RigUnit_ModifyTransforms(
    ItemToModify=((Item=(Type=Bone,Name="dyn_shoulderpad_l"),Transform=(Rotation=(X=-0.0298,Y=0.0416,Z=0.0506,W=0.9974),Translation=(X=-0.0986,Y=-1.2521,Z=-0.0972),Scale3D=(X=1,Y=1,Z=1.0905))))
    Weight=@n46.Result
   )
   > &RigUnit_ModifyTransforms(
    ItemToModify=((Item=(Type=Bone,Name="dyn_shoulderpad_l"),Transform=(Rotation=(X=-0.0829,Y=0.0234,Z=-0.0454,W=0.9952),Translation=(X=-0.048,Y=-0.5504,Z=0),Scale3D=(X=1,Y=1,Z=1))))
    Weight=@n43.OutputParam
   )
   > &RigUnit_ModifyTransforms(
    ItemToModify=((Item=(Type=Bone,Name="dyn_shoulderpad_l"),Transform=(Rotation=(X=-0.0034,Y=-0.0625,Z=-0.0546,W=0.9965),Translation=(X=0,Y=0,Z=0),Scale3D=(X=1,Y=1,Z=1))))
    Weight=@n42.OutputParam
   )
   > n47 = &RigUnit_ModifyTransforms(
    ItemToModify=((Item=(Type=Bone,Name="dyn_shoulderpad_l"),Transform=(Rotation=(X=-0.0541,Y=0,Z=0,W=0.9985),Translation=(X=0,Y=-0.043,Z=-0.3676),Scale3D=(X=1,Y=1,Z=1))))
    Weight=@n45.OutputParam
   )
   n48 = &RigVMFunction_ControlFlowBranch(
    Condition=RigVMDispatchNode(B=1, A=RigVMVariableNode() {variable=MelleePistolOtherAssaultRifle}.Value) {template="DISPATCH_RigVMDispatch_CoreEquals(in A,in B,out Result)"}.Result
    ExecuteContext=@n47.ExecutePin
   )
   n49 = &RigVMFunction_ControlFlowBranch(
    Condition=RigVMDispatchNode(B=3, A=RigVMVariableNode() {variable=MelleePistolOtherAssaultRifle}.Value) {template="DISPATCH_RigVMDispatch_CoreEquals(in A,in B,out Result)"}.Result
    ExecuteContext=@n48.False
   )
   n50 = &RigVMFunction_ControlFlowBranch(
    Condition=RigVMDispatchNode(B=0, A=RigVMVariableNode() {variable=MelleePistolOtherAssaultRifle}.Value) {template="DISPATCH_RigVMDispatch_CoreEquals(in A,in B,out Result)"}.Result
    ExecuteContext=@n49.False
   )
   &RigUnit_ModifyTransforms(
    ItemToModify=((Item=(Type=Bone,Name="dyn_shoulderpad_l"),Transform=(Rotation=(X=0.1695,Y=-0.0144,Z=0.0835,W=0.9819),Translation=(X=0.5951,Y=-0.4973,Z=-0.1765),Scale3D=(X=1,Y=1,Z=1))))
    ExecutePin=@n50.True
    Weight=@n44.OutputParam
   )
   n51 = &RigVMFunction_MathDoubleRemap(SourceMaximum=0.92, SourceMinimum=0.45, TargetMaximum=1, TargetMinimum=0, Value=@n46.Result)
   &RigUnit_ModifyTransforms(
    ItemToModify=((Item=(Type=Bone,Name="dyn_shoulderpad_l"),Transform=(Rotation=(X=0.2639,Y=0.021,Z=0.0931,W=0.9598),Translation=(X=2.6647,Y=-0.1881,Z=-0.1641),Scale3D=(X=1,Y=1,Z=1.0913))))
    Weight=@n51.Result
    ExecutePin=@n49.True
   )
   > &RigUnit_ModifyTransforms(
    ItemToModify=((Item=(Type=Bone,Name="dyn_shoulderpad_l"),Transform=(Rotation=(X=0,Y=0,Z=0.0175,W=0.9998),Translation=(X=0,Y=0,Z=0),Scale3D=(X=1,Y=1,Z=1))))
    Weight=@n42.OutputParam
   )
   n52 = RigVMVariableNode() {variable=Targeting}
   n53 = &RigVMFunction_AlphaInterp(
    InterpSpeedDecreasing=5
    InterpSpeedIncreasing=5
    Scale=1
    bClampResult=False
    Value=&RigVMFunction_MathDoubleSub(A=1, B=&RigVMFunction_MathBoolToFloat(Value=@n52.Value).Result).Result
   )
   > &RigUnit_ModifyTransforms(
    ItemToModify=((Item=(Type=Bone,Name="dyn_shoulderpad_l"),Transform=(Rotation=(X=-0.0343,Y=0.0009,Z=0.0258,W=0.9991),Translation=(X=0.0291,Y=-0.5973,Z=0.0346),Scale3D=(X=1,Y=0.941,Z=1.1))))
    Weight=@n53.Result
   )
   &RigUnit_ModifyTransforms(
    ItemToModify=((Item=(Type=Bone,Name="dyn_shoulderpad_l"),Transform=(Rotation=(X=-0.062,Y=0.0249,Z=-0.0015,W=0.9978),Translation=(X=-0.0669,Y=-0.0693,Z=-0.5298),Scale3D=(X=1,Y=1,Z=1))))
    Weight=@n51.Result
    ExecutePin=@n48.True
   )
   n54 = RigVMVariableNode() {variable=Targeting}
   n55 = &RigVMFunction_AlphaInterp(
    InterpSpeedDecreasing=10
    InterpSpeedIncreasing=3
    Scale=1
    bClampResult=False
    Value=&RigVMFunction_MathDoubleSub(A=1, B=&RigVMFunction_MathBoolToFloat(Value=@n54.Value).Result).Result
   )
   > &RigUnit_ModifyTransforms(
    ItemToModify=((Item=(Type=Bone,Name="dyn_shoulderpad_l"),Transform=(Rotation=(X=0.2696,Y=0.0131,Z=0.0187,W=0.9627),Translation=(X=1.3369,Y=-0.2824,Z=-0.8007),Scale3D=(X=1,Y=1,Z=1.1))))
    Weight=@n55.Result
   )
   n56 = &RigUnit_SphericalPoseReader(
    ActiveRegionScaleFactors=(PositiveWidth=1,NegativeWidth=1,PositiveHeight=1,NegativeHeight=1)
    ActiveRegionSize=0.05
    Debug=(bDrawDebug=false,bDraw2D=false,DebugScale=25,DebugSegments=20,DebugThickness=0.25)
    DriverAxis=(X=0,Y=0,Z=1)
    DriverItem=(Type=Bone,Name="upperarm_twist_01_l")
    FalloffRegionScaleFactors=(PositiveWidth=1,NegativeWidth=1,PositiveHeight=1,NegativeHeight=1)
    FalloffSize=0.2
    OptionalParentItem=(Type=Bone,Name="dyn_sleeve_l_main")
    RotationOffset=(X=0,Y=0,Z=0)
    ExecutePin=@n41.True
   )
   > n57 = &RigUnit_SphericalPoseReader(
    ActiveRegionScaleFactors=(PositiveWidth=0.2,NegativeWidth=0.2,PositiveHeight=0.2,NegativeHeight=1)
    ActiveRegionSize=0.2
    Debug=(bDrawDebug=false,bDraw2D=false,DebugScale=25,DebugSegments=20,DebugThickness=0.25)
    DriverAxis=(X=1,Y=0,Z=0)
    DriverItem=(Type=Bone,Name="clavicle_l")
    FalloffRegionScaleFactors=(PositiveWidth=0.5,NegativeWidth=0.7,PositiveHeight=1,NegativeHeight=1)
    FalloffSize=0.1
    OptionalParentItem=(Type=Bone,Name="pelvis")
    RotationOffset=(X=40,Y=-60,Z=0)
   )
   > n58 = &RigUnit_SphericalPoseReader(
    ActiveRegionScaleFactors=(PositiveWidth=3,NegativeWidth=3,PositiveHeight=1,NegativeHeight=3)
    ActiveRegionSize=0.1
    Debug=(bDrawDebug=false,bDraw2D=false,DebugScale=25,DebugSegments=20,DebugThickness=0.25)
    DriverAxis=(X=0,Y=0,Z=1)
    DriverItem=(Type=Bone,Name="upperarm_twist_01_l")
    FalloffRegionScaleFactors=(PositiveWidth=1,NegativeWidth=1,PositiveHeight=1.4,NegativeHeight=1)
    FalloffSize=0.5
    OptionalParentItem=(Type=Bone,Name="upperarm_l")
    RotationOffset=(X=-40,Y=0,Z=0)
   )
   > n59 = &RigUnit_SphericalPoseReader(
    ActiveRegionScaleFactors=(PositiveWidth=1,NegativeWidth=1,PositiveHeight=1,NegativeHeight=1)
    ActiveRegionSize=0.1
    Debug=(bDrawDebug=false,bDraw2D=false,DebugScale=25,DebugSegments=20,DebugThickness=0.25)
    DriverAxis=(X=1,Y=0,Z=0)
    DriverItem=(Type=Bone,Name="head")
    FalloffRegionScaleFactors=(PositiveWidth=1,NegativeWidth=1,PositiveHeight=1,NegativeHeight=1)
    FalloffSize=0.2
    OptionalParentItem=(Type=Bone,Name="pelvis")
    RotationOffset=(X=-60,Y=0,Z=40)
   )
   > &RigUnit_ModifyTransforms(
    ItemToModify=((Item=(Type=Bone,Name="dyn_shoulderpad_l"),Transform=(Rotation=(X=0.1251,Y=0.0449,Z=-0.0256,W=0.9908),Translation=(X=0.0326,Y=0.5088,Z=0.1291),Scale3D=(X=1,Y=1,Z=1))))
    Weight=@n58.OutputParam
   )
   n60 = &RigVMFunction_MathDoubleSub(A=1, B=@n58.OutputParam)
   > &RigUnit_ModifyTransforms(
    ItemToModify=((Item=(Type=Bone,Name="dyn_shoulderpad_l"),Transform=(Rotation=(X=-0.2561,Y=0.0056,Z=0.054,W=0.9651),Translation=(X=-0.0986,Y=-1.2521,Z=-0.0972),Scale3D=(X=1,Y=1,Z=1.0905))))
    Weight=@n60.Result
   )
   > &RigUnit_ModifyTransforms(
    ItemToModify=((Item=(Type=Bone,Name="dyn_shoulderpad_l"),Transform=(Rotation=(X=-0.0855,Y=-0.0367,Z=-0.0403,W=0.9948),Translation=(X=-0.048,Y=-0.5504,Z=0),Scale3D=(X=1,Y=1,Z=1))))
    Weight=@n57.OutputParam
   )
   > &RigUnit_ModifyTransforms(
    ItemToModify=((Item=(Type=Bone,Name="dyn_shoulderpad_l"),Transform=(Rotation=(X=-0.0986,Y=0.0054,Z=-0.0544,W=0.9936),Translation=(X=0,Y=0,Z=0),Scale3D=(X=1,Y=1,Z=1))))
    Weight=@n56.OutputParam
   )
   > &RigUnit_ModifyTransforms(
    ItemToModify=((Item=(Type=Bone,Name="dyn_shoulderpad_l"),Transform=(Rotation=(X=0.0364,Y=0.0565,Z=0.0229,W=0.9975),Translation=(X=0.8335,Y=-0.0581,Z=-0.203),Scale3D=(X=1,Y=1,Z=1))))
    Weight=@n59.OutputParam
   )
   > &RigUnit_ModifyTransforms(
    ItemToModify=((Item=(Type=Bone,Name="dyn_shoulderpad_l"),Transform=(Rotation=(X=0.1255,Y=0.0474,Z=0.0479,W=0.9898),Translation=(X=1.0045,Y=0.3304,Z=0.4764),Scale3D=(X=1,Y=1,Z=1))))
    Weight=&RigVMFunction_MathDoubleRemap(SourceMaximum=0.92, SourceMinimum=0.68, TargetMaximum=1, TargetMinimum=0, Value=@n60.Result).Result
   )
   n61 = &RigVMFunction_AlphaInterp(
    InterpSpeedDecreasing=5
    InterpSpeedIncreasing=2
    Scale=1
    bClampResult=False
    Value=&RigVMFunction_MathBoolToFloat(Value=RigVMVariableNode() {variable=static_crouching_rotation}.Value).Result
   )
   > &RigUnit_ModifyTransforms(
    ItemToModify=((Item=(Type=Bone,Name="dyn_shoulderpad_l"),Transform=(Rotation=(X=0.2588,Y=0,Z=0,W=0.9659),Translation=(X=0,Y=0,Z=0),Scale3D=(X=1,Y=1,Z=1))))
    Weight=@n61.Result
   )
   n62 = RigVMVariableNode() {variable=MelleePistolOtherAssaultRifle}
   n63 = &RigVMFunction_AlphaInterp(
    InterpSpeedDecreasing=5
    InterpSpeedIncreasing=2
    Scale=1
    bClampResult=False
    Value=&RigVMFunction_MathBoolToFloat(Value=RigVMDispatchNode(B=2, A=@n62.Value) {template="DISPATCH_RigVMDispatch_CoreEquals(in A,in B,out Result)"}.Result).Result
   )
   > &RigUnit_ModifyTransforms(
    ItemToModify=((Item=(Type=Bone,Name="dyn_shoulderpad_l"),Transform=(Rotation=(X=0.0189,Y=0,Z=0,W=0.9998),Translation=(X=0,Y=-0.5122,Z=0),Scale3D=(X=1,Y=1,Z=1))))
    Weight=@n63.Result
   )
   n64 = &RigUnit_SphericalPoseReader(
    ActiveRegionScaleFactors=(PositiveWidth=1,NegativeWidth=1,PositiveHeight=2,NegativeHeight=2)
    ActiveRegionSize=0.2
    Debug=(bDrawDebug=false,bDraw2D=false,DebugScale=25,DebugSegments=20,DebugThickness=0.25)
    DriverAxis=(X=1,Y=0,Z=0)
    DriverItem=(Type=Bone,Name="clavicle_l")
    FalloffRegionScaleFactors=(PositiveWidth=1,NegativeWidth=1,PositiveHeight=1,NegativeHeight=1)
    FalloffSize=0.6
    OptionalParentItem=(Type=Bone,Name="pelvis")
    RotationOffset=(X=0,Y=-20,Z=0)
    ExecutePin=@n19.True
   )
   > n65 = &RigUnit_SphericalPoseReader(
    ActiveRegionScaleFactors=(PositiveWidth=1,NegativeWidth=0.2,PositiveHeight=1,NegativeHeight=1)
    ActiveRegionSize=0.2
    Debug=(bDrawDebug=false,bDraw2D=false,DebugScale=25,DebugSegments=20,DebugThickness=0.25)
    DriverAxis=(X=1,Y=0,Z=0)
    DriverItem=(Type=Bone,Name="clavicle_l")
    FalloffRegionScaleFactors=(PositiveWidth=1,NegativeWidth=1,PositiveHeight=1,NegativeHeight=1)
    FalloffSize=0.5
    OptionalParentItem=(Type=Bone,Name="pelvis")
    RotationOffset=(X=-10,Y=-100,Z=0)
   )
   > &RigUnit_ModifyTransforms(
    ItemToModify=((Item=(Type=Bone,Name="dyn_shoulderpad_l"),Transform=(Rotation=(X=-0.0398,Y=0.0746,Z=0.0456,W=0.9954),Translation=(X=3,Y=-1.5,Z=-0.0036),Scale3D=(X=1,Y=1,Z=1))))
    Weight=@n64.OutputParam
   )
   > &RigUnit_ModifyTransforms(
    ItemToModify=((Item=(Type=Bone,Name="dyn_shoulderpad_l"),Transform=(Rotation=(X=-0.4243,Y=0.0183,Z=-0.0073,W=0.9053),Translation=(X=2.0338,Y=-0.3653,Z=1.0615),Scale3D=(X=0.9,Y=0.8973,Z=1.0339))))
    Weight=@n65.OutputParam
   )
   n66 = &RigUnit_SphericalPoseReader(
    ActiveRegionScaleFactors=(PositiveWidth=1,NegativeWidth=1,PositiveHeight=1,NegativeHeight=1)
    ActiveRegionSize=0.05
    Debug=(bDrawDebug=false,bDraw2D=false,DebugScale=25,DebugSegments=20,DebugThickness=0.25)
    DriverAxis=(X=0,Y=0,Z=1)
    DriverItem=(Type=Bone,Name="upperarm_twist_01_l")
    FalloffRegionScaleFactors=(PositiveWidth=1,NegativeWidth=1,PositiveHeight=1,NegativeHeight=1)
    FalloffSize=0.2
    OptionalParentItem=(Type=Bone,Name="dyn_sleeve_l_main")
    RotationOffset=(X=0,Y=0,Z=0)
    ExecutePin=@n18.True
   )
   > n67 = &RigUnit_SphericalPoseReader(
    ActiveRegionScaleFactors=(PositiveWidth=0.2,NegativeWidth=0.2,PositiveHeight=0.2,NegativeHeight=1)
    ActiveRegionSize=0.2
    Debug=(bDrawDebug=false,bDraw2D=false,DebugScale=25,DebugSegments=20,DebugThickness=0.25)
    DriverAxis=(X=1,Y=0,Z=0)
    DriverItem=(Type=Bone,Name="clavicle_l")
    FalloffRegionScaleFactors=(PositiveWidth=0.5,NegativeWidth=0.7,PositiveHeight=1,NegativeHeight=1)
    FalloffSize=0.1
    OptionalParentItem=(Type=Bone,Name="pelvis")
    RotationOffset=(X=10,Y=-60,Z=0)
   )
   > n68 = &RigUnit_SphericalPoseReader(
    ActiveRegionScaleFactors=(PositiveWidth=3,NegativeWidth=3,PositiveHeight=1,NegativeHeight=3)
    ActiveRegionSize=0.1
    Debug=(bDrawDebug=false,bDraw2D=false,DebugScale=25,DebugSegments=20,DebugThickness=0.25)
    DriverAxis=(X=0,Y=0,Z=1)
    DriverItem=(Type=Bone,Name="upperarm_twist_01_l")
    FalloffRegionScaleFactors=(PositiveWidth=1,NegativeWidth=1,PositiveHeight=1.4,NegativeHeight=1)
    FalloffSize=0.5
    OptionalParentItem=(Type=Bone,Name="upperarm_l")
    RotationOffset=(X=-40,Y=0,Z=0)
   )
   > n69 = &RigUnit_SphericalPoseReader(
    ActiveRegionScaleFactors=(PositiveWidth=1,NegativeWidth=1,PositiveHeight=1,NegativeHeight=1)
    ActiveRegionSize=0.1
    Debug=(bDrawDebug=false,bDraw2D=false,DebugScale=25,DebugSegments=20,DebugThickness=0.25)
    DriverAxis=(X=1,Y=0,Z=0)
    DriverItem=(Type=Bone,Name="head")
    FalloffRegionScaleFactors=(PositiveWidth=1,NegativeWidth=1,PositiveHeight=1,NegativeHeight=1)
    FalloffSize=0.2
    OptionalParentItem=(Type=Bone,Name="pelvis")
    RotationOffset=(X=-40,Y=0,Z=50)
   )
   > &RigUnit_ModifyTransforms(
    ItemToModify=((Item=(Type=Bone,Name="dyn_shoulderpad_l"),Transform=(Rotation=(X=0.4521,Y=0.0203,Z=0.0166,W=0.8916),Translation=(X=-0.0564,Y=-0.1378,Z=-0.531),Scale3D=(X=1,Y=1,Z=1))))
    Weight=@n68.OutputParam
   )
   > &RigUnit_ModifyTransforms(
    ItemToModify=((Item=(Type=Bone,Name="dyn_shoulderpad_l"),Transform=(Rotation=(X=-0.0286,Y=0.0178,Z=0.0513,W=0.9981),Translation=(X=-0.0986,Y=-1.2521,Z=-0.0972),Scale3D=(X=1,Y=1,Z=1.0905))))
    Weight=&RigVMFunction_MathDoubleSub(A=1, B=@n68.OutputParam).Result
   )
   > &RigUnit_ModifyTransforms(
    ItemToModify=((Item=(Type=Bone,Name="dyn_shoulderpad_l"),Transform=(Rotation=(X=-0.0855,Y=-0.0367,Z=-0.0403,W=0.9948),Translation=(X=-0.048,Y=-0.5504,Z=0),Scale3D=(X=1,Y=1,Z=1))))
    Weight=@n67.OutputParam
   )
   > &RigUnit_ModifyTransforms(
    ItemToModify=((Item=(Type=Bone,Name="dyn_shoulderpad_l"),Transform=(Rotation=(X=-0.1789,Y=0.0098,Z=-0.0538,W=0.9823),Translation=(X=0,Y=0,Z=0),Scale3D=(X=1,Y=1,Z=1))))
    Weight=@n66.OutputParam
   )
   > &RigUnit_ModifyTransforms(
    ItemToModify=((Item=(Type=Bone,Name="dyn_shoulderpad_l"),Transform=(Rotation=(X=-0.0541,Y=0,Z=0,W=0.9985),Translation=(X=0,Y=-0.043,Z=-0.3676),Scale3D=(X=1,Y=1,Z=1))))
    Weight=@n69.OutputParam
   )
   RigVMFunction_MathBoolToFloat = &RigVMFunction_MathBoolToFloat(Value=RigVMVariableNode() {variable=static_crouching_rotation}.Value)
   AlphaInterp = &RigVMFunction_AlphaInterp(InterpSpeedDecreasing=5, InterpSpeedIncreasing=2, Scale=1, bClampResult=False, Value=@RigVMFunction_MathBoolToFloat.Result)
   &RigUnit_ModifyTransforms(
    ItemToModify="((Item=(Type=Bone,Name=\"deform_upperarm_bk_l\"),Transform=(Rotation=(X=0,Y=0.0892,Z=0,W=0.996),Translation=(X=0,Y=0.4206,Z=-1.5073),Scale3D=(X=1,Y=1,Z=1))),(Item=(Type=Bone,Name=\"deform_upperarm_bk_r\"),Transform=(Rotation=(X=0.2186,Y=0.1434,Z=0.0184,W=0.965 ...(truncated)"
    Weight=@AlphaInterp.Result
    ExecutePin=@RigVMFunction_Sequence.F
   )
   n70 = &RigVMFunction_AlphaInterp(
    InterpSpeedDecreasing=5
    InterpSpeedIncreasing=5
    Scale=1
    bClampResult=False
    Value=&RigVMFunction_MathBoolToFloat(Value=RigVMVariableNode() {variable=TacticalJump}.Value).Result
   )
   > &RigUnit_ModifyTransforms(
    ItemToModify=((Item=(Type=Bone,Name="dyn_shoulderpad_l"),Transform=(Rotation=(X=-0.0713,Y=0,Z=0,W=0.9975),Translation=(X=4,Y=0,Z=0),Scale3D=(X=1,Y=1,Z=1))))
    Weight=@n70.Result
   )
   FN_DeformRig_FNC_1 = RigVMFunctionReferenceNode(
    DeformRig_ Curve="(deformRig_l_upperarm_up_alpha_72_#1=False,deformRig_r_upperarm_up_alpha_73_#2=False,deformRig_l_upperarm_fr_alpha_74_#3=False,deformRig_r_upperarm_fr_alpha_75_CB51B8 ...(truncated)"
    DeformRig_Bool="(upperarm_1_#4=True,elbow_3_#5=false,wrist_7_#6=false,groin_8_#7=true,glute_17_#8=true,knee_9_40C408324CA2 ...(truncated)"
    DeformRig_Debug=(upperarm_debug_18_#4=False,wrist_debug_20_#6=False,lat_debug_26_#9=False,groinGlute_debug_27_#7=False)
    DeformRig_Value="(upperarm_up_18_#10=(Rotation=(X=0,Y=-0.4226,Z=0,W=0.9063),Translation=(X=0,Y=0,Z=8),Scale3D=(X=0.7,Y=1,Z=1)),upperarm_fwd_fr_19_#11=(Rotation=(X=0,Y=0,Z=0,W=1),Translation=(X=-10,Y=10,Z=5),Scale3D= ...(truncated)"
    ExecuteContext=()
   )
   &RigVMFunction_MathDoubleSub(A=1, B=@n64.OutputParam)
  ]}
]
hierarchy [
 root {loc=<0 0 0>, rot=<0 0 0 1>}
 pelvis {parent=root, loc=<0 -0.7071 93.1191>, rot=<0.0127 0.707 -0.0127 -0.707>}
 spine_01 {parent=pelvis, loc=<6.7494 -0.6388 0>, rot=<0 0 0.034 -0.9994>}
 spine_02 {parent=spine_01, loc=<7.2518 -0.035 0>, rot=<0 0 0.0238 -0.9997>}
 spine_03 {parent=spine_02, loc=<6.0104 0 0>, rot=<0 0 0.0149 -0.9999>}
 spine_04 {parent=spine_03, loc=<8.1024 0 0>, rot=<0 0 -0.0132 -0.9999>}
 spine_05 {parent=spine_04, loc=<10.0146 0 0>, rot=<0 0 -0.0015 -1>}
 attach_cape {parent=spine_05, loc=<9.017 -11.714 0>, rot=<0.0616 -0.7044 -0.0616 -0.7044>}
 neck_01 {parent=spine_05, loc=<11.3138 0.1529 0>, rot=<0 0 -0.2003 -0.9797>}
 neck_02 {parent=neck_01, loc=<4.5627 0 0>, rot=<0 0 0.0504 -0.9987>}
 head {parent=neck_02, loc=<5.3914 0 0>, rot=<0 0 0.115 -0.9934>}
 dyn_hood {parent=neck_01, loc=<0 0 0>, rot=<0 0 0 1>}
 attach_backpack {parent=spine_05, loc=<-0.123 -13.473 0>, rot=<-0.0093 -0.707 0.0093 -0.707>}
 clavicle_l {parent=spine_05, loc=<4.3575 -0.1466 -2.3223>, rot=<0.0465 -0.7289 -0.009 -0.6829>}
 upperarm_l {parent=clavicle_l, loc=<15.5111 0 0>, rot=<0.0226 -0.3234 -0.0088 -0.9459>}
 lowerarm_l {parent=upperarm_l, loc=<26.6514 0 0>, rot=<0.0391 -0.0099 -0.2453 -0.9686>}
 hand_l {parent=lowerarm_l, loc=<22.3491 0 0>, rot=<-0.7081 -0.0507 0.0975 -0.6975>}
 thumb_01_l {parent=hand_l, loc=<1.3118 -0.3191 -2.459>, rot=<0.6546 -0.4042 0.0664 -0.6354>}
 thumb_02_l {parent=thumb_01_l, loc=<5.9336 -0.0448 0.0401>, rot=<-0.0037 0.0213 0.298 -0.9543>}
 thumb_03_l {parent=thumb_02_l, loc=<2.2223 0.3682 -0.0907>, rot=<0.0118 -0.013 0.0066 -0.9998>}
 index_metacarpal_l {parent=hand_l, loc=<1.6535 0.0183 -2.96>, rot=<0.0158 -0.0124 -0.0173 -0.9996>}
 index_01_l {parent=index_metacarpal_l, loc=<8.3008 0.0357 -0.0524>, rot=<-0.0044 0.0485 0.0924 -0.9945>}
 index_02_l {parent=index_01_l, loc=<3.3587 -0.3004 0.0452>, rot=<-0.002 0.0027 0.3225 -0.9466>}
 index_03_l {parent=index_02_l, loc=<2.5578 -0.123 -0.0252>, rot=<-0.002 0.0067 0.2558 -0.9667>}
 middle_metacarpal_l {parent=hand_l, loc=<2.1856 -0.1725 -0.5984>, rot=<0.006 0.0168 -0.0329 -0.9993>}
 middle_01_l {parent=middle_metacarpal_l, loc=<7.3275 -0.2447 0.0129>, rot=<-0.0208 0.0042 0.0858 -0.9961>}
 middle_02_l {parent=middle_01_l, loc=<4.0123 0.1401 0.1175>, rot=<-0.0218 0.0258 0.3512 -0.9357>}
 middle_03_l {parent=middle_02_l, loc=<2.6077 0.0534 0.1373>, rot=<0.0082 0.0188 0.1708 -0.9851>}
 ring_metacarpal_l {parent=hand_l, loc=<2.2074 -0.3625 1.3638>, rot=<-0.0284 0.0722 -0.0458 -0.9959>}
 ring_01_l {parent=ring_metacarpal_l, loc=<7.0295 -0.6197 -0.1446>, rot=<-0.0489 -0.0252 0.1892 -0.9804>}
 ring_02_l {parent=ring_01_l, loc=<3.6174 0.4647 -0.1027>, rot=<0 0 0.2806 -0.9598>}
 ring_03_l {parent=ring_02_l, loc=<2.4152 0.114 0.0199>, rot=<0 0 0.1189 -0.9929>}
 pinky_metacarpal_l {parent=hand_l, loc=<1.7616 -0.9753 3.3814>, rot=<-0.0397 0.0814 -0.069 -0.9935>}
 pinky_01_l {parent=pinky_metacarpal_l, loc=<7.1709 -0.4831 -0.1094>, rot=<-0.0654 -0.0035 0.2242 -0.9723>}
 pinky_02_l {parent=pinky_01_l, loc=<2.8274 0.0893 -0.1043>, rot=<0.0037 0.0011 0.2854 -0.9584>}
 pinky_03_l {parent=pinky_02_l, loc=<2.2046 0.0307 -0.0078>, rot=<-0.0002 0.0098 0.1445 -0.9895>}
 weapon_l {parent=hand_l, loc=<-0.1974 -3.4111 0.3811>, rot=<0 0 0 1>}
 hand_attach_l {parent=hand_l, loc=<8.5 -3 0>, rot=<0.7071 0.7071 0 0>}
 lowerarm_twist_02_l {parent=lowerarm_l, loc=<9.6199 0 0>, rot=<0 0 0 -1>}
 lowerarm_twist_01_l {parent=lowerarm_l, loc=<17.8549 0 0>, rot=<0 0 0 -1>}
 deform_bracelet_l {parent=lowerarm_twist_01_l, loc=<0 0 0>, rot=<0 0 0 1>}
 dyn_bracelet_l_main {parent=lowerarm_l, loc=<22.3491 0 0>, rot=<0 0 0 1>}
 dyn_bracelet_l_c {parent=dyn_bracelet_l_main, loc=<0 0 0>, rot=<0.6253 -0.2203 -0.1265 -0.7379>}
 dyn_bracelet_l_down {parent=dyn_bracelet_l_c, loc=<0 4.6051 0>, rot=<0.7071 0.7071 0 0>}
 dyn_bracelet_l_up {parent=dyn_bracelet_l_c, loc=<0 -3.6433 0>, rot=<0 0 0.7071 -0.7071>}
 upperarm_twist_01_l {parent=upperarm_l, loc=<0.0041 0 0>, rot=<0 0 0 1>}
 deform_upperarm_bk_base_l {parent=upperarm_twist_01_l, loc=<1.0813 -7.2193 -1.9453>, rot=<0 0 0 1>}
 deform_upperarm_bk_l {parent=deform_upperarm_bk_base_l, loc=<0 0 0>, rot=<0 0 0 1>}
 deform_upperarm_fr_base_l {parent=upperarm_twist_01_l, loc=<-0.3691 7.0506 -2.765>, rot=<0 0 0 1>}
 deform_upperarm_fr_l {parent=deform_upperarm_fr_base_l, loc=<0 0 0>, rot=<0 0 0 1>}
 deform_upperarm_up_base_l {parent=upperarm_twist_01_l, loc=<-3.2216 -1.1391 3.3311>, rot=<0 0 0 1>}
 deform_upperarm_up_l {parent=deform_upperarm_up_base_l, loc=<0 0 0>, rot=<0 0 0 1>}
 upperarm_twist_02_l {parent=upperarm_l, loc=<13.7941 0 0>, rot=<0 0 0 1>}
 dyn_sleeve_l_main {parent=upperarm_l, loc=<10.3102 0 0>, rot=<0 0 0 1>}
 dyn_sleeve_l_up {parent=dyn_sleeve_l_main, loc=<0 0 7.9416>, rot=<0.7071 0 0 -0.7071>}
 dyn_sleeve_l_down {parent=dyn_sleeve_l_main, loc=<0 0 -7.564>, rot=<-0.7071 0 0 -0.7071>}
 dyn_sleeve_l_front {parent=dyn_sleeve_l_main, loc=<0 6.014 0>, rot=<1 0 0 0>}
 dyn_sleeve_l_back {parent=dyn_sleeve_l_main, loc=<0 -6.666 0>, rot=<0 0 0 1>}
 dyn_shoulderpad_l {parent=upperarm_l, loc=<10.3102 0 0>, rot=<0.7086 0 0 -0.7056>}
 clavicle_r {parent=spine_05, loc=<4.3575 -0.1466 2.3223>, rot=<0.7289 0.0465 -0.6829 0.009>}
 upperarm_r {parent=clavicle_r, loc=<-15.5111 0 0>, rot=<0.0226 -0.3234 -0.0088 -0.9459>}
 lowerarm_r {parent=upperarm_r, loc=<-26.6514 0 0>, rot=<0.0391 -0.0099 -0.2453 -0.9686>}
 hand_r {parent=lowerarm_r, loc=<-22.3491 0 0>, rot=<-0.7081 -0.0507 0.0975 -0.6975>}
 thumb_01_r {parent=hand_r, loc=<-1.3118 0.3191 2.459>, rot=<0.6546 -0.4042 0.0664 -0.6354>}
 thumb_02_r {parent=thumb_01_r, loc=<-5.9336 0.0448 -0.0401>, rot=<-0.0037 0.0213 0.298 -0.9543>}
 thumb_03_r {parent=thumb_02_r, loc=<-2.2223 -0.3682 0.0907>, rot=<0.0118 -0.013 0.0066 -0.9998>}
 index_metacarpal_r {parent=hand_r, loc=<-1.6535 -0.0183 2.96>, rot=<0.0158 -0.0124 -0.0173 -0.9996>}
 index_01_r {parent=index_metacarpal_r, loc=<-8.3008 -0.0357 0.0524>, rot=<-0.0044 0.0485 0.0924 -0.9945>}
 index_02_r {parent=index_01_r, loc=<-3.3587 0.3004 -0.0452>, rot=<-0.002 0.0027 0.3225 -0.9466>}
 index_03_r {parent=index_02_r, loc=<-2.5578 0.123 0.0252>, rot=<-0.002 0.0067 0.2558 -0.9667>}
 middle_metacarpal_r {parent=hand_r, loc=<-2.1856 0.1725 0.5984>, rot=<0.006 0.0168 -0.0329 -0.9993>}
 middle_01_r {parent=middle_metacarpal_r, loc=<-7.3275 0.2447 -0.0129>, rot=<-0.0208 0.0042 0.0858 -0.9961>}
 middle_02_r {parent=middle_01_r, loc=<-4.0123 -0.1401 -0.1175>, rot=<-0.0218 0.0258 0.3512 -0.9357>}
 middle_03_r {parent=middle_02_r, loc=<-2.6077 -0.0534 -0.1373>, rot=<0.0082 0.0188 0.1708 -0.9851>}
 ring_metacarpal_r {parent=hand_r, loc=<-2.2074 0.3625 -1.3638>, rot=<-0.0284 0.0722 -0.0458 -0.9959>}
 ring_01_r {parent=ring_metacarpal_r, loc=<-7.0295 0.6197 0.1446>, rot=<-0.0489 -0.0252 0.1892 -0.9804>}
 ring_02_r {parent=ring_01_r, loc=<-3.6174 -0.4647 0.1027>, rot=<0 0 0.2806 -0.9598>}
 ring_03_r {parent=ring_02_r, loc=<-2.4152 -0.114 -0.0199>, rot=<0 0 0.1189 -0.9929>}
 pinky_metacarpal_r {parent=hand_r, loc=<-1.7616 0.9753 -3.3814>, rot=<-0.0397 0.0814 -0.069 -0.9935>}
 pinky_01_r {parent=pinky_metacarpal_r, loc=<-7.1709 0.4831 0.1094>, rot=<-0.0654 -0.0035 0.2242 -0.9723>}
 pinky_02_r {parent=pinky_01_r, loc=<-2.8274 -0.0893 0.1043>, rot=<0.0037 0.0011 0.2854 -0.9584>}
 pinky_03_r {parent=pinky_02_r, loc=<-2.2046 -0.0307 0.0078>, rot=<-0.0002 0.0098 0.1445 -0.9895>}
 hand_attach_r {parent=hand_r, loc=<-8.5001 2.9999 0>, rot=<0 0 -0.7071 -0.7071>}
 weapon_r {parent=hand_r, loc=<0.1974 3.4111 -0.3811>, rot=<0 0 0 1>}
 lowerarm_twist_02_r {parent=lowerarm_r, loc=<-9.6199 0 0>, rot=<0 0 0 -1>}
 lowerarm_twist_01_r {parent=lowerarm_r, loc=<-17.8549 0 0>, rot=<0 0 0 -1>}
 dyn_bracelet_r_main {parent=lowerarm_r, loc=<-22.3488 -0.0001 -0.0003>, rot=<0 0 0 -1>}
 dyn_bracelet_r_c {parent=dyn_bracelet_r_main, loc=<0 0 0>, rot=<0.6246 -0.2221 -0.1285 -0.7376>}
 dyn_bracelet_r_down {parent=dyn_bracelet_r_c, loc=<-0.0007 -4.1928 0.0002>, rot=<0.7071 0.7071 0 0>}
 dyn_bracelet_r_up {parent=dyn_bracelet_r_c, loc=<-0.0001 4.0423 0>, rot=<0 0 0.7071 -0.7071>}
 upperarm_twist_01_r {parent=upperarm_r, loc=<-0.0041 0 0>, rot=<0 0 0 -1>}
 deform_upperarm_bk_base_r {parent=upperarm_twist_01_r, loc=<-1.0814 7.2192 1.9455>, rot=<0 0 0 -1>}
 deform_upperarm_bk_r {parent=deform_upperarm_bk_base_r, loc=<0 0 0>, rot=<0 0 0 1>}
 deform_upperarm_fr_base_r {parent=upperarm_twist_01_r, loc=<0.3689 -7.0506 2.7652>, rot=<0 0 0 -1>}
 deform_upperarm_fr_r {parent=deform_upperarm_fr_base_r, loc=<0 0 0>, rot=<0 0 0 1>}
 deform_upperarm_up_base_r {parent=upperarm_twist_01_r, loc=<3.2214 1.1391 -3.3308>, rot=<0 0 0 -1>}
 deform_upperarm_up_r {parent=deform_upperarm_up_base_r, loc=<0 0 0>, rot=<0 0 0 -1>}
 upperarm_twist_02_r {parent=upperarm_r, loc=<-13.7941 0 0>, rot=<0 0 0 -1>}
 dyn_sleeve_r_main {parent=upperarm_r, loc=<-10.3105 0 0.0003>, rot=<0 0 0 -1>}
 dyn_sleeve_r_up {parent=dyn_sleeve_r_main, loc=<0.0001 0 -5.4601>, rot=<0.7071 0 0 -0.7071>}
 dyn_sleeve_r_down {parent=dyn_sleeve_r_main, loc=<0.0003 0 7.5636>, rot=<-0.7071 0 0 -0.7071>}
 dyn_sleeve_r_front {parent=dyn_sleeve_r_main, loc=<0.0002 -6.0136 -0.0002>, rot=<1 0 0 0>}
 dyn_sleeve_r_back {parent=dyn_sleeve_r_main, loc=<0 6.6656 0.0001>, rot=<0 0 0 -1>}
 dyn_necklace_l {parent=spine_05, loc=<6.9412 7.7372 -8.1729>, rot=<0.402 -0.0383 0.8771 0.2601>}
 dyn_necklace_c {parent=spine_05, loc=<7.299 8.336 0>, rot=<0 0 0.9588 0.2841>}
 dyn_necklace_c_end {parent=dyn_necklace_c, loc=<7.7041 0 0>, rot=<0 0 0 1>}
 dyn_necklace_r {parent=spine_05, loc=<6.9413 7.7372 8.1729>, rot=<0.0383 0.402 0.2601 -0.8771>}
 deform_pec_aim_base_l {parent=spine_04, loc=<7.8732 12.6388 -2.4123>, rot=<0.0102 -0.4602 0.3979 -0.7936>}
 deform_pec_aim_l {parent=deform_pec_aim_base_l, loc=<0 0 0>, rot=<0 0 0 1>}
 deform_pec_base_l {parent=deform_pec_aim_l, loc=<10.766 0 0>, rot=<0 0 0 1>}
 deform_pec_l {parent=deform_pec_base_l, loc=<0.055 0 0>, rot=<0 0 0 1>}
 deform_pec_aim_base_r {parent=spine_04, loc=<7.8727 12.6388 2.4123>, rot=<-0.4602 -0.0102 0.7936 0.3979>}
 deform_pec_aim_r {parent=deform_pec_aim_base_r, loc=<0 0 0>, rot=<0 0 0 1>}
 deform_pec_base_r {parent=deform_pec_aim_r, loc=<-10.7662 -0.0006 -0.0007>, rot=<0 0 0 1>}
 deform_pec_r {parent=deform_pec_base_r, loc=<0 0 0>, rot=<0 0 0 1>}
 dyn_left_loin_1 {parent=spine_01, loc=<2.0706 -2.5331 -13.6827>, rot=<0.1925 -0.7909 -0.5685 -0.1191>}
 dyn_left_loin_2 {parent=dyn_left_loin_1, loc=<6.9552 0.3385 0>, rot=<0 0 -0.1474 -0.9891>}
 dyn_left_loin_3 {parent=dyn_left_loin_1, loc=<3.678 2.7624 7.5829>, rot=<0.253 0.5082 -0.1525 -0.809>}
 dyn_right_loin_1 {parent=spine_01, loc=<5.7761 -6.2667 13.5987>, rot=<0.1541 -0.8589 0.4833 0.0707>}
 dyn_right_loin_2 {parent=dyn_right_loin_1, loc=<7.6209 0 0>, rot=<-0.1191 -0.0569 0.192 -0.9725>}
 thigh_l {parent=pelvis, loc=<-2.5693 -0.0109 -7.7501>, rot=<0.0012 0.0572 0.0155 -0.9982>}
 calf_l {parent=thigh_l, loc=<-39.0484 0 0>, rot=<0.0006 -0.0161 -0.0391 -0.9991>}
 foot_l {parent=calf_l, loc=<-42.561 0.1077 0.4292>, rot=<0.0418 -0.0455 0.0343 -0.9975>}
 ball_l {parent=foot_l, loc=<-6.1599 13.8264 0.6806>, rot=<0.0001 -0.0001 -0.7063 -0.7079>}
 calf_twist_02_l {parent=calf_l, loc=<-18.8064 -0.0967 0.0397>, rot=<0.0237 0.0095 -0.0035 -0.9997>}
 calf_twist_01_l {parent=calf_l, loc=<-35.813 -0.0967 0.0397>, rot=<0.0237 0.0095 -0.0035 -0.9997>}
 deform_knee_in_base_l {parent=calf_l, loc=<1.298 -5.2957 -0.0496>, rot=<0 0 0 -1>}
 deform_knee_in_l {parent=deform_knee_in_base_l, loc=<0 0 0>, rot=<0 0 0 1>}
 deform_knee_out_base_l {parent=calf_l, loc=<0 4 0>, rot=<0 0 0 -1>}
 deform_knee_out_l {parent=deform_knee_out_base_l, loc=<0 0 0>, rot=<0 0 0 1>}
 thigh_twist_01_l {parent=thigh_l, loc=<-21.6303 0 0>, rot=<0.0013 0 0 -1>}
 deform_groin_base_l {parent=thigh_l, loc=<-2.1908 8.1142 -1.7442>, rot=<0 0 0 -1>}
 deform_groin_l {parent=deform_groin_base_l, loc=<0 0 0>, rot=<0 0 0 1>}
 thigh_r {parent=pelvis, loc=<-2.5693 -0.0109 7.7501>, rot=<0.0572 -0.0012 0.9982 0.0155>}
 calf_r {parent=thigh_r, loc=<39.0484 0 0>, rot=<0.0006 -0.0161 -0.0391 -0.9991>}
 foot_r {parent=calf_r, loc=<42.561 -0.1077 -0.4292>, rot=<0.0418 -0.0455 0.0343 -0.9975>}
 ball_r {parent=foot_r, loc=<6.1599 -13.8264 -0.6806>, rot=<0.0001 -0.0001 -0.7063 -0.7079>}
 calf_twist_02_r {parent=calf_r, loc=<18.8064 0.0967 -0.0397>, rot=<0.0237 0.0095 -0.0035 -0.9997>}
 calf_twist_01_r {parent=calf_r, loc=<35.813 0.0967 -0.0397>, rot=<0.0237 0.0095 -0.0035 -0.9997>}
 deform_knee_in_base_r {parent=calf_r, loc=<-1.298 5.2957 0.0496>, rot=<0 0 0 -1>}
 deform_knee_in_r {parent=deform_knee_in_base_r, loc=<0 0 0>, rot=<0 0 0 1>}
 deform_knee_out_base_r {parent=calf_r, loc=<0 -4 0>, rot=<0 0 0 -1>}
 deform_knee_out_r {parent=deform_knee_out_base_r, loc=<0 0 0>, rot=<0 0 0 1>}
 thigh_twist_01_r {parent=thigh_r, loc=<21.6303 0 0>, rot=<0.0013 0 0 -1>}
 deform_groin_base_r {parent=thigh_r, loc=<2.1908 -8.1142 1.7442>, rot=<0 0 0 -1>}
 deform_groin_r {parent=deform_groin_base_r, loc=<0 0 0>, rot=<0 0 0 1>}
 dyn_skirt_main_r {parent=pelvis, loc=<0 0 0>, rot=<-0.4854 0.5246 -0.4705 -0.5175>}
 dyn_skirt_r {parent=dyn_skirt_main_r, loc=<14.1508 -1.2464 -0.2758>, rot=<-0.0214 0.0135 -0.6096 -0.7923>}
 dyn_skirt_1_r {parent=dyn_skirt_r, loc=<16.9559 0 0>, rot=<0.0045 0.0011 -0.0391 -0.9992>}
 dyn_skirt_2_r {parent=dyn_skirt_1_r, loc=<14.976 0 0>, rot=<0 0 -0.0397 -0.9992>}
 dyn_right_loin_correct {parent=dyn_skirt_main_r, loc=<13.0445 2.6643 -7.2537>, rot=<-0.2973 -0.1976 -0.5709 -0.7393>}
 dyn_skirt_main_bk_r {parent=pelvis, loc=<0 0 0>, rot=<0.6494 -0.7127 0.1763 0.1979>}
 dyn_skirt_bk_r {parent=dyn_skirt_main_bk_r, loc=<14.4284 -1.2613 -0.2264>, rot=<0.0483 0.079 -0.6326 -0.7689>}
 dyn_skirt_bk_1_r {parent=dyn_skirt_bk_r, loc=<16.0803 0 0>, rot=<0 0 0 1>}
 dyn_skirt_main_bk_l {parent=pelvis, loc=<0 0 0>, rot=<0.651 -0.7072 -0.1844 -0.205>}
 dyn_skilt_bk_l {parent=dyn_skirt_main_bk_l, loc=<14.4402 -1.1372 -0.0993>, rot=<-0.058 -0.0854 -0.6457 -0.7566>}
 dyn_skilt_bk_1_l {parent=dyn_skilt_bk_l, loc=<16.0803 0 0>, rot=<0 0 0 1>}
 dyn_skirt_main_l {parent=pelvis, loc=<0 0 0>, rot=<0.4187 -0.4615 -0.5308 -0.5745>}
 dyn_skirt_l {parent=dyn_skirt_main_l, loc=<13.5699 -4.1954 -0.18>, rot=<0.0974 0.0675 -0.5491 -0.8273>}
 dyn_skirt_1_l {parent=dyn_skirt_l, loc=<10.4662 -0.1209 -0.0541>, rot=<0.0131 -0.0207 -0.0949 -0.9952>}
 dyn_skirt_2_l {parent=dyn_skirt_1_l, loc=<11.3252 0 0>, rot=<0 0 0 1>}
 dyn_knife {parent=dyn_skirt_l, loc=<10.1501 1.5169 -0.091>, rot=<0.0131 -0.0207 -0.0949 -0.9952>}
 dyn_knife_chain_f {parent=dyn_skirt_main_l, loc=<12.1838 -2.6657 5.5455>, rot=<0.1555 0.077 -0.6237 -0.7621>}
 dyn_knife_chain_b {parent=dyn_skirt_main_l, loc=<12.9356 -1.8718 -6.5003>, rot=<-0.0847 -0.1705 -0.4901 -0.8506>}
 dyn_ornament_main {parent=pelvis, loc=<7.435 -7.3692 13.3028>, rot=<0.2421 -0.8291 0.4888 0.1227>}
 dyn_ornament_1 {parent=dyn_ornament_main, loc=<6.7072 0.0095 0.1748>, rot=<0.0293 0.0216 0.0235 -0.9991>}
 dyn_ornament_2 {parent=dyn_ornament_1, loc=<2.3367 -0.0021 0.4208>, rot=<0.008 -0.0347 -0.0679 -0.9971>}
 deform_glute_base_l {parent=pelvis, loc=<-8.8064 -8.8222 -8.4287>, rot=<0.0275 0.0501 -0.4472 -0.8926>}
 deform_glute_l {parent=deform_glute_base_l, loc=<0 0 0>, rot=<0 0 0 1>}
 deform_glute_base_r {parent=pelvis, loc=<-8.8062 -8.8222 8.4287>, rot=<0.0501 -0.0275 0.8926 -0.4472>}
 deform_glute_r {parent=deform_glute_base_r, loc=<0 0 0>, rot=<0 0 0 1>}
 attach {parent=root, loc=<0 0 0>, rot=<0 0 0 1>}
 ik_hand_root {parent=root, loc=<0 0 0>, rot=<0 0 0 1>}
 ik_hand_gun {parent=ik_hand_root, loc=<-51.9727 4.0883 102.6935>, rot=<0.6084 0.1793 0.4613 -0.6205>}
 ik_hand_r {parent=ik_hand_gun, loc=<0 0 0>, rot=<0 0 0 1>}
 ik_hand_l {parent=ik_hand_gun, loc=<53.0285 82.1781 35.2051>, rot=<-0.5102 -0.7906 -0.3387 0>}
 ik_foot_root {parent=root, loc=<0 0 0>, rot=<0 0 0 1>}
 ik_foot_l {parent=ik_foot_root, loc=<15.2608 -4.4176 9.4432>, rot=<0.0357 0.7045 0.0284 -0.7082>}
 ik_foot_r {parent=ik_foot_root, loc=<-15.2608 -4.4176 9.4432>, rot=<0.7082 0.0284 -0.7045 0.0357>}
 attach_fp {parent=root, loc=<0 0 0>, rot=<0 0 0 1>}
 camera_root {parent=root, loc=<0 0 0>, rot=<0 0 0 1>}
 camera {parent=camera_root, loc=<0 0 0>, rot=<0 0 0 1>}
 attach_emote_01 {parent=root, loc=<0 0 0>, rot=<0 0 0 1>}
 attach_emote_02 {parent=root, loc=<0 0 0>, rot=<0 0 0 1>}
 attach_emote_03 {parent=root, loc=<0 0 0>, rot=<0 0 0 1>}
 warp_point_01 {parent=root, loc=<0 0 0>, rot=<0 0 0 1>}
 warp_point_02 {parent=root, loc=<0 0 0>, rot=<0 0 0 1>}
 warp_point_03 {parent=root, loc=<0 0 0>, rot=<0 0 0 1>}
]
variables [
 IsBaseballSliding {type=double, default=0}
 IsCrouching {type=double, default=0}
 IsCrouchingBool {type=bool, default=False}
 IsFalling {type=bool, default=False}
 IsInFrontEnd {type=bool, default=False}
 IsJumping {type=bool, default=False}
 IsOnGround {type=bool, default=False}
 IsSliding {type=double, default=0}
 Left {type=double, default=0}
 MelleePistolOtherAssaultRifle {type=int32, default=0}
 Right {type=double, default=0}
 SkyDiving {type=bool, default=False}
 TacticalJump {type=bool, default=False}
 TacticalMovement {type=bool, default=False}
 Targeting {type=bool, default=False}
 static_crouching_rotation {type=bool, default=False}
]
