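# Physics asset for the front-end variant of the male medium PaperBlaze body, covering its dangling skirt, loincloth, ornament and knife bones.
# Limited-angle chains hang from skirt anchor bones, the pelvis (ornament) and spine_01 (loincloths), while skirt anchors, knife and body links use free angular motion.
# Capsules and a hand sphere cover only selected body bones, and a huge zero-length root capsule sits mostly below the origin, resembling a floor collider.
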PhysicsAsset /BRCosmetics/Characters/Player/Male/Medium/Bodies/M_MED_PaperBlaze/Meshes/M_MED_PaperBlaze_Phys1_FrontEnd
bodies {
 calf_r [{shapeType=Capsule, length=46.17, radius=9.0435, center=<27.3475 1.841 -1.0738>, rotation=<rot 90 0 0>}]
 dyn_knife [{
   shapeType=Capsule
   length=35.9535
   radius=6.1043
   shapeName=dyn_knife_capsule
   center=<0.4402 3.3184 1.0812>
   rotation=<rot -0.4353 -90 -17.9437>
  }]
 dyn_left_loin_1 [{shapeType=Capsule, length=12.6166, radius=5, center=<1.3004 -0.5628 0>}]
 dyn_left_loin_2 [{shapeType=Capsule, length=11.5443, radius=5, center=<1.8004 -4.8329 -0.0296>, rotation=<rot 0.0001 0 4.7516>}]
 dyn_ornament_1 [{
   shapeType=Capsule
   length=2.7514
   radius=1.8608
   shapeName=dyn_ornament_1_capsule
   center=<0.2341 0.7312 -0.1245>
   rotation=<rot 5.1594 -90 -2.9979>
  }]
 dyn_ornament_2 [{
   shapeType=Capsule
   length=3.483
   radius=10.3003
   shapeName=dyn_ornament_2_capsule
   center=<12.3949 -5.4382 -0.0079>
   rotation=<rot 0 -90.5303 -88.9523>
  }]
 dyn_ornament_main [{
   shapeType=Capsule
   length=7.2494
   radius=5.4572
   shapeName=dyn_ornament_main_capsule
   center=<1.367 4.256 -0.5565>
   rotation=<rot -1.7934 97.1791 -83.3247>
  }]
 dyn_right_loin_1 [{shapeType=Capsule, length=6.9461, radius=5.8387, center=<1.4478 -4.8998 0.7154>, rotation=<rot 0 0 7.5571>}]
 dyn_right_loin_2 [{shapeType=Capsule, length=1.4909, radius=5.1766, center=<1.6891 -5.1122 0.2114>}]
 dyn_skirt_1_r [{shapeType=Capsule, length=11, radius=7.5, center=<4.5847 7.2739 0>}]
 dyn_skirt_2_r [{shapeType=Capsule, length=11, radius=7.5, center=<4.5847 7.2739 0>, rotation=<rot 19.6874 0 0>}]
 dyn_skirt_bk_1_r [{shapeType=Capsule, length=11, radius=7.5, center=<4.5847 7.2739 0>}]
 dyn_skirt_bk_r [{shapeType=Capsule, length=11, radius=7.5, center=<4.5847 8.3592 0>}]
 dyn_skirt_l [{shapeType=Capsule, length=5, radius=5, center=<6.8046 0.5869 0>, rotation=<rot 90 0 0>}]
 dyn_skirt_main_bk_l [{shapeType=Capsule, length=10, radius=1.5, rotation=<rot 90 0 0>}]
 dyn_skirt_main_bk_r [{shapeType=Capsule, length=10, radius=1.5, rotation=<rot 90 0 0>}]
 dyn_skirt_main_r [{shapeType=Capsule, length=10, radius=1.5, rotation=<rot 90 0 0>}]
 dyn_skirt_r [{shapeType=Capsule, length=11, radius=7.5, center=<4.5847 6.5834 0>}]
 hand_l [{shapeType=Sphere, radius=8.2965, shapeName=hand_l_sphere, center=<7.8534 -0.6807 0>}]
 lowerarm_l [{shapeType=Capsule, length=0.505, radius=0.505, shapeName=lowerarm_l_capsule, rotation=<rot 0 -90 0>}]
 pelvis [{shapeType=Capsule, length=7.5, radius=12, center=<-10.2448 -1.2387 0>}]
 root [{shapeType=Capsule, length=0, radius=218.8018, shapeName=root_capsule, center=<0 0 -213.4834>, rotation=<rot 0 -90 0>}]
 spine_01 [{
   shapeType=Capsule
   length=0.8644
   radius=9.5316
   shapeName=spine_01_capsule
   center=<2.9656 0.5491 0.8656>
   rotation=<rot -5.1645 -90 -0.1578>
  }]
 thigh_l [{shapeType=Capsule, length=23, radius=9.5, center=<-21.4345 0 0>, rotation=<rot 0 90 -90>}]
 thigh_r [{shapeType=Capsule, length=23, radius=9.5, center=<21.4345 0 0>, rotation=<rot 0 90 90>}]
}
constraints [
 {
  bone1Name=thigh_r
  bone2Name=pelvis
  swing1LimitDegrees=45
  swing1Motion=Free
  swing2LimitDegrees=45
  swing2Motion=Free
  twistLimitDegrees=45
  twistMotion=Free
 }
 {
  bone1Name=dyn_skirt_main_bk_l
  bone2Name=pelvis
  swing1LimitDegrees=45
  swing1Motion=Free
  swing2LimitDegrees=45
  swing2Motion=Free
  twistLimitDegrees=45
  twistMotion=Free
 }
 {
  bone1Name=dyn_skirt_main_bk_r
  bone2Name=pelvis
  swing1LimitDegrees=45
  swing1Motion=Free
  swing2LimitDegrees=45
  swing2Motion=Free
  twistLimitDegrees=45
  twistMotion=Free
 }
 {bone1Name=dyn_skirt_bk_r, bone2Name=dyn_skirt_main_bk_r, swing1LimitDegrees=30, swing2LimitDegrees=25, twistLimitDegrees=10}
 {bone1Name=dyn_skirt_bk_1_r, bone2Name=dyn_skirt_bk_r, swing1LimitDegrees=30, swing2LimitDegrees=25, twistLimitDegrees=10}
 {
  bone1Name=dyn_skirt_main_r
  bone2Name=pelvis
  swing1LimitDegrees=45
  swing1Motion=Free
  swing2LimitDegrees=45
  swing2Motion=Free
  twistLimitDegrees=45
  twistMotion=Free
 }
 {bone1Name=dyn_skirt_r, bone2Name=dyn_skirt_main_r, swing1LimitDegrees=30, swing2LimitDegrees=25, twistLimitDegrees=10}
 {bone1Name=dyn_skirt_1_r, bone2Name=dyn_skirt_r, swing1LimitDegrees=30, swing2LimitDegrees=25, twistLimitDegrees=10}
 {bone1Name=dyn_skirt_2_r, bone2Name=dyn_skirt_1_r, swing1LimitDegrees=30, swing2LimitDegrees=25, twistLimitDegrees=10}
 {
  bone1Name=thigh_l
  bone2Name=pelvis
  swing1LimitDegrees=45
  swing1Motion=Free
  swing2LimitDegrees=45
  swing2Motion=Free
  twistLimitDegrees=45
  twistMotion=Free
 }
 {bone1Name=dyn_ornament_main, bone2Name=pelvis, swing1LimitDegrees=15, swing2LimitDegrees=25, twistLimitDegrees=10}
 {bone1Name=dyn_ornament_1, bone2Name=dyn_ornament_main, swing1LimitDegrees=15, swing2LimitDegrees=1, twistLimitDegrees=1}
 {bone1Name=dyn_ornament_2, bone2Name=dyn_ornament_1, swing1LimitDegrees=60, swing2LimitDegrees=5, twistLimitDegrees=1}
 {bone1Name=dyn_left_loin_1, bone2Name=spine_01, swing1LimitDegrees=25, swing2LimitDegrees=5, twistLimitDegrees=1}
 {bone1Name=dyn_left_loin_2, bone2Name=dyn_left_loin_1, swing1LimitDegrees=35, swing2LimitDegrees=15, twistLimitDegrees=1}
 {bone1Name=dyn_right_loin_1, bone2Name=spine_01, swing1LimitDegrees=20, swing2LimitDegrees=15, twistLimitDegrees=10}
 {bone1Name=dyn_right_loin_2, bone2Name=dyn_right_loin_1, swing1LimitDegrees=30, swing2LimitDegrees=20, twistLimitDegrees=10}
 {
  bone1Name=spine_01
  bone2Name=pelvis
  swing1LimitDegrees=45
  swing1Motion=Free
  swing2LimitDegrees=45
  swing2Motion=Free
  twistLimitDegrees=45
  twistMotion=Free
 }
 {
  bone1Name=calf_r
  bone2Name=thigh_r
  swing1LimitDegrees=45
  swing1Motion=Free
  swing2LimitDegrees=45
  swing2Motion=Free
  twistLimitDegrees=45
  twistMotion=Free
 }
 {
  bone1Name=hand_l
  bone2Name=lowerarm_l
  swing1LimitDegrees=45
  swing1Motion=Free
  swing2LimitDegrees=45
  swing2Motion=Free
  twistLimitDegrees=45
  twistMotion=Free
 }
 {
  bone1Name=lowerarm_l
  bone2Name=spine_01
  swing1LimitDegrees=45
  swing1Motion=Free
  swing2LimitDegrees=45
  swing2Motion=Free
  twistLimitDegrees=45
  twistMotion=Free
 }
 {
  bone1Name=dyn_knife
  bone2Name=dyn_skirt_l
  swing1LimitDegrees=45
  swing1Motion=Free
  swing2LimitDegrees=45
  swing2Motion=Free
  twistLimitDegrees=45
  twistMotion=Free
 }
 {
  bone1Name=dyn_skirt_l
  bone2Name=pelvis
  swing1LimitDegrees=45
  swing1Motion=Free
  swing2LimitDegrees=45
  swing2Motion=Free
  twistLimitDegrees=45
  twistMotion=Free
 }
 {
  bone1Name=pelvis
  bone2Name=root
  swing1LimitDegrees=45
  swing1Motion=Free
  swing2LimitDegrees=45
  swing2Motion=Free
  twistLimitDegrees=45
  twistMotion=Free
 }
]
