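# Physics asset for the front-end variant of the M_MED_PaperBlaze male character skirt, driving secondary motion of skirt, knife and sleeves.
# Capsules cover root, pelvis and legs, while front and back skirt chains per side hang from pelvis through free-moving main bones with limited joints below.
# Mostly free cross-constraints link right front to back and both back chains.
# The knife hangs tightly from the left front chain; eight sleeve bodies constrain directly to pelvis.
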PhysicsAsset /BRCosmetics/Characters/Player/Male/Medium/Bodies/M_MED_PaperBlaze/Meshes/M_MED_PaperBlaze_Skirt_FrontEnd
bodies {
 calf_l [{shapeType=Capsule, length=37.1466, radius=16.8706, center=<-28.7825 7.2164 7.3215>, rotation=<rot 90 0 -172.1708>}]
 calf_r [{shapeType=Capsule, length=44.6966, radius=14.4134, center=<31.1068 -5.0805 -6.9768>, rotation=<rot 88.3692 -8.7603 0.001>}]
 dyn_knife [
  {shapeType=Sphere, radius=13.9147, shapeName=dyn_knife_sphere, center=<0.4412 -12.2422 21.1829>}
  {shapeType=Sphere, radius=30.5182, shapeName=dyn_knife_sphere, center=<-0.2508 -31.6488 -27.8897>}
  {
   shapeType=Capsule
   length=21.2464
   radius=9.5133
   shapeName=dyn_knife_capsule
   center=<1.6693 -8.5468 -13.6287>
   rotation=<rot 1.848 -89.5905 6.3906>
  }
 ]
 dyn_skilt_bk_1_l [{shapeType=Capsule, length=8.3786, radius=7.5, center=<4.5847 -6.0155 0>, rotation=<rot 90 0 0>}]
 dyn_skilt_bk_l [{shapeType=Capsule, length=0.8705, radius=7.5, center=<4.5847 -6.0155 0>}]
 dyn_skirt_1_l [{shapeType=Capsule, length=9.8098, radius=5.5111, center=<2.4653 -5.3726 0>}]
 dyn_skirt_1_r [{shapeType=Capsule, length=11, radius=7.5, center=<4.5847 -6.0155 0>}]
 dyn_skirt_2_l [{shapeType=Capsule, length=11.4308, radius=6.6113, center=<7.6499 -4.8164 0.0598>, rotation=<rot 0.0004 -0.0002 1.2958>}]
 dyn_skirt_2_r [{shapeType=Capsule, length=33.7791, radius=18.3963, center=<4.4353 -16.9428 -1.6764>}]
 dyn_skirt_bk_1_r [
  {shapeType=Capsule, length=24.5, radius=7.5, center=<4.5847 -6.0155 0>}
  {
   shapeType=Capsule
   length=14.973
   radius=7.2098
   shapeName=dyn_skirt_bk_1_r_capsule
   center=<14.0399 -4.8239 -7.4822>
   rotation=<rot 10.4371 -2.8412 -11.8914>
  }
 ]
 dyn_skirt_bk_r [{shapeType=Capsule, length=11, radius=7.5, center=<4.5847 -6.0155 0>}]
 dyn_skirt_l [{shapeType=Capsule, length=9.1575, radius=3.9006, center=<2.4653 -4.0409 0>}]
 dyn_skirt_main_bk_l [{shapeType=Capsule, length=10, radius=1.5, rotation=<rot 90 0 0>}]
 dyn_skirt_main_bk_r [{shapeType=Capsule, length=10, radius=1.5, rotation=<rot 90 0 0>}]
 dyn_skirt_main_l [{shapeType=Capsule, length=0.505, radius=0.505, shapeName=dyn_skirt_main_l_capsule, rotation=<rot 0 -90 0>}]
 dyn_skirt_main_r [{shapeType=Capsule, length=10, radius=1.5, rotation=<rot 90 0 0>}]
 dyn_skirt_r [{shapeType=Capsule, length=9.1892, radius=11.6371, center=<21.0725 -8.9844 -6.5629>, rotation=<rot 90 180 177.5852>}]
 dyn_sleeve_l_back [{shapeType=Capsule, length=11.0151, radius=6.1657, center=<5.8421 5.4918 0.6537>, rotation=<rot 90 179.5524 -177.6361>}]
 dyn_sleeve_l_down [{shapeType=Capsule, length=11.0151, radius=6.1657, center=<5.8421 5.4918 0.6537>, rotation=<rot 90 -179.7762 -176.9642>}]
 dyn_sleeve_l_front [{shapeType=Capsule, length=11.0151, radius=6.1657, center=<5.8421 5.4918 0.6537>, rotation=<rot 90 180 -177.1876>}]
 dyn_sleeve_l_up [{shapeType=Capsule, length=11.0151, radius=6.1657, center=<5.8421 5.4918 0.6537>, rotation=<rot 90 180 -177.1881>}]
 dyn_sleeve_r_back [{shapeType=Capsule, length=7.3091, radius=6.1657, center=<-6.1742 -4.5902 0>, rotation=<rot 87.1875 -2.8175 2.8065>}]
 dyn_sleeve_r_down [{shapeType=Capsule, length=7.3091, radius=6.1657, center=<-6.1742 -4.5902 0>, rotation=<rot 87.1872 -2.8178 2.8068>}]
 dyn_sleeve_r_front [{shapeType=Capsule, length=7.3091, radius=6.1657, center=<-6.0092 -3.6921 -0.0059>, rotation=<rot 87.1875 -2.8244 5.6135>}]
 dyn_sleeve_r_up [{shapeType=Capsule, length=7.3091, radius=6.1657, center=<-6.1742 -5.3628 0>, rotation=<rot 87.1877 -2.818 2.8066>}]
 pelvis [{shapeType=Capsule, length=7, radius=11.9901, center=<-10.2448 -1.2387 0>}]
 root [{shapeType=Capsule, length=54.9527, radius=24.2785, center=<-5.698 0.8759 -4.951>, rotation=<rot 0 0 90>}]
 thigh_l [{shapeType=Capsule, length=21.4521, radius=11.0479, center=<-21.4016 0.4949 1.795>, rotation=<rot 0 86.1981 -90>}]
 thigh_r [{shapeType=Capsule, length=23, radius=9.5, center=<21.4345 0 0>, rotation=<rot 0 90 90>}]
}
constraints [
 {
  bone1Name=thigh_r
  bone2Name=pelvis
  swing1LimitDegrees=45
  swing1Motion=Free
  swing2LimitDegrees=45
  swing2Motion=Free
  twistLimitDegrees=45
  twistMotion=Free
 }
 {
  bone1Name=dyn_skirt_main_bk_l
  bone2Name=pelvis
  swing1LimitDegrees=45
  swing1Motion=Free
  swing2LimitDegrees=45
  swing2Motion=Free
  twistLimitDegrees=45
  twistMotion=Free
 }
 {bone1Name=dyn_skilt_bk_l, bone2Name=dyn_skirt_main_bk_l, swing1LimitDegrees=35, swing2LimitDegrees=15, twistLimitDegrees=1}
 {bone1Name=dyn_skilt_bk_1_l, bone2Name=dyn_skilt_bk_l, swing1LimitDegrees=30, swing2LimitDegrees=8, twistLimitDegrees=5}
 {
  bone1Name=dyn_skirt_main_bk_r
  bone2Name=pelvis
  swing1LimitDegrees=45
  swing1Motion=Free
  swing2LimitDegrees=45
  swing2Motion=Free
  twistLimitDegrees=45
  twistMotion=Free
 }
 {bone1Name=dyn_skirt_bk_r, bone2Name=dyn_skirt_main_bk_r, swing1LimitDegrees=35, swing2LimitDegrees=15, twistLimitDegrees=1}
 {bone1Name=dyn_skirt_bk_1_r, bone2Name=dyn_skirt_bk_r, swing1LimitDegrees=30, swing2LimitDegrees=15, twistLimitDegrees=1}
 {
  bone1Name=dyn_skirt_main_r
  bone2Name=pelvis
  swing1LimitDegrees=45
  swing1Motion=Free
  swing2LimitDegrees=45
  swing2Motion=Free
  twistLimitDegrees=45
  twistMotion=Free
 }
 {bone1Name=dyn_skirt_r, bone2Name=dyn_skirt_main_r, swing1LimitDegrees=30, swing2LimitDegrees=25, twistLimitDegrees=1}
 {bone1Name=dyn_skirt_1_r, bone2Name=dyn_skirt_r, swing1LimitDegrees=30, swing2LimitDegrees=30, twistLimitDegrees=30}
 {bone1Name=dyn_skirt_2_r, bone2Name=dyn_skirt_1_r, swing1LimitDegrees=35, swing2LimitDegrees=25, twistLimitDegrees=10}
 {
  bone1Name=calf_r
  bone2Name=thigh_r
  swing1LimitDegrees=45
  swing1Motion=Free
  swing2LimitDegrees=45
  swing2Motion=Free
  twistLimitDegrees=45
  twistMotion=Free
 }
 {
  bone1Name=calf_l
  bone2Name=thigh_l
  swing1LimitDegrees=45
  swing1Motion=Free
  swing2LimitDegrees=45
  swing2Motion=Free
  twistLimitDegrees=45
  twistMotion=Free
 }
 {
  bone1Name=thigh_l
  bone2Name=pelvis
  swing1LimitDegrees=45
  swing1Motion=Free
  swing2LimitDegrees=45
  swing2Motion=Free
  twistLimitDegrees=45
  twistMotion=Free
 }
 {bone1Name=dyn_skirt_l, bone2Name=dyn_skirt_main_l, swing1LimitDegrees=35, swing2LimitDegrees=5, twistLimitDegrees=10, twistMotion=Locked}
 {bone1Name=dyn_skirt_1_l, bone2Name=dyn_skirt_l, swing1LimitDegrees=20, swing2LimitDegrees=15, twistLimitDegrees=1}
 {bone1Name=dyn_skirt_2_l, bone2Name=dyn_skirt_1_l, swing1LimitDegrees=50, swing2LimitDegrees=10, twistLimitDegrees=10}
 {bone1Name=dyn_knife, bone2Name=dyn_skirt_1_l, swing1LimitDegrees=3, swing1Motion=Locked, swing2LimitDegrees=3, twistLimitDegrees=0.1}
 {bone1Name=dyn_sleeve_l_back, bone2Name=pelvis, swing1LimitDegrees=10, swing2LimitDegrees=10, twistLimitDegrees=5}
 {bone1Name=dyn_sleeve_l_down, bone2Name=pelvis, swing1LimitDegrees=10, swing2LimitDegrees=10, twistLimitDegrees=5}
 {bone1Name=dyn_sleeve_l_front, bone2Name=pelvis, swing1LimitDegrees=10, swing2LimitDegrees=10, twistLimitDegrees=5}
 {bone1Name=dyn_sleeve_l_up, bone2Name=pelvis, swing1LimitDegrees=10, swing2LimitDegrees=10, twistLimitDegrees=5}
 {bone1Name=dyn_sleeve_r_back, bone2Name=pelvis, swing1LimitDegrees=10, swing2LimitDegrees=10, twistLimitDegrees=5}
 {bone1Name=dyn_sleeve_r_down, bone2Name=pelvis, swing1LimitDegrees=10, swing2LimitDegrees=10, twistLimitDegrees=5}
 {bone1Name=dyn_sleeve_r_front, bone2Name=pelvis, swing1LimitDegrees=10, swing2LimitDegrees=10, twistLimitDegrees=5}
 {bone1Name=dyn_sleeve_r_up, bone2Name=pelvis, swing1LimitDegrees=10, swing2LimitDegrees=10, twistLimitDegrees=5}
 {
  bone1Name=pelvis
  bone2Name=root
  swing1LimitDegrees=45
  swing1Motion=Free
  swing2LimitDegrees=45
  swing2Motion=Free
  twistLimitDegrees=45
  twistMotion=Free
 }
 {
  bone1Name=dyn_skirt_main_l
  bone2Name=pelvis
  swing1LimitDegrees=45
  swing1Motion=Free
  swing2LimitDegrees=45
  swing2Motion=Free
  twistLimitDegrees=45
  twistMotion=Free
 }
 {
  bone1Name=dyn_skirt_bk_r
  bone2Name=dyn_skirt_r
  swing1LimitDegrees=45
  swing1Motion=Free
  swing2LimitDegrees=45
  swing2Motion=Free
  twistLimitDegrees=45
  twistMotion=Free
 }
 {
  bone1Name=dyn_skirt_bk_1_r
  bone2Name=dyn_skirt_1_r
  swing1LimitDegrees=45
  swing1Motion=Free
  swing2LimitDegrees=45
  swing2Motion=Free
  twistLimitDegrees=45
  twistMotion=Free
 }
 {
  bone1Name=dyn_skirt_bk_1_r
  bone2Name=dyn_skirt_2_r
  swing1LimitDegrees=45
  swing1Motion=Free
  swing2LimitDegrees=45
  swing2Motion=Free
  twistLimitDegrees=45
  twistMotion=Free
 }
 {
  bone1Name=dyn_skilt_bk_l
  bone2Name=dyn_skirt_bk_r
  swing1LimitDegrees=45
  swing1Motion=Free
  swing2LimitDegrees=10
  twistLimitDegrees=45
  twistMotion=Free
 }
 {
  bone1Name=dyn_skilt_bk_1_l
  bone2Name=dyn_skirt_bk_1_r
  swing1LimitDegrees=45
  swing1Motion=Free
  swing2LimitDegrees=45
  swing2Motion=Free
  twistLimitDegrees=45
  twistMotion=Free
 }
 {
  bone1Name=dyn_skirt_l
  bone2Name=dyn_knife
  swing1LimitDegrees=10
  swing2LimitDegrees=5
  swing2Motion=Free
  twistLimitDegrees=0.1
  twistMotion=Free
 }
]
