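# Animation Blueprint for the PaperBlaze player body, adding secondary motion and pose fixes to skirt, sleeves, bracelets and knife chain.
# RigidBody passes simulate from a cached pose, blending between front-end and in-game physics assets, then sleeve Anim Dynamics follow.
# The output chain starts from a pre-bracelet cached pose, layering weighted constraints, springs and scale-based bracelet hiding.
# Variables for movement, sprint, DBNO, emote, whip and front-end state gate corrective bone offsets.
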AnimBlueprint /BRCosmetics/Characters/Player/Male/Medium/Bodies/M_MED_PaperBlaze/Meshes/M_MED_PaperBlaze_AnimBP
target_skeleton=&/M_MED_PaperBlaze_Skeleton
graphs [
 AnimGraph {nodes [
   n1 = Anim Dynamics - Bone: dyn_bracelet_l_up(Alpha=1, ExternalForce="0,0,0") {
    angularDampingOverride=0.99
    angularSpringConstant=500
    bAngularSpring=true
    bGravityOverrideInSimSpace=true
    bOverrideAngularDamping=true
    gravityScale=0.3
    linearDampingOverride=1
    linearSpringConstant=500
    numSolverIterationsPostUpdate=4
    numSolverIterationsPreUpdate=16
    simulationSpace=BoneRelative
    alphaBoolBlend {blendOption=Linear}
    boundBone {boneName=dyn_bracelet_l_up}
    componentAppliedLinearAccClamp=<3000 3000 3000>
    componentLinearAccScale=<0.5 0.5 0.5>
    componentLinearVelScale=<0.5 0.5 0.5>
    physicsBodyDefinitions [{
      sphereCollisionRadius=5
      boundBone {boneName=dyn_bracelet_l_up}
      boxExtents=<5 5 10>
      constraintSetup {angularLimitsMax=<5 5 5>, angularLimitsMin=<-5 -5 -5>, angularTarget=<1 0 0>}
      localJointOffset=<0.5 0 0>
     }]
    relativeSpaceBone {boneName=lowerarm_l}
   }
   n2 = Anim Dynamics - Bone: dyn_bracelet_l_down(ComponentPose=@n1, Alpha=1, ExternalForce="0,0,0") {
    angularDampingOverride=0.99
    angularSpringConstant=500
    bAngularSpring=true
    bGravityOverrideInSimSpace=true
    bOverrideAngularDamping=true
    gravityScale=0.3
    linearDampingOverride=1
    linearSpringConstant=500
    numSolverIterationsPostUpdate=4
    numSolverIterationsPreUpdate=16
    simulationSpace=BoneRelative
    alphaBoolBlend {blendOption=Linear}
    boundBone {boneName=dyn_bracelet_l_down}
    componentAppliedLinearAccClamp=<3000 3000 3000>
    componentLinearAccScale=<0.5 0.5 0.5>
    componentLinearVelScale=<0.5 0.5 0.5>
    physicsBodyDefinitions [{
      sphereCollisionRadius=5
      boundBone {boneName=dyn_bracelet_l_down}
      boxExtents=<5 5 10>
      constraintSetup {angularLimitsMax=<5 5 5>, angularLimitsMin=<-5 -5 -5>, angularTarget=<1 0 0>}
      localJointOffset=<0.5 0 0>
     }]
    relativeSpaceBone {boneName=lowerarm_l}
   }
   Anim Dynamics - Bone: dyn_bracelet_l_c(ComponentPose=@n2, Alpha=1, ExternalForce="0,0,0") {
    angularDampingOverride=0.99
    angularSpringConstant=500
    bAngularSpring=true
    bGravityOverrideInSimSpace=true
    bLinearSpring=true
    bOverrideAngularDamping=true
    bOverrideLinearDamping=true
    linearDampingOverride=0.8
    linearSpringConstant=50
    numSolverIterationsPostUpdate=4
    numSolverIterationsPreUpdate=16
    simulationSpace=BoneRelative
    alphaBoolBlend {blendOption=Linear}
    boundBone {boneName=dyn_bracelet_l_c}
    componentAppliedLinearAccClamp=<3000 3000 3000>
    componentLinearAccScale=<0.5 0.5 0.5>
    componentLinearVelScale=<0.5 0.5 0.5>
    physicsBodyDefinitions [{
      sphereCollisionRadius=5
      boundBone {boneName=dyn_bracelet_l_c}
      boxExtents=<5 10 10>
      constraintSetup {angularLimitsMax=<5 5 0>, angularLimitsMin=<-5 -5 0>, angularTarget=<1 0 0>}
      localJointOffset=<0.5 0 0>
     }]
    relativeSpaceBone {boneName=pelvis}
   }
   n3 = Anim Dynamics - Bone: dyn_bracelet_r_main(Alpha=1, ExternalForce="0,0,0") {
    angularDampingOverride=0.99
    angularSpringConstant=50
    bGravityOverrideInSimSpace=true
    bLinearSpring=true
    bOverrideAngularDamping=true
    bOverrideLinearDamping=true
    bUsePlanarLimit=false
    bUseSphericalLimits=true
    linearDampingOverride=0.7
    linearSpringConstant=200
    numSolverIterationsPostUpdate=4
    numSolverIterationsPreUpdate=16
    simulationSpace=BoneRelative
    alphaBoolBlend {blendOption=Linear}
    boundBone {boneName=dyn_bracelet_r_main}
    componentAppliedLinearAccClamp=<50 150 50>
    componentLinearAccScale=<1 1 1>
    componentLinearVelScale=<1 1 1>
    physicsBodyDefinitions [{
      sphereCollisionRadius=0
      boundBone {boneName=dyn_bracelet_r_main}
      boxExtents=<5 10 10>
      constraintSetup {angularLimitsMax=<0 0 0>, angularLimitsMin=<0 0 0>, angularTarget=<1 0 0>, linearAxesMax=<5 0 0>, linearAxesMin=<-1 0 0>}
      localJointOffset=<-1 0 0>
     }]
    relativeSpaceBone {boneName=lowerarm_r}
    sphericalLimits [{limitRadius=6, drivingBone {boneName=hand_r}, sphereLocalOffset=<-4 0 0>}]
   }
   n4 = Anim Dynamics - Bone: dyn_bracelet_r_up(ComponentPose=@n3, Alpha=1, ExternalForce="0,0,0") {
    alpha=0.6
    angularDampingOverride=0.99
    angularSpringConstant=500
    bAngularSpring=true
    bGravityOverrideInSimSpace=true
    bLinearSpring=true
    bOverrideAngularDamping=true
    bOverrideLinearDamping=true
    gravityScale=0.3
    linearDampingOverride=1
    linearSpringConstant=500
    numSolverIterationsPostUpdate=4
    numSolverIterationsPreUpdate=16
    simulationSpace=BoneRelative
    alphaBoolBlend {blendOption=Linear}
    alphaScaleBiasClamp {bMapRange=true, outRange {max=0.6}}
    boundBone {boneName=dyn_bracelet_r_up}
    componentAppliedLinearAccClamp=<3000 3000 3000>
    componentLinearAccScale=<0.5 0.5 0.5>
    componentLinearVelScale=<0.5 0.5 0.5>
    physicsBodyDefinitions [{
      sphereCollisionRadius=5
      boundBone {boneName=dyn_bracelet_r_up}
      boxExtents=<5 5 10>
      constraintSetup {angularLimitsMax=<5 5 5>, angularLimitsMin=<-5 -5 -5>, angularTarget=<1 0 0>}
      localJointOffset=<-0.1 0 0>
     }]
    relativeSpaceBone {boneName=lowerarm_r}
   }
   Anim Dynamics - Bone: dyn_bracelet_r_down(ComponentPose=@n4, Alpha=1, ExternalForce="0,0,0") {
    alpha=0.6
    angularDampingOverride=0.99
    angularSpringConstant=500
    bAngularSpring=true
    bGravityOverrideInSimSpace=true
    bLinearSpring=true
    bOverrideAngularDamping=true
    bOverrideLinearDamping=true
    gravityScale=0.3
    linearDampingOverride=0.7
    linearSpringConstant=1500
    numSolverIterationsPostUpdate=4
    numSolverIterationsPreUpdate=16
    simulationSpace=BoneRelative
    alphaBoolBlend {blendOption=Linear}
    alphaScaleBiasClamp {bMapRange=true, outRange {max=0.6}}
    boundBone {boneName=dyn_bracelet_r_down}
    componentAppliedLinearAccClamp=<3000 3000 3000>
    componentLinearAccScale=<1 1 1>
    componentLinearVelScale=<1 1 1>
    physicsBodyDefinitions [{
      sphereCollisionRadius=5
      boundBone {boneName=dyn_bracelet_r_down}
      boxExtents=<5 5 10>
      constraintSetup {angularLimitsMax=<5 5 5>, angularLimitsMin=<-5 -5 -5>, angularTarget=<1 0 0>}
      localJointOffset=<-2 0 0>
     }]
    relativeSpaceBone {boneName=lowerarm_r}
   }
   n5 = Get bIsOnGrindRail()
   n6 = "Transform (Modify) Bone - Bone: dyn_skirt_l"!(ComponentPose, bAlphaBoolEnabled:bool=@n5, Translation="0,0,0", Rotation="-0.002,-0.000,-6.400", Scale="1,1,1") -> (Pose) {
    alphaInputType=Bool
    rotationMode=BMM_Additive
    rotationSpace=BCS_BoneSpace
    translationMode=BMM_Additive
    alphaBoolBlend {blendInTime=0.1, blendOption=Linear, blendOutTime=0.1}
    alphaScaleBiasClamp {bInterpResult=true}
    boneToModify {boneName=dyn_skirt_l}
    rotation=<rot -0.002 0 -6.4>
   }
   n7 = "Transform (Modify) Bone - Bone: dyn_left_loin_3"!(ComponentPose=@n6, bAlphaBoolEnabled:bool=@n5, Translation="0,0,0", Rotation="0,0,0", Scale="1,1,1") -> (Pose) {
    alphaInputType=Bool
    rotationMode=BMM_Additive
    rotationSpace=BCS_BoneSpace
    translationMode=BMM_Additive
    alphaBoolBlend {blendInTime=0.1, blendOption=Linear, blendOutTime=0.1}
    alphaScaleBiasClamp {bInterpResult=true}
    boneToModify {boneName=dyn_left_loin_3}
   }
   "Transform (Modify) Bone - Bone: dyn_shoulderpad_l"!(ComponentPose=@n7, bAlphaBoolEnabled:bool=@n5, Translation="0,0,0", Rotation="0,0,0", Scale="1,1,1") -> (Pose) {
    alphaInputType=Bool
    rotationMode=BMM_Additive
    rotationSpace=BCS_BoneSpace
    translationMode=BMM_Additive
    alphaBoolBlend {blendInTime=0.1, blendOption=Linear, blendOutTime=0.1}
    alphaScaleBiasClamp {bInterpResult=true}
    boneToModify {boneName=dyn_shoulderpad_l}
   }
   n8 = "Constraint - dyn_bracelet_l_c (RR)"!(
    ComponentPose=Local To Component(LocalPose=Use cached pose 'Pre Braceletes'())
    Alpha=Get Rifle Bracelet Correction Alpha() -> (Rifle Bracelet Correction Alpha:float)
    ConstraintWeights_0=0.2
    ConstraintWeights_1:float=0.8
   ) -> (Pose) {
    alphaBoolBlend {blendOption=Linear}
    alphaScaleBiasClamp {bInterpResult=true, bMapRange=true}
    boneToModify {boneName=dyn_bracelet_l_c}
    constraintSetup [{transformType=Rotation, targetBone {boneName=lowerarm_twist_02_l}}, {transformType=Rotation, targetBone {boneName=hand_l}}]
    constraintWeights [1,1]
   }
   n9 = Get HideLeftBraselet() -> (HideLeftBraselet:double)
   n10 = "Constraint - dyn_bracelet_l_up (PPPPPPP)"!(
    ComponentPose=@n8
    Alpha=@n9
    ConstraintWeights_0=0.0255
    ConstraintWeights_1:float=0.101
    ConstraintWeights_2:float=0
    ConstraintWeights_3:float=0.003
    ConstraintWeights_4:float=0.0052
    ConstraintWeights_5:float=0.22
    ConstraintWeights_6:float=0.6452
   ) -> (Pose) {
    alphaBoolBlend {blendOption=Linear}
    alphaScaleBiasClamp {bMapRange=true, outRange {max=0.4}}
    boneToModify {boneName=dyn_bracelet_l_up}
    constraintSetup [
     {targetBone {boneName=lowerarm_l}}
     {targetBone {boneName=hand_l}}
     {targetBone {boneName=thumb_01_l}}
     {targetBone {boneName=index_metacarpal_l}}
     {targetBone {boneName=pinky_metacarpal_l}}
     {targetBone {boneName=lowerarm_twist_02_l}}
     {targetBone {boneName=lowerarm_twist_01_l}}
    ]
    constraintWeights [1,1,1,1,1,1,1]
   }
   n11 = "Constraint - dyn_bracelet_l_down (PPPPPPP)"!(
    ComponentPose=@n10
    Alpha=@n9
    ConstraintWeights_0=0.0333
    ConstraintWeights_1:float=0.0002
    ConstraintWeights_2:float=0.045
    ConstraintWeights_3:float=0.034
    ConstraintWeights_4:float=0.1625
    ConstraintWeights_5:float=0.4011
    ConstraintWeights_6:float=0.324
   ) -> (Pose) {
    alphaBoolBlend {blendOption=Linear}
    alphaScaleBiasClamp {bMapRange=true, outRange {max=0.5}}
    boneToModify {boneName=dyn_bracelet_l_down}
    constraintSetup [
     {targetBone {boneName=lowerarm_l}}
     {targetBone {boneName=hand_l}}
     {targetBone {boneName=thumb_01_l}}
     {targetBone {boneName=pinky_metacarpal_l}}
     {targetBone {boneName=lowerarm_twist_02_l}}
     {targetBone {boneName=lowerarm_twist_01_l}}
     {targetBone {boneName=deform_bracelet_l}}
    ]
    constraintWeights [1,1,1,1,1,1,1]
   }
   n12 = Get Stop wobling left() -> (Stop wobling left:double)
   n13 = Anim Dynamics - Bone: dyn_bracelet_l_up!(ComponentPose=@n11, Alpha=@n12) -> (Pose) {
    angularDampingOverride=0.99
    angularSpringConstant=500
    bAngularSpring=true
    bGravityOverrideInSimSpace=true
    bOverrideAngularDamping=true
    gravityScale=0.01
    linearDampingOverride=1
    linearSpringConstant=500
    numSolverIterationsPostUpdate=4
    numSolverIterationsPreUpdate=16
    simulationSpace=BoneRelative
    alphaBoolBlend {blendOption=Linear}
    boundBone {boneName=dyn_bracelet_l_up}
    componentAppliedLinearAccClamp=<1000 1000 1000>
    physicsBodyDefinitions [{
      sphereCollisionRadius=5
      boundBone {boneName=dyn_bracelet_l_up}
      boxExtents=<5 5 10>
      constraintSetup {angularLimitsMax=<5 5 5>, angularLimitsMin=<-5 -5 -5>, angularTarget=<1 0 0>}
      localJointOffset=<0.5 0 0>
     }]
    relativeSpaceBone {boneName=hand_l}
   }
   n14 = Anim Dynamics - Bone: dyn_bracelet_l_down!(ComponentPose=@n13, Alpha=@n12) -> (Pose) {
    angularDampingOverride=0.99
    angularSpringConstant=500
    bAngularSpring=true
    bGravityOverrideInSimSpace=true
    bOverrideAngularDamping=true
    gravityScale=0.01
    linearDampingOverride=1
    linearSpringConstant=500
    numSolverIterationsPostUpdate=4
    numSolverIterationsPreUpdate=16
    simulationSpace=BoneRelative
    alphaBoolBlend {blendOption=Linear}
    boundBone {boneName=dyn_bracelet_l_down}
    componentAppliedLinearAccClamp=<1000 1000 1000>
    physicsBodyDefinitions [{
      sphereCollisionRadius=5
      boundBone {boneName=dyn_bracelet_l_down}
      boxExtents=<5 5 10>
      constraintSetup {angularLimitsMax=<5 5 5>, angularLimitsMin=<-5 -5 -5>, angularTarget=<1 0 0>}
      localJointOffset=<0.5 0 0>
     }]
    relativeSpaceBone {boneName=hand_l}
   }
   n15 = Anim Dynamics - Bone: dyn_bracelet_l_c!(ComponentPose=@n14, Alpha=@n9) -> (Pose) {
    angularDampingOverride=1
    angularSpringConstant=1
    bAngularSpring=true
    bGravityOverrideInSimSpace=true
    bLinearSpring=true
    bOverrideAngularDamping=true
    bOverrideLinearDamping=true
    gravityScale=0
    linearDampingOverride=1
    linearSpringConstant=1
    numSolverIterationsPostUpdate=4
    numSolverIterationsPreUpdate=16
    simulationSpace=BoneRelative
    alphaBoolBlend {blendOption=Linear}
    boundBone {boneName=dyn_bracelet_l_c}
    componentAppliedLinearAccClamp=<100 100 100>
    physicsBodyDefinitions [{
      sphereCollisionRadius=5
      boundBone {boneName=dyn_bracelet_l_c}
      boxExtents=<5 10 10>
      constraintSetup {angularLimitsMax=<5 5 0>, angularLimitsMin=<-5 -5 0>, angularTarget=<1 0 0>}
      localJointOffset=<0.01 0 0>
     }]
    relativeSpaceBone {boneName=hand_l}
   }
   n16 = "Transform (Modify) Bone - Bone: dyn_bracelet_r_c"!(ComponentPose=@n15, bAlphaBoolEnabled:bool=Get is whip() -> (is whip:bool), Translation="0.885,0.196,-1.880", Rotation="9.979,21.428,7.542") -> (Pose) {
    alphaInputType=Bool
    rotationMode=BMM_Additive
    rotationSpace=BCS_BoneSpace
    scaleMode=BMM_Additive
    translationMode=BMM_Additive
    alphaBoolBlend {blendInTime=0.1, blendOption=Linear, blendOutTime=0.1}
    alphaScaleBiasClamp {bInterpResult=true}
    boneToModify {boneName=dyn_bracelet_r_c}
    rotation=<rot 9.979 21.428 7.542>
    translation=<0.885 0.196 -1.88>
   }
   n17 = "Transform (Modify) Bone - Bone: dyn_bracelet_r_c"!(ComponentPose=@n16, bAlphaBoolEnabled=NOT Boolean(A=Get bIsInFrontEnd()), Translation="-0.437,0.501,0.129") -> (Pose) {
    alphaInputType=Bool
    rotationMode=BMM_Additive
    rotationSpace=BCS_BoneSpace
    scaleMode=BMM_Additive
    translationMode=BMM_Additive
    alphaBoolBlend {blendInTime=0.1, blendOption=Linear, blendOutTime=0.1}
    alphaScaleBiasClamp {bInterpResult=true}
    boneToModify {boneName=dyn_bracelet_r_c}
    translation=<-0.437 0.501 0.129>
   }
   n18 = Get HideRightBraselet() -> (HideRightBraselet:double)
   n19 = "Constraint - dyn_bracelet_r_down (PPPPP)"!(
    ComponentPose=@n17
    Alpha=@n18
    ConstraintWeights_0=0.0879
    ConstraintWeights_1:float=0.0301
    ConstraintWeights_2:float=0.0333
    ConstraintWeights_3:float=0.3441
    ConstraintWeights_4:float=0.5046
   ) -> (Pose) {
    alphaBoolBlend {blendOption=Linear}
    alphaScaleBiasClamp {bMapRange=true, outRange {max=0.7}}
    boneToModify {boneName=dyn_bracelet_r_down}
    constraintSetup [
     {targetBone {boneName=lowerarm_r}}
     {targetBone {boneName=thumb_01_r}}
     {targetBone {boneName=pinky_metacarpal_r}}
     {targetBone {boneName=lowerarm_twist_02_r}}
     {targetBone {boneName=lowerarm_twist_01_r}}
    ]
    constraintWeights [1,1,1,1,1]
   }
   n20 = "Constraint - dyn_bracelet_r_up (PPPPPP)"!(
    ComponentPose=@n19
    Alpha=@n18
    ConstraintWeights_0=0.0415
    ConstraintWeights_1:float=0.0529
    ConstraintWeights_2:float=0.0077
    ConstraintWeights_3:float=0.0112
    ConstraintWeights_4:float=0.2465
    ConstraintWeights_5:float=0.6403
   ) -> (Pose) {
    alphaBoolBlend {blendOption=Linear}
    alphaScaleBiasClamp {bMapRange=true}
    boneToModify {boneName=dyn_bracelet_r_up}
    constraintSetup [
     {targetBone {boneName=lowerarm_r}}
     {targetBone {boneName=hand_r}}
     {targetBone {boneName=index_metacarpal_r}}
     {targetBone {boneName=pinky_metacarpal_r}}
     {targetBone {boneName=lowerarm_twist_02_r}}
     {targetBone {boneName=lowerarm_twist_01_r}}
    ]
    constraintWeights [1,1,1,1,1,1]
   }
   n21 = "Constraint - dyn_bracelet_r_c (RR)"!(ComponentPose=@n20, Alpha=@n18, ConstraintWeights_0=0.8, ConstraintWeights_1:float=0.2) -> (Pose) {
    alphaBoolBlend {blendOption=Linear}
    alphaScaleBiasClamp {bInterpResult=true, bMapRange=true, outRange {max=0.6}}
    boneToModify {boneName=dyn_bracelet_r_c}
    constraintSetup [{transformType=Rotation, targetBone {boneName=lowerarm_twist_01_r}}, {transformType=Rotation, targetBone {boneName=hand_r}}]
    constraintWeights [1,1]
   }
   n22 = Anim Dynamics - Bone: dyn_bracelet_r_main!(ComponentPose=@n21, Alpha=@n18) -> (Pose) {
    angularDampingOverride=1
    angularSpringConstant=5000
    bGravityOverrideInSimSpace=true
    bLinearSpring=true
    bOverrideAngularDamping=true
    bOverrideLinearDamping=true
    bUsePlanarLimit=false
    bUseSphericalLimits=true
    gravityScale=0
    linearDampingOverride=1
    linearSpringConstant=5000
    numSolverIterationsPostUpdate=4
    numSolverIterationsPreUpdate=16
    simulationSpace=BoneRelative
    alphaBoolBlend {blendOption=Linear}
    boundBone {boneName=dyn_bracelet_r_main}
    componentAppliedLinearAccClamp=<100 100 100>
    physicsBodyDefinitions [{
      sphereCollisionRadius=0
      boundBone {boneName=dyn_bracelet_r_main}
      boxExtents=<5 10 10>
      constraintSetup {angularLimitsMax=<0 0 0>, angularLimitsMin=<0 0 0>, angularTarget=<1 0 0>, linearAxesMax=<5 0 0>, linearAxesMin=<-1 0 0>}
      localJointOffset=<-0.01 0 0>
     }]
    relativeSpaceBone {boneName=lowerarm_r}
    sphericalLimits [{limitRadius=6, drivingBone {boneName=hand_r}, sphereLocalOffset=<-4 0 0>}]
   }
   n23 = Get Stop wobling right() -> (Stop wobling right:double)
   n24 = Anim Dynamics - Bone: dyn_bracelet_r_up!(ComponentPose=@n22, Alpha=@n23) -> (Pose) {
    alpha=0.6
    angularDampingOverride=0.99
    angularSpringConstant=500
    bAngularSpring=true
    bGravityOverrideInSimSpace=true
    bLinearSpring=true
    bOverrideAngularDamping=true
    bOverrideLinearDamping=true
    gravityScale=0.3
    linearDampingOverride=1
    linearSpringConstant=500
    numSolverIterationsPostUpdate=4
    numSolverIterationsPreUpdate=16
    simulationSpace=BoneRelative
    alphaBoolBlend {blendOption=Linear}
    alphaScaleBiasClamp {bMapRange=true, outRange {max=0.6}}
    boundBone {boneName=dyn_bracelet_r_up}
    componentAppliedLinearAccClamp=<3000 3000 3000>
    componentLinearAccScale=<0.5 0.5 0.5>
    componentLinearVelScale=<0.5 0.5 0.5>
    physicsBodyDefinitions [{
      sphereCollisionRadius=5
      boundBone {boneName=dyn_bracelet_r_up}
      boxExtents=<5 5 10>
      constraintSetup {angularLimitsMax=<5 5 5>, angularLimitsMin=<-5 -5 -5>, angularTarget=<1 0 0>}
      localJointOffset=<-0.1 0 0>
     }]
    relativeSpaceBone {boneName=lowerarm_r}
   }
   n25 = Anim Dynamics - Bone: dyn_bracelet_r_down!(ComponentPose=@n24, Alpha=@n23) -> (Pose) {
    alpha=0.6
    angularDampingOverride=0.99
    angularSpringConstant=500
    bAngularSpring=true
    bGravityOverrideInSimSpace=true
    bLinearSpring=true
    bOverrideAngularDamping=true
    bOverrideLinearDamping=true
    gravityScale=0.3
    linearDampingOverride=0.7
    linearSpringConstant=1500
    numSolverIterationsPostUpdate=4
    numSolverIterationsPreUpdate=16
    simulationSpace=BoneRelative
    alphaBoolBlend {blendOption=Linear}
    alphaScaleBiasClamp {bMapRange=true, outRange {max=0.6}}
    boundBone {boneName=dyn_bracelet_r_down}
    componentAppliedLinearAccClamp=<3000 3000 3000>
    componentLinearAccScale=<1 1 1>
    componentLinearVelScale=<1 1 1>
    physicsBodyDefinitions [{
      sphereCollisionRadius=5
      boundBone {boneName=dyn_bracelet_r_down}
      boxExtents=<5 5 10>
      constraintSetup {angularLimitsMax=<5 5 5>, angularLimitsMin=<-5 -5 -5>, angularTarget=<1 0 0>}
      localJointOffset=<-2 0 0>
     }]
    relativeSpaceBone {boneName=lowerarm_r}
   }
   n26 = "Transform (Modify) Bone - Bone: dyn_bracelet_l_main"!(ComponentPose=@n25, Alpha=Get HideLeftBraselet(), Scale="0,0,0") -> (Pose) {
    rotationMode=BMM_Additive
    rotationSpace=BCS_BoneSpace
    scaleMode=BMM_Additive
    scaleSpace=BCS_BoneSpace
    translationMode=BMM_Additive
    alphaBoolBlend {blendOption=Linear}
    alphaScaleBiasClamp {bInterpResult=true, bMapRange=true, outRange {max=0, min=1}}
    boneToModify {boneName=dyn_bracelet_l_main}
    scale=<0 0 0>
   }
   n27 = "Transform (Modify) Bone - Bone: dyn_bracelet_r_main"!(ComponentPose=@n26, Alpha=Get HideRightBraselet(), Scale="0,0,0") -> (Pose) {
    rotationMode=BMM_Additive
    rotationSpace=BCS_BoneSpace
    scaleMode=BMM_Additive
    scaleSpace=BCS_BoneSpace
    translationMode=BMM_Additive
    alphaBoolBlend {blendOption=Linear}
    alphaScaleBiasClamp {bInterpResult=true, bMapRange=true, outRange {max=0, min=1}}
    boneToModify {boneName=dyn_bracelet_r_main}
    scale=<0 0 0>
   }
   n28 = "Transform (Modify) Bone - Bone: dyn_skirt_l"!(
    ComponentPose=@n27
    Alpha:float=Get L Skirt correction then moving and crouching() -> (L Skirt correction then moving and crouching:float)
    Translation="0.531,-0.337,0.532"
    Rotation="0,0,0"
   ) -> (Pose) {
    rotationMode=BMM_Additive
    rotationSpace=BCS_BoneSpace
    translationMode=BMM_Additive
    alphaBoolBlend {blendOption=Linear}
    alphaScaleBiasClamp {bInterpResult=true}
    boneToModify {boneName=dyn_skirt_l}
    translation=<0.531 -0.337 0.532>
   }
   n29 = Get bIsTacticalSprinting()
   n30 = "Transform (Modify) Bone - Bone: dyn_skirt_main_l"!(ComponentPose=@n28, bAlphaBoolEnabled:bool=@n29, Translation="1.961,-2.620,-1.283", Rotation="1,-8.55,1.709") -> (Pose) {
    alphaInputType=Bool
    rotationMode=BMM_Additive
    rotationSpace=BCS_BoneSpace
    translationMode=BMM_Additive
    alphaBoolBlend {blendInTime=0.1, blendOption=Linear, blendOutTime=0.1}
    alphaScaleBiasClamp {bInterpResult=true}
    boneToModify {boneName=dyn_skirt_main_l}
    rotation=<rot 1 -8.55 1.709>
    translation=<1.961 -2.62 -1.283>
   }
   n31 = "Transform (Modify) Bone - Bone: dyn_skilt_bk_l"!(ComponentPose=@n30, bAlphaBoolEnabled:bool=@n29, Translation="-1.576,-1.119,0.200", Rotation="-0.047,-11.717,-0.397") -> (Pose) {
    alphaInputType=Bool
    rotationMode=BMM_Additive
    rotationSpace=BCS_BoneSpace
    translationMode=BMM_Additive
    alphaBoolBlend {blendInTime=0.1, blendOption=Linear, blendOutTime=0.1}
    alphaScaleBiasClamp {bInterpResult=true}
    boneToModify {boneName=dyn_skilt_bk_l}
    rotation=<rot -0.047 -11.717 -0.397>
    translation=<-1.5761 -1.1191 0.2>
   }
   n32 = "Transform (Modify) Bone - Bone: dyn_skirt_bk_r"!(ComponentPose=@n31, bAlphaBoolEnabled:bool=@n29, Translation="0,0,0", Rotation="-0.000,-8.362,0.000") -> (Pose) {
    alphaInputType=Bool
    rotationMode=BMM_Additive
    rotationSpace=BCS_BoneSpace
    translationMode=BMM_Additive
    alphaBoolBlend {blendInTime=0.1, blendOption=Linear, blendOutTime=0.1}
    alphaScaleBiasClamp {bInterpResult=true}
    boneToModify {boneName=dyn_skirt_bk_r}
    rotation=<rot 0 -8.3619 0>
   }
   n33 = "Constraint - dyn_sleeve_r_main (R)"(ComponentPose=@n32, Alpha=1, ConstraintWeights_0=1) {
    alphaBoolBlend {blendOption=Linear}
    boneToModify {boneName=dyn_sleeve_r_main}
    constraintSetup [{transformType=Rotation, targetBone {boneName=upperarm_twist_01_r}}]
    constraintWeights [1]
   }
   n34 = "Constraint - dyn_sleeve_l_main (R)"(ComponentPose=@n33, Alpha=0.5, ConstraintWeights_0=1) {
    alphaBoolBlend {blendOption=Linear}
    boneToModify {boneName=dyn_sleeve_l_main}
    constraintSetup [{transformType=Rotation, targetBone {boneName=upperarm_twist_01_l}}]
    constraintWeights [1]
   }
   n35 = Get bIsMoving2D()
   n36 = "Constraint - dyn_knife_chain_f (UUUUUUUU)"!(
    ComponentPose=@n34
    bAlphaBoolEnabled:bool=@n35
    ConstraintWeights_0=0.7479
    ConstraintWeights_1:float=0.0843
    ConstraintWeights_2:float=0.0001
    ConstraintWeights_3:float=0.1128
    ConstraintWeights_4:float=0.0033
    ConstraintWeights_5:float=0.0482
    ConstraintWeights_6:float=0.0031
    ConstraintWeights_7:float=0.0003
   ) -> (Pose) {
    alphaInputType=Bool
    alphaBoolBlend {blendInTime=0.2, blendOption=Linear, blendOutTime=0.2}
    boneToModify {boneName=dyn_knife_chain_f}
    constraintSetup [
     {transformType=LookAt, targetBone {boneName=pelvis}}
     {transformType=LookAt, targetBone {boneName=spine_01}}
     {transformType=LookAt, targetBone {boneName=thigh_l}}
     {transformType=LookAt, targetBone {boneName=thigh_twist_01_l}}
     {transformType=LookAt, targetBone {boneName=dyn_left_loin_1}}
     {transformType=LookAt, targetBone {boneName=dyn_left_loin_2}}
     {transformType=LookAt, targetBone {boneName=dyn_left_loin_3}}
     {transformType=LookAt, targetBone {boneName=dyn_skilt_bk_l}}
    ]
    constraintWeights [1,1,1,1,1,1,1,1]
   }
   n37 = "Constraint - dyn_knife_chain_b (UUUUUUUU)"!(
    ComponentPose=@n36
    bAlphaBoolEnabled:bool=@n35
    ConstraintWeights_0=0.8526
    ConstraintWeights_1:float=0.0499
    ConstraintWeights_2:float=0
    ConstraintWeights_3:float=0.0975
    ConstraintWeights_4:float=0
    ConstraintWeights_5:float=0
    ConstraintWeights_6:float=0
    ConstraintWeights_7:float=0
   ) -> (Pose) {
    alphaInputType=Bool
    alphaBoolBlend {blendInTime=0.2, blendOption=Linear, blendOutTime=0.2}
    boneToModify {boneName=dyn_knife_chain_b}
    constraintSetup [
     {transformType=LookAt, targetBone {boneName=pelvis}}
     {transformType=LookAt, targetBone {boneName=spine_01}}
     {transformType=LookAt, targetBone {boneName=spine_03}}
     {transformType=LookAt, targetBone {boneName=thigh_twist_01_l}}
     {transformType=LookAt, targetBone {boneName=thigh_twist_01_r}}
     {transformType=LookAt, targetBone {boneName=dyn_skirt_bk_r}}
     {transformType=LookAt, targetBone {boneName=dyn_skilt_bk_l}}
     {transformType=LookAt, targetBone {boneName=dyn_skirt_r}}
    ]
    constraintWeights [1,1,1,1,1,1,1,1]
   }
   n38 = "Transform (Modify) Bone - Bone: dyn_knife_chain_b"!(ComponentPose=@n37, bAlphaBoolEnabled:bool=Get bIsDBNO(), Translation="0.751,0.177,0.855", Rotation="-14.395,0.000,-0.000") -> (Pose) {
    alphaInputType=Bool
    rotationMode=BMM_Additive
    rotationSpace=BCS_BoneSpace
    translationMode=BMM_Additive
    alphaBoolBlend {blendInTime=0.1, blendOption=Linear, blendOutTime=0.1}
    alphaScaleBiasClamp {bInterpResult=true}
    boneToModify {boneName=dyn_knife_chain_b}
    rotation=<rot -14.395 0 0>
    translation=<0.751 0.177 0.855>
   }
   n39 = "Transform (Modify) Bone - Bone: dyn_knife_chain_f"!(ComponentPose=@n38, bAlphaBoolEnabled:bool=@n35, Translation="-2.186,0,0", Rotation="0,-60,0") -> (Pose) {
    alphaInputType=Bool
    rotationMode=BMM_Additive
    rotationSpace=BCS_BoneSpace
    translationMode=BMM_Additive
    alphaBoolBlend {blendInTime=0.1, blendOption=Linear, blendOutTime=0.1}
    alphaScaleBiasClamp {bInterpResult=true}
    boneToModify {boneName=dyn_knife_chain_f}
    rotation {yaw=-60}
    translation {x=-2.186}
   }
   n40 = "Transform (Modify) Bone - Bone: dyn_knife_chain_f"(ComponentPose=@n39, bAlphaBoolEnabled=Get JumpingOrFalling() -> (JumpingOrFalling:bool), Translation="0,2,0", Rotation="0,-60,0") {
    alphaInputType=Bool
    rotationMode=BMM_Additive
    rotationSpace=BCS_BoneSpace
    translationMode=BMM_Additive
    alphaBoolBlend {blendInTime=0.1, blendOption=Linear, blendOutTime=0.1}
    alphaScaleBiasClamp {bInterpResult=true}
    boneToModify {boneName=dyn_knife_chain_f}
    rotation {yaw=-60}
    translation {y=2}
   }
   n41 = Get bIsPlayingEmote()
   n42 = "Constraint - dyn_knife_chain_f (PPPPPPPP)"!(
    ComponentPose=@n40
    bAlphaBoolEnabled:bool=@n41
    ConstraintWeights_0=0.7479
    ConstraintWeights_1:float=0.0843
    ConstraintWeights_2:float=0.0001
    ConstraintWeights_3:float=0.1128
    ConstraintWeights_4:float=0.0033
    ConstraintWeights_5:float=0.0482
    ConstraintWeights_6:float=0.0031
    ConstraintWeights_7:float=0.0003
   ) -> (Pose) {
    alphaInputType=Bool
    alphaBoolBlend {blendInTime=0.2, blendOption=Linear, blendOutTime=0.2}
    boneToModify {boneName=dyn_knife_chain_f}
    constraintSetup [
     {targetBone {boneName=pelvis}}
     {targetBone {boneName=spine_01}}
     {targetBone {boneName=thigh_l}}
     {targetBone {boneName=thigh_twist_01_l}}
     {targetBone {boneName=dyn_left_loin_1}}
     {targetBone {boneName=dyn_left_loin_2}}
     {targetBone {boneName=dyn_left_loin_3}}
     {targetBone {boneName=dyn_skilt_bk_l}}
    ]
    constraintWeights [1,1,1,1,1,1,1,1]
   }
   n43 = "Transform (Modify) Bone - Bone: dyn_knife_chain_f"(ComponentPose=@n42, bAlphaBoolEnabled=@n41, Translation="0,0,0", Rotation="-66.575,-107.573,46.361") {
    alphaInputType=Bool
    rotationMode=BMM_Additive
    rotationSpace=BCS_BoneSpace
    translationMode=BMM_Additive
    translationSpace=BCS_ParentBoneSpace
    alphaBoolBlend {blendInTime=0.1, blendOption=Linear, blendOutTime=0.1}
    alphaScaleBiasClamp {bInterpResult=true}
    boneToModify {boneName=dyn_knife_chain_f}
    rotation=<rot -66.575 -107.573 46.361>
   }
   n44 = "Constraint - dyn_skirt_1_l (P)"(ComponentPose=@n43, Alpha=1, ConstraintWeights_0=1) {
    alphaBoolBlend {blendOption=Linear}
    boneToModify {boneName=dyn_skirt_1_l}
    constraintSetup [{targetBone {boneName=dyn_knife}}]
    constraintWeights [1]
   }
   n45 = "Transform (Modify) Bone - Bone: dyn_skirt_1_l"!(ComponentPose=@n44, Alpha=Get Left() -> (Left:double), Translation="-0.000,-2.398,1.265", Rotation="-0.000,-0.000,22.220") -> (Pose) {
    rotationMode=BMM_Additive
    rotationSpace=BCS_BoneSpace
    translationMode=BMM_Additive
    translationSpace=BCS_BoneSpace
    alphaBoolBlend {blendInTime=0.1, blendOption=Linear, blendOutTime=0.1}
    alphaScaleBiasClamp {bInterpResult=true}
    boneToModify {boneName=dyn_skirt_1_l}
    rotation=<rot 0 0 22.22>
    translation {y=-2.398, z=1.265}
   }
   n46 = Get Whip() -> (Whip:double)
   n47 = Look At - Bone: dyn_skirt_r(ComponentPose=@n45, Alpha=@n46) {
    alphaBoolBlend {blendOption=Linear}
    alphaScaleBiasClamp {bInterpResult=true, bMapRange=true, interpSpeedDecreasing=5, outRange {max=0.6}}
    boneToModify {boneName=dyn_skirt_r}
    lookAtLocation=<-0.0021 0.0107 17.1904>
    lookAtTarget {boneReference {boneName=calf_r}}
    lookAt_Axis {axis=<1 0>}
    lookUp_Axis {axis=<0 -1>}
   }
   n48 = "Transform (Modify) Bone - Bone: dyn_skirt_r"!(ComponentPose=@n47, Alpha=@n46, Translation="-0.590,-2.417,0.037", Rotation="-0.000,-16.251,0.005") -> (Pose) {
    rotationMode=BMM_Additive
    rotationSpace=BCS_BoneSpace
    scaleSpace=BCS_BoneSpace
    translationMode=BMM_Additive
    translationSpace=BCS_BoneSpace
    alphaBoolBlend {blendInTime=0.1, blendOption=Linear, blendOutTime=0.1}
    alphaScaleBiasClamp {bInterpResult=true, interpSpeedDecreasing=5}
    boneToModify {boneName=dyn_skirt_r}
    rotation=<rot 0 -16.2509 0.005>
    translation=<-0.59 -2.417 0.037>
   }
   n49 = "Transform (Modify) Bone - Bone: dyn_skirt_bk_r"!(ComponentPose=@n48, Alpha:float=@n46, Translation="0,0,0", Rotation="0.033,-22.631,0.001") -> (Pose) {
    rotationMode=BMM_Additive
    rotationSpace=BCS_BoneSpace
    scaleSpace=BCS_BoneSpace
    translationMode=BMM_Additive
    translationSpace=BCS_BoneSpace
    alphaBoolBlend {blendInTime=0.1, blendOption=Linear, blendOutTime=0.1}
    alphaScaleBiasClamp {bInterpResult=true, interpSpeedDecreasing=5}
    boneToModify {boneName=dyn_skirt_bk_r}
    rotation=<rot 0.033 -22.631 0.001>
   }
   n50 = "Transform (Modify) Bone - Bone: dyn_skirt_l"!(
    ComponentPose=@n49
    bAlphaBoolEnabled:bool=Get Is In Front End and not playing emote() -> (Is In Front End and not playing emote:bool)
    Translation="-0.037,0.338,0.027"
    Rotation="0.798,6.286,-7.291"
   ) -> (Pose) {
    alphaInputType=Bool
    rotationMode=BMM_Additive
    rotationSpace=BCS_BoneSpace
    scaleSpace=BCS_BoneSpace
    translationMode=BMM_Additive
    translationSpace=BCS_BoneSpace
    alphaBoolBlend {blendInTime=0.1, blendOption=Linear, blendOutTime=0.1}
    alphaScaleBiasClamp {bInterpResult=true, interpSpeedDecreasing=5}
    boneToModify {boneName=dyn_skirt_l}
    rotation=<rot 0.798 6.286 -7.291>
    translation=<-0.037 0.338 0.027>
   }
   Output Pose(Result=Component To Local(ComponentPose=@n50)) {layerGroup=None}
   n51 = Get bIsSkyDiving()
   n52 = Get rigidbody_alpha() -> (rigidbody_alpha:double)
   n53 = Get gravity_override() -> (gravity_override:Vector)
   n54 = RigidBody!(ComponentPose=Local To Component(LocalPose=Use cached pose 'SP_PreRigidBody'()), Alpha=@n52, OverrideWorldGravity:Vector=@n53) -> (Pose) {
    bClampLinearTranslationLimitToRefPose=true
    bOverrideWorldGravity=true
    overridePhysicsAsset=&/M_MED_PaperBlaze_Skirt_FrontEnd
    simulationSpace=BaseBoneSpace
    alphaBoolBlend {blendOption=Linear}
    alphaScaleBiasClamp {bInterpResult=true, interpSpeedDecreasing=5, interpSpeedIncreasing=5}
    baseBoneRef {boneName=pelvis}
    componentAppliedLinearAccClamp=<3000 3000 3000>
    componentLinearAccScale=<0.4 0.4 0.2>
    componentLinearVelScale=<0.6 0.6 0.6>
    overrideWorldGravity {z=-980}
    simSpaceSettings {maxAngularAcceleration=1000, maxAngularVelocity=3000, maxLinearAcceleration=1000, maxLinearVelocity=3000, worldAlpha=0.376}
   }
   n55 = RigidBody(ComponentPose=@n54, Alpha=@n52, OverrideWorldGravity=@n53) {
    bClampLinearTranslationLimitToRefPose=true
    bOverrideWorldGravity=true
    overridePhysicsAsset=&/M_MED_PaperBlaze_Phys1_FrontEnd
    simulationSpace=BaseBoneSpace
    alphaBoolBlend {blendOption=Linear}
    alphaScaleBiasClamp {bInterpResult=true, interpSpeedDecreasing=5, interpSpeedIncreasing=5}
    baseBoneRef {boneName=pelvis}
    componentAppliedLinearAccClamp=<3000 3000 3000>
    componentLinearAccScale=<0.35 0.35 0.2>
    componentLinearVelScale=<0.7 0.7 0.5>
    overrideWorldGravity {z=-980}
    simSpaceSettings {maxAngularAcceleration=500, maxAngularVelocity=300, maxLinearAcceleration=400, maxLinearVelocity=250, velocityScaleZ=0, worldAlpha=1}
   }
   n56 = Component To Local(ComponentPose=@n55)
   n57 = RigidBody(ComponentPose=Local To Component(LocalPose=Use cached pose 'SP_PreRigidBody'()), Alpha=@n52, OverrideWorldGravity=@n53) {
    bClampLinearTranslationLimitToRefPose=true
    bOverrideWorldGravity=true
    overridePhysicsAsset=&/M_MED_PaperBlaze_Skirt
    simulationSpace=BaseBoneSpace
    alphaBoolBlend {blendOption=Linear}
    alphaScaleBiasClamp {bInterpResult=true, interpSpeedDecreasing=5, interpSpeedIncreasing=5}
    baseBoneRef {boneName=pelvis}
    componentAppliedLinearAccClamp=<3000 3000 3000>
    componentLinearAccScale=<0.4 0.4 0.2>
    componentLinearVelScale=<0.6 0.6 0.6>
    overrideWorldGravity {z=-980}
    simSpaceSettings {maxAngularAcceleration=1000, maxAngularVelocity=3000, maxLinearAcceleration=1000, maxLinearVelocity=3000, worldAlpha=0.376}
   }
   n58 = RigidBody(ComponentPose=@n57, Alpha=@n52, OverrideWorldGravity=@n53) {
    bClampLinearTranslationLimitToRefPose=true
    bOverrideWorldGravity=true
    overridePhysicsAsset=&/M_MED_PaperBlaze_Physics1
    simulationSpace=BaseBoneSpace
    alphaBoolBlend {blendOption=Linear}
    alphaScaleBiasClamp {bInterpResult=true, interpSpeedDecreasing=5, interpSpeedIncreasing=5}
    baseBoneRef {boneName=pelvis}
    componentAppliedLinearAccClamp=<3000 3000 3000>
    componentLinearAccScale=<0.35 0.35 0.2>
    componentLinearVelScale=<0.7 0.7 0.5>
    overrideWorldGravity {z=-980}
    simSpaceSettings {maxAngularAcceleration=500, maxAngularVelocity=300, maxLinearAcceleration=400, maxLinearVelocity=250, velocityScaleZ=0, worldAlpha=1}
   }
   n59 = Component To Local(ComponentPose=@n58)
   n60 = Get bIsInFrontEnd()
   n61 = "Transform (Modify) Bone - Bone: dyn_sleeve_r_main"!(
    ComponentPose=Local To Component(LocalPose=Blend Poses by bool(BlendPose_0=@n56, BlendPose_1=@n59, bActiveValue=@n60) {blendType=Linear})
    bAlphaBoolEnabled:bool=@n51
    Translation="1,-0.075,0.167"
   ) -> (Pose) {
    alphaInputType=Bool
    rotationMode=BMM_Additive
    rotationSpace=BCS_BoneSpace
    translationMode=BMM_Additive
    alphaBoolBlend {blendInTime=0.1, blendOption=Linear, blendOutTime=0.1}
    alphaScaleBiasClamp {bInterpResult=true}
    boneToModify {boneName=dyn_sleeve_r_main}
    translation=<1 -0.075 0.167>
   }
   n62 = "Transform (Modify) Bone - Bone: dyn_sleeve_r_up"!(ComponentPose=@n61, bAlphaBoolEnabled:bool=@n51, Translation="2,0.353,-0.352") -> (Pose) {
    alphaInputType=Bool
    rotationMode=BMM_Additive
    rotationSpace=BCS_BoneSpace
    translationMode=BMM_Additive
    alphaBoolBlend {blendInTime=0.1, blendOption=Linear, blendOutTime=0.1}
    alphaScaleBiasClamp {bInterpResult=true}
    boneToModify {boneName=dyn_sleeve_r_up}
    translation=<2 0.353 -0.352>
   }
   n63 = Anim Dynamics - Bone: dyn_sleeve_l_down!(ComponentPose=@n62, Alpha=1) -> (Pose) {
    angularDampingOverride=0.9999
    angularSpringConstant=5000
    bAngularSpring=true
    bLinearSpring=true
    bOverrideAngularDamping=true
    bOverrideLinearDamping=true
    gravityScale=0.3
    linearDampingOverride=0.999
    linearSpringConstant=60
    simulationSpace=RootRelative
    alphaBoolBlend {blendOption=Linear}
    alphaScaleBiasClamp {scale=0.2}
    boundBone {boneName=dyn_sleeve_l_down}
    componentAppliedLinearAccClamp=<6000 6000 6000>
    componentLinearAccScale=<0.4 0.4 0.4>
    componentLinearVelScale=<0.4 0.4 0.4>
    gravityOverride {z=-400}
    physicsBodyDefinitions [{
      boundBone {boneName=dyn_sleeve_l_down}
      boxExtents=<20 10 10>
      constraintSetup {angularLimitsMax=<0 0 5>, angularLimitsMin=<0 0 -10>, angularTarget=<1 0 0>, linearAxesMin=<-3 0 0>}
      localJointOffset=<1 0 0>
     }]
   }
   n64 = Anim Dynamics - Bone: dyn_sleeve_l_front!(ComponentPose=@n63, Alpha=1) -> (Pose) {
    angularDampingOverride=0.9999
    angularSpringConstant=10
    bAngularSpring=true
    bLinearSpring=true
    bOverrideAngularDamping=true
    bOverrideLinearDamping=true
    bUsePlanarLimit=false
    bUseSphericalLimits=true
    gravityScale=0.3
    linearDampingOverride=0.999
    linearSpringConstant=60
    simulationSpace=RootRelative
    alphaBoolBlend {blendOption=Linear}
    boundBone {boneName=dyn_sleeve_l_front}
    componentAppliedLinearAccClamp=<6000 6000 6000>
    componentLinearAccScale=<0.4 0.4 0.4>
    componentLinearVelScale=<0.4 0.4 0.4>
    gravityOverride {z=-400}
    physicsBodyDefinitions [{
      collisionType=CustomSphere
      sphereCollisionRadius=12
      boundBone {boneName=dyn_sleeve_l_front}
      boxExtents=<20 10 10>
      constraintSetup {angularLimitsMax=<0 0 2>, angularLimitsMin=<0 0 -5>, angularTarget=<1 0 0>, linearAxesMin=<-3 0 0>}
      localJointOffset=<1 0 0>
     }]
    planarLimits [{drivingBone {boneName=lowerarm_l}, planeTransform {location=<0 2 0>, rotation=<rot 0 0 90>}}]
    sphericalLimits [{limitRadius=5, drivingBone {boneName=lowerarm_l}, sphereLocalOffset=<5 0 0>}]
   }
   n65 = Anim Dynamics - Bone: dyn_sleeve_l_back!(ComponentPose=@n64, Alpha=1) -> (Pose) {
    angularDampingOverride=0.9999
    angularSpringConstant=10
    bAngularSpring=true
    bLinearSpring=true
    bOverrideAngularDamping=true
    bOverrideLinearDamping=true
    gravityScale=0.3
    linearDampingOverride=0.999
    linearSpringConstant=60
    simulationSpace=RootRelative
    alphaBoolBlend {blendOption=Linear}
    alphaScaleBiasClamp {scale=0.6}
    boundBone {boneName=dyn_sleeve_l_back}
    componentAppliedLinearAccClamp=<6000 6000 6000>
    componentLinearAccScale=<0.4 0.4 0.4>
    componentLinearVelScale=<0.4 0.4 0.4>
    gravityOverride {z=-400}
    physicsBodyDefinitions [{
      boundBone {boneName=dyn_sleeve_l_back}
      boxExtents=<20 10 10>
      constraintSetup {angularLimitsMax=<0 0 5>, angularLimitsMin=<0 0 -5>, angularTarget=<1 0 0>, linearAxesMin=<-1 0 0>}
      localJointOffset=<1 0 0>
     }]
   }
   n66 = Anim Dynamics - Bone: dyn_sleeve_r_main!(ComponentPose=@n65, Alpha=1) -> (Pose) {
    angularDampingOverride=0.9999
    angularSpringConstant=10
    bAngularSpring=true
    bLinearSpring=true
    bOverrideAngularDamping=true
    bOverrideLinearDamping=true
    linearDampingOverride=0.998
    linearSpringConstant=100
    simulationSpace=RootRelative
    alphaBoolBlend {blendOption=Linear}
    alphaScaleBiasClamp {scale=0.2}
    boundBone {boneName=dyn_sleeve_r_main}
    componentAppliedLinearAccClamp=<200 200 200>
    componentLinearAccScale=<0.1 0.1 0.1>
    componentLinearVelScale=<0.1 0.1 0.1>
    physicsBodyDefinitions [{
      boundBone {boneName=dyn_sleeve_r_main}
      boxExtents=<20 10 10>
      constraintSetup {angularLimitsMax=<0 0 5>, angularLimitsMin=<0 0 -10>, angularTarget=<1 0 0>, linearAxesMax=<0 7 6>, linearAxesMin=<0 -4 -6>}
      localJointOffset=<-1 0 0>
     }]
    relativeSpaceBone {boneName=root}
   }
   n67 = Anim Dynamics - Bone: dyn_sleeve_r_up!(ComponentPose=@n66, Alpha=1) -> (Pose) {
    angularDampingOverride=0.9999
    angularSpringConstant=5000
    bAngularSpring=true
    bLinearSpring=true
    bOverrideAngularDamping=true
    bOverrideLinearDamping=true
    gravityScale=0.3
    linearDampingOverride=0.999
    linearSpringConstant=40
    simulationSpace=RootRelative
    alphaBoolBlend {blendOption=Linear}
    alphaScaleBiasClamp {scale=0.6}
    boundBone {boneName=dyn_sleeve_r_up}
    componentAppliedLinearAccClamp=<6000 6000 6000>
    componentLinearAccScale=<0.4 0.4 0.4>
    componentLinearVelScale=<0.4 0.4 0.4>
    gravityOverride {z=-400}
    physicsBodyDefinitions [{
      boundBone {boneName=dyn_sleeve_r_up}
      boxExtents=<20 10 10>
      constraintSetup {angularLimitsMax=<0 0 -10>, angularLimitsMin=<0 0 -20>, angularTarget=<1 0 0>, linearAxesMax=<3 0 0>}
      localJointOffset=<1 0 0>
     }]
   }
   n68 = Anim Dynamics - Bone: dyn_sleeve_r_down!(ComponentPose=@n67, Alpha=1) -> (Pose) {
    angularDampingOverride=0.9999
    angularSpringConstant=5000
    bAngularSpring=true
    bLinearSpring=true
    bOverrideAngularDamping=true
    bOverrideLinearDamping=true
    gravityScale=0.3
    linearDampingOverride=0.999
    linearSpringConstant=40
    simulationSpace=RootRelative
    alphaBoolBlend {blendOption=Linear}
    alphaScaleBiasClamp {scale=0.6}
    boundBone {boneName=dyn_sleeve_r_down}
    componentAppliedLinearAccClamp=<6000 6000 6000>
    componentLinearAccScale=<0.4 0.4 0.4>
    componentLinearVelScale=<0.4 0.4 0.4>
    gravityOverride {z=-400}
    physicsBodyDefinitions [{
      boundBone {boneName=dyn_sleeve_r_down}
      boxExtents=<20 10 10>
      constraintSetup {angularLimitsMax=<0 0 3>, angularLimitsMin=<0 0 -5>, angularTarget=<1 0 0>, linearAxesMax=<3 0 0>}
      localJointOffset=<1 0 0>
     }]
   }
   n69 = Anim Dynamics - Bone: dyn_sleeve_r_front!(ComponentPose=@n68, Alpha=1) -> (Pose) {
    angularDampingOverride=0.9999
    angularSpringConstant=10
    bAngularSpring=true
    bLinearSpring=true
    bOverrideAngularDamping=true
    bOverrideLinearDamping=true
    bUsePlanarLimit=false
    bUseSphericalLimits=true
    gravityScale=0.3
    linearDampingOverride=1
    linearSpringConstant=400
    simulationSpace=RootRelative
    alphaBoolBlend {blendOption=Linear}
    boundBone {boneName=dyn_sleeve_r_front}
    componentAppliedLinearAccClamp=<6000 6000 6000>
    componentLinearAccScale=<0.4 0.4 0.4>
    componentLinearVelScale=<0.4 0.4 0.4>
    gravityOverride {z=-400}
    physicsBodyDefinitions [{
      collisionType=CustomSphere
      sphereCollisionRadius=11
      boundBone {boneName=dyn_sleeve_r_front}
      boxExtents=<20 10 10>
      constraintSetup {angularLimitsMax=<0 0 -2>, angularLimitsMin=<0 0 -5>, angularTarget=<1 0 0>, linearAxesMax=<8 0 0>}
      localJointOffset=<1 0 0>
     }]
    sphericalLimits [
     {limitRadius=5, drivingBone {boneName=lowerarm_r}, sphereLocalOffset=<-7 0 0>}
     {limitRadius=5, drivingBone {boneName=lowerarm_r}, sphereLocalOffset=<-15 0 0>}
    ]
   }
   n70 = Anim Dynamics - Bone: dyn_sleeve_r_back!(ComponentPose=@n69, Alpha=1) -> (Pose) {
    angularDampingOverride=0.9999
    angularSpringConstant=5000
    bAngularSpring=true
    bLinearSpring=true
    bOverrideAngularDamping=true
    bOverrideLinearDamping=true
    gravityScale=0.3
    linearDampingOverride=0.999
    linearSpringConstant=40
    simulationSpace=RootRelative
    alphaBoolBlend {blendOption=Linear}
    alphaScaleBiasClamp {scale=0.6}
    boundBone {boneName=dyn_sleeve_r_back}
    componentAppliedLinearAccClamp=<6000 6000 6000>
    componentLinearAccScale=<0.4 0.4 0.4>
    componentLinearVelScale=<0.4 0.4 0.4>
    gravityOverride {z=-400}
    physicsBodyDefinitions [{
      boundBone {boneName=dyn_sleeve_r_back}
      boxExtents=<20 10 10>
      constraintSetup {angularLimitsMax=<0 10 3>, angularLimitsMin=<0 -10 -5>, angularTarget=<1 0 0>, linearAxesMax=<3 0 0>}
      localJointOffset=<1 0 0>
     }]
   }
   n71 = Anim Dynamics - Bone: dyn_necklace_c!(ComponentPose=@n70) -> (Pose) {
    alpha=0.6
    angularDampingOverride=0.99
    angularSpringConstant=500
    bAngularSpring=true
    bOverrideAngularDamping=true
    numSolverIterationsPostUpdate=4
    numSolverIterationsPreUpdate=16
    simulationSpace=RootRelative
    alphaBoolBlend {blendOption=Linear}
    boundBone {boneName=dyn_necklace_c}
    physicsBodyDefinitions [{
      sphereCollisionRadius=0
      boundBone {boneName=dyn_necklace_c}
      boxExtents=<20 5 10>
      constraintSetup {angularLimitsMax=<0 20 -5>, angularLimitsMin=<0 -20 -60>, angularTarget=<1 0 0>}
      localJointOffset=<1 0 0>
     }]
    planarLimits [{drivingBone {boneName=spine_05}, planeTransform {location=<7 8 0>, rotation=<rot 0 -30 90>}}]
   }
   n72 = Anim Dynamics - Bone: dyn_necklace_c_end!(ComponentPose=@n71) -> (Pose) {
    alpha=0.6
    angularDampingOverride=0.99
    angularSpringConstant=1500
    bAngularSpring=true
    bGravityOverrideInSimSpace=true
    bOverrideAngularDamping=true
    numSolverIterationsPostUpdate=4
    numSolverIterationsPreUpdate=16
    simulationSpace=BoneRelative
    alphaBoolBlend {blendOption=Linear}
    boundBone {boneName=dyn_necklace_c_end}
    physicsBodyDefinitions [{
      sphereCollisionRadius=0
      boundBone {boneName=dyn_necklace_c_end}
      boxExtents=<10 5 10>
      constraintSetup {angularLimitsMax=<0 15 0>, angularLimitsMin=<0 -15 -5>, angularTarget=<1 0 0>}
      localJointOffset=<1 0 0>
     }]
    planarLimits [{drivingBone {boneName=spine_04}, planeTransform {location=<0 16 0>, rotation=<rot 0 -20 95>}}]
    relativeSpaceBone {boneName=spine_05}
   }
   n73 = "Constraint - dyn_necklace_r (PPPPPP)"!(
    ComponentPose=@n72
    Alpha=1
    ConstraintWeights_0=0.0721
    ConstraintWeights_1:float=0.1254
    ConstraintWeights_2:float=0.5157
    ConstraintWeights_3:float=0.2252
    ConstraintWeights_4:float=0.0615
    ConstraintWeights_5:float=0
   ) -> (Pose) {
    alphaBoolBlend {blendOption=Linear}
    boneToModify {boneName=dyn_necklace_r}
    constraintSetup [
     {targetBone {boneName=spine_03}}
     {targetBone {boneName=spine_04}}
     {targetBone {boneName=spine_05}}
     {targetBone {boneName=clavicle_r}}
     {targetBone {boneName=deform_pec_r}}
     {targetBone {boneName=deform_upperarm_bk_r}}
    ]
    constraintWeights [1,1,1,1,1,1]
   }
   n74 = "Constraint - dyn_necklace_l (PPPPPPPP)"!(
    ComponentPose=@n73
    Alpha=1
    ConstraintWeights_0=0.0546
    ConstraintWeights_1:float=0.1079
    ConstraintWeights_2:float=0.533
    ConstraintWeights_3:float=0.0134
    ConstraintWeights_4:float=0.0235
    ConstraintWeights_5:float=0.2185
    ConstraintWeights_6:float=0.0091
    ConstraintWeights_7:float=0.0398
   ) -> (Pose) {
    alphaBoolBlend {blendOption=Linear}
    boneToModify {boneName=dyn_necklace_l}
    constraintSetup [
     {targetBone {boneName=spine_03}}
     {targetBone {boneName=spine_04}}
     {targetBone {boneName=spine_05}}
     {targetBone {boneName=neck_01}}
     {targetBone {boneName=neck_02}}
     {targetBone {boneName=clavicle_l}}
     {targetBone {boneName=upperarm_twist_01_l}}
     {targetBone {boneName=deform_pec_l}}
    ]
    constraintWeights [1,1,1,1,1,1,1,1]
   }
   n75 = "Transform (Modify) Bone - Bone: dyn_necklace_r"!(ComponentPose=@n74, bAlphaBoolEnabled:bool=True, Translation="0.474,0.962,0.208", Rotation="0,0,0") -> (Pose) {
    alphaInputType=Bool
    rotationMode=BMM_Additive
    rotationSpace=BCS_BoneSpace
    scaleSpace=BCS_WorldSpace
    translationMode=BMM_Additive
    translationSpace=BCS_ParentBoneSpace
    alphaBoolBlend {blendInTime=0.1, blendOption=Linear, blendOutTime=0.1}
    alphaScaleBiasClamp {bInterpResult=true}
    boneToModify {boneName=dyn_necklace_r}
    translation=<0.474 0.962 0.208>
   }
   n76 = "Transform (Modify) Bone - Bone: dyn_necklace_l"!(ComponentPose=@n75, bAlphaBoolEnabled:bool=True, Translation="0.562,0.883,0.056", Rotation="-0.000,6.017,0.000") -> (Pose) {
    alphaInputType=Bool
    rotationMode=BMM_Additive
    rotationSpace=BCS_BoneSpace
    scaleSpace=BCS_WorldSpace
    translationMode=BMM_Additive
    translationSpace=BCS_ParentBoneSpace
    alphaBoolBlend {blendInTime=0.1, blendOption=Linear, blendOutTime=0.1}
    alphaScaleBiasClamp {bInterpResult=true}
    boneToModify {boneName=dyn_necklace_l}
    rotation {yaw=6.017}
    translation=<0.562 0.883 0.056>
   }
   n77 = "Transform (Modify) Bone - Bone: dyn_necklace_c"!(ComponentPose=@n76, bAlphaBoolEnabled:bool=True, Translation="0.528,0.717,-0.171", Rotation="0.001,0.848,-0.003") -> (Pose) {
    alphaInputType=Bool
    rotationMode=BMM_Additive
    rotationSpace=BCS_BoneSpace
    scaleSpace=BCS_WorldSpace
    translationMode=BMM_Additive
    translationSpace=BCS_ParentBoneSpace
    alphaBoolBlend {blendInTime=0.1, blendOption=Linear, blendOutTime=0.1}
    alphaScaleBiasClamp {bInterpResult=true}
    boneToModify {boneName=dyn_necklace_c}
    rotation=<rot 0.001 0.8479 -0.003>
    translation=<0.528 0.717 -0.171>
   }
   n78 = "Constraint - dyn_necklace_l (U)"(ComponentPose=@n77, Alpha=1, ConstraintWeights_0=1) {
    alphaBoolBlend {blendOption=Linear}
    boneToModify {boneName=dyn_necklace_l}
    constraintSetup [{transformType=LookAt, targetBone {boneName=dyn_necklace_c_end}}]
    constraintWeights [1]
   }
   n79 = "Constraint - dyn_necklace_r (U)"(ComponentPose=@n78, Alpha=1, ConstraintWeights_0=1) {
    alphaBoolBlend {blendOption=Linear}
    boneToModify {boneName=dyn_necklace_r}
    constraintSetup [{transformType=LookAt, targetBone {boneName=dyn_necklace_c_end}}]
    constraintWeights [1]
   }
   n80 = Get Is In Front End and not playing emote()
   n81 = "Transform (Modify) Bone - Bone: dyn_necklace_c"!(ComponentPose=@n79, bAlphaBoolEnabled=@n80, Translation="0,0.8,0") -> (Pose) {
    alphaInputType=Bool
    rotationMode=BMM_Additive
    rotationSpace=BCS_BoneSpace
    scaleSpace=BCS_BoneSpace
    translationMode=BMM_Additive
    translationSpace=BCS_BoneSpace
    alphaBoolBlend {blendInTime=0.1, blendOption=Linear, blendOutTime=0.1}
    alphaScaleBiasClamp {bInterpResult=true, interpSpeedDecreasing=5}
    boneToModify {boneName=dyn_necklace_c}
    translation {y=0.8}
   }
   n82 = "Transform (Modify) Bone - Bone: dyn_necklace_l"!(ComponentPose=@n81, bAlphaBoolEnabled=@n80, Translation="0,1,0") -> (Pose) {
    alphaInputType=Bool
    rotationMode=BMM_Additive
    rotationSpace=BCS_BoneSpace
    scaleSpace=BCS_BoneSpace
    translationMode=BMM_Additive
    translationSpace=BCS_BoneSpace
    alphaBoolBlend {blendInTime=0.1, blendOption=Linear, blendOutTime=0.1}
    alphaScaleBiasClamp {bInterpResult=true, interpSpeedDecreasing=5}
    boneToModify {boneName=dyn_necklace_l}
    translation {y=1}
   }
   n83 = "Transform (Modify) Bone - Bone: dyn_necklace_r"!(ComponentPose=@n82, bAlphaBoolEnabled=@n80, Translation="0,-1,0") -> (Pose) {
    alphaInputType=Bool
    rotationMode=BMM_Additive
    rotationSpace=BCS_BoneSpace
    scaleSpace=BCS_BoneSpace
    translationMode=BMM_Additive
    translationSpace=BCS_BoneSpace
    alphaBoolBlend {blendInTime=0.1, blendOption=Linear, blendOutTime=0.1}
    alphaScaleBiasClamp {bInterpResult=true, interpSpeedDecreasing=5}
    boneToModify {boneName=dyn_necklace_r}
    translation {y=-1}
   }
   n84 = Get IsSliding_alpha() -> (IsSliding_alpha:double)
   n85 = Look At - Bone: dyn_skirt_bk_r(ComponentPose=@n83, Alpha=@n84) {
    alphaBoolBlend {blendOption=Linear}
    alphaScaleBiasClamp {bInterpResult=true, scale=0.5}
    boneToModify {boneName=dyn_skirt_bk_r}
    lookAtLocation=<-8.107 -11.7941 5.4824>
    lookAtTarget {boneReference {boneName=pelvis}}
    lookAt_Axis {axis=<1 0>}
    lookUp_Axis {axis=<0 1>}
   }
   n86 = Look At - Bone: dyn_skirt_bk_r(ComponentPose=@n85, Alpha=@n84) {
    alphaBoolBlend {blendOption=Linear}
    alphaScaleBiasClamp {bInterpResult=true, scale=0.5}
    boneToModify {boneName=dyn_skirt_bk_r}
    lookAtLocation=<0.062 -13.1777 1.4629>
    lookAtTarget {boneReference {boneName=calf_r}}
    lookAt_Axis {axis=<1 0>}
    lookUp_Axis {axis=<0 1>}
   }
   n87 = Look At - Bone: dyn_skilt_bk_l(ComponentPose=@n86, Alpha=@n84) {
    alphaBoolBlend {blendOption=Linear}
    alphaScaleBiasClamp {bInterpResult=true}
    boneToModify {boneName=dyn_skilt_bk_l}
    lookAtLocation=<-10.0029 -13.766 -9.1251>
    lookAtTarget {boneReference {boneName=pelvis}}
    lookAt_Axis {axis=<1 0>}
    lookUp_Axis {axis=<0 1>}
   }
   n88 = "Transform (Modify) Bone - Bone: dyn_skirt_l"!(ComponentPose=@n87, Alpha:float=@n84, Translation="1.196,0.160,1.399", Rotation="16.418,-1.521,20.701", Scale:Vector="0.884,0.884,0.884") -> (Pose) {
    rotationMode=BMM_Additive
    rotationSpace=BCS_BoneSpace
    scaleMode=BMM_Additive
    scaleSpace=BCS_BoneSpace
    translationMode=BMM_Additive
    translationSpace=BCS_BoneSpace
    alphaBoolBlend {blendOption=Linear}
    alphaScaleBiasClamp {bInterpResult=true, interpSpeedIncreasing=5}
    boneToModify {boneName=dyn_skirt_l}
    rotation=<rot 16.418 -1.521 20.701>
    scale=<0.884 0.884 0.884>
    translation=<1.196 0.16 1.399>
   }
   n89 = Get IsBaseballSliding() -> (IsBaseballSliding:double)
   n90 = Look At - Bone: dyn_skirt_l(ComponentPose=@n88, Alpha=@n89) {
    alphaBoolBlend {blendOption=Linear}
    boneToModify {boneName=dyn_skirt_l}
    lookAtLocation=<-40.2071 6.5435 -16.5927>
    lookAtTarget {boneReference {boneName=thigh_l}}
    lookAt_Axis {axis=<1 0>}
    lookUp_Axis {axis=<0 1>}
   }
   n91 = Anim Dynamics - Bone: dyn_skirt_2_l!(ComponentPose=@n90, Alpha=@n89) -> (Pose) {
    angularDampingOverride=0.9999
    angularSpringConstant=5000
    bAngularSpring=true
    bLinearSpring=true
    bOverrideAngularDamping=true
    bOverrideLinearDamping=true
    gravityScale=0.3
    linearDampingOverride=0.999
    linearSpringConstant=40
    simulationSpace=RootRelative
    alphaBoolBlend {blendOption=Linear}
    alphaScaleBiasClamp {scale=0.6}
    boundBone {boneName=dyn_skirt_2_l}
    componentAppliedLinearAccClamp=<6000 6000 6000>
    componentLinearAccScale=<0.4 0.4 0.4>
    componentLinearVelScale=<0.4 0.4 0.4>
    gravityOverride {z=-400}
    physicsBodyDefinitions [{
      boundBone {boneName=dyn_skirt_2_l}
      boxExtents=<20 10 10>
      constraintSetup {angularLimitsMax=<10 10 10>, angularLimitsMin=<-10 -10 -10>, angularTarget=<1 0 0>}
      localJointOffset=<1 0 0>
     }]
   }
   Save cached pose 'Pre Braceletes'(Pose=Component To Local(ComponentPose=@n91)) {cacheName=Pre Braceletes}
   n92 = Get IsCrouching_alpha() -> (IsCrouching_alpha:double)
   n93 = "Constraint - dyn_knife_chain_f (PPPPP)"!(
    ComponentPose=Local To Component(LocalPose=Copy Pose From Mesh() {bUseAttachedParent=true})
    Alpha=@n92
    ConstraintWeights_0=0.5339
    ConstraintWeights_1:float=0.1565
    ConstraintWeights_2:float=0.0504
    ConstraintWeights_3:float=0.0748
    ConstraintWeights_4:float=0.1845
   ) -> (Pose) {
    alphaBoolBlend {blendOption=Linear}
    boneToModify {boneName=dyn_knife_chain_f}
    constraintSetup [
     {targetBone {boneName=pelvis}}
     {targetBone {boneName=spine_01}}
     {targetBone {boneName=thigh_twist_01_l}}
     {targetBone {boneName=dyn_left_loin_2}}
     {targetBone {boneName=dyn_left_loin_3}}
    ]
    constraintWeights [1,1,1,1,1]
   }
   n94 = "Constraint - dyn_knife_chain_b (PPPPP)"!(
    ComponentPose=@n93
    Alpha=@n92
    ConstraintWeights_0=0.7983
    ConstraintWeights_1:float=0.0544
    ConstraintWeights_2:float=0.1186
    ConstraintWeights_3:float=0.004
    ConstraintWeights_4:float=0.0248
   ) -> (Pose) {
    alphaBoolBlend {blendOption=Linear}
    boneToModify {boneName=dyn_knife_chain_b}
    constraintSetup [
     {targetBone {boneName=pelvis}}
     {targetBone {boneName=spine_01}}
     {targetBone {boneName=thigh_twist_01_l}}
     {targetBone {boneName=dyn_left_loin_1}}
     {targetBone {boneName=dyn_left_loin_2}}
    ]
    constraintWeights [1,1,1,1,1]
   }
   n95 = "Constraint - dyn_right_loin_correct (PPPPPPPP)"!(
    ComponentPose=@n94
    Alpha=1
    ConstraintWeights_0=0.235
    ConstraintWeights_1:float=0.0219
    ConstraintWeights_2:float=0
    ConstraintWeights_3:float=0
    ConstraintWeights_4:float=0
    ConstraintWeights_5:float=0.0004
    ConstraintWeights_6:float=0.0001
    ConstraintWeights_7:float=0.7426
   ) -> (Pose) {
    alphaBoolBlend {blendOption=Linear}
    alphaScaleBiasClamp {bMapRange=true, outRange {max=0.5, min=1}}
    boneToModify {boneName=dyn_right_loin_correct}
    constraintSetup [
     {targetBone {boneName=pelvis}}
     {targetBone {boneName=spine_01}}
     {targetBone {boneName=spine_02}}
     {targetBone {boneName=spine_03}}
     {targetBone {boneName=thigh_l}}
     {targetBone {boneName=thigh_twist_01_l}}
     {targetBone {boneName=thigh_r}}
     {targetBone {boneName=thigh_twist_01_r}}
    ]
    constraintWeights [1,1,1,1,1,1,1,1]
   }
   n96 = "Constraint - dyn_left_loin_4 (PPPPPPPP)"!(
    ComponentPose=@n95
    Alpha=1
    ConstraintWeights_0=0.5855
    ConstraintWeights_1:float=0.3236
    ConstraintWeights_2:float=0.0404
    ConstraintWeights_3:float=0.015
    ConstraintWeights_4:float=0.0064
    ConstraintWeights_5:float=0.0009
    ConstraintWeights_6:float=0.0166
    ConstraintWeights_7:float=0.0116
   ) -> (Pose) {
    alphaBoolBlend {blendOption=Linear}
    boneToModify {boneName=dyn_left_loin_4}
    constraintSetup [
     {targetBone {boneName=pelvis}}
     {targetBone {boneName=spine_01}}
     {targetBone {boneName=spine_02}}
     {targetBone {boneName=spine_03}}
     {targetBone {boneName=thigh_twist_01_l}}
     {targetBone {boneName=dyn_left_loin_3}}
     {targetBone {boneName=deform_groin_l}}
     {targetBone {boneName=deform_glute_l}}
    ]
    constraintWeights [1,1,1,1,1,1,1,1]
   }
   n97 = Get SlidingAndBaseballSliding() -> (SlidingAndBaseballSliding:double)
   n98 = "Constraint - dyn_skirt_r (PPP)"!(ComponentPose=@n96, Alpha=@n97, ConstraintWeights_0=0.6019, ConstraintWeights_1:float=0.0562, ConstraintWeights_2:float=0.342) -> (Pose) {
    alphaBoolBlend {blendOption=Linear}
    boneToModify {boneName=dyn_skirt_r}
    constraintSetup [{targetBone {boneName=pelvis}}, {targetBone {boneName=spine_01}}, {targetBone {boneName=thigh_twist_01_r}}]
    constraintWeights [1,1,1]
   }
   n99 = "Constraint - dyn_skirt_1_r (PPP)"!(ComponentPose=@n98, Alpha=@n97, ConstraintWeights_0=0.0931, ConstraintWeights_1:float=0.0489, ConstraintWeights_2:float=0.858) -> (Pose) {
    alphaBoolBlend {blendOption=Linear}
    boneToModify {boneName=dyn_skirt_1_r}
    constraintSetup [{targetBone {boneName=pelvis}}, {targetBone {boneName=thigh_r}}, {targetBone {boneName=thigh_twist_01_r}}]
    constraintWeights [1,1,1]
   }
   n100 = "Constraint - dyn_skirt_2_r (PPP)"!(ComponentPose=@n99, Alpha=@n97, ConstraintWeights_0=0.0204, ConstraintWeights_1:float=0.1142, ConstraintWeights_2:float=0.8654) -> (Pose) {
    alphaBoolBlend {blendOption=Linear}
    boneToModify {boneName=dyn_skirt_2_r}
    constraintSetup [{targetBone {boneName=pelvis}}, {targetBone {boneName=thigh_r}}, {targetBone {boneName=thigh_twist_01_r}}]
    constraintWeights [1,1,1]
   }
   n101 = "Constraint - dyn_bracelet_l_up (PPPPPPPP)"!(
    ComponentPose=@n100
    Alpha=1
    ConstraintWeights_0=0.4451
    ConstraintWeights_1:float=0.003
    ConstraintWeights_2:float=0.0631
    ConstraintWeights_3:float=0.0555
    ConstraintWeights_4:float=0.0968
    ConstraintWeights_5:float=0.1226
    ConstraintWeights_6:float=0.0126
    ConstraintWeights_7:float=0.2013
   ) -> (Pose) {
    alphaBoolBlend {blendOption=Linear}
    alphaScaleBiasClamp {scale=0.8}
    boneToModify {boneName=dyn_bracelet_l_up}
    constraintSetup [
     {targetBone {boneName=hand_l}}
     {targetBone {boneName=thumb_01_l}}
     {targetBone {boneName=index_metacarpal_l}}
     {targetBone {boneName=middle_metacarpal_l}}
     {targetBone {boneName=ring_metacarpal_l}}
     {targetBone {boneName=pinky_metacarpal_l}}
     {targetBone {boneName=lowerarm_twist_02_l}}
     {targetBone {boneName=lowerarm_twist_01_l}}
    ]
    constraintWeights [1,1,1,1,1,1,1,1]
   }
   n102 = "Constraint - dyn_bracelet_l_down (PPPPPPPP)"!(
    ComponentPose=@n101
    Alpha=1
    ConstraintWeights_0=0.14
    ConstraintWeights_1:float=0.1932
    ConstraintWeights_2:float=0.0017
    ConstraintWeights_3:float=0.0294
    ConstraintWeights_4:float=0.1345
    ConstraintWeights_5:float=0.3602
    ConstraintWeights_6:float=0.0118
    ConstraintWeights_7:float=0.1292
   ) -> (Pose) {
    alphaBoolBlend {blendOption=Linear}
    alphaScaleBiasClamp {scale=0.8}
    boneToModify {boneName=dyn_bracelet_l_down}
    constraintSetup [
     {targetBone {boneName=hand_l}}
     {targetBone {boneName=thumb_01_l}}
     {targetBone {boneName=thumb_02_l}}
     {targetBone {boneName=middle_metacarpal_l}}
     {targetBone {boneName=ring_metacarpal_l}}
     {targetBone {boneName=pinky_metacarpal_l}}
     {targetBone {boneName=lowerarm_twist_02_l}}
     {targetBone {boneName=lowerarm_twist_01_l}}
    ]
    constraintWeights [1,1,1,1,1,1,1,1]
   }
   n103 = Get bIsTacticalSprinting()
   n104 = "Constraint - dyn_knife_chain_f (PPPPPPPP)"(
    ComponentPose=@n102
    bAlphaBoolEnabled=@n103
    ConstraintWeights_0=0.7479
    ConstraintWeights_1=0.0843
    ConstraintWeights_2=0.0001
    ConstraintWeights_3=0.1128
    ConstraintWeights_4=0.0033
    ConstraintWeights_5=0.0482
    ConstraintWeights_6=0.0031
    ConstraintWeights_7=0.0003
   ) {
    alphaInputType=Bool
    alphaBoolBlend {blendInTime=0.1, blendOption=Linear, blendOutTime=0.1}
    boneToModify {boneName=dyn_knife_chain_f}
    constraintSetup [
     {targetBone {boneName=pelvis}}
     {targetBone {boneName=spine_01}}
     {targetBone {boneName=thigh_l}}
     {targetBone {boneName=thigh_twist_01_l}}
     {targetBone {boneName=dyn_left_loin_1}}
     {targetBone {boneName=dyn_left_loin_2}}
     {targetBone {boneName=dyn_left_loin_3}}
     {targetBone {boneName=dyn_skilt_bk_l}}
    ]
    constraintWeights [1,1,1,1,1,1,1,1]
   }
   n105 = "Constraint - dyn_knife_chain_b (PPPPPPPP)"!(
    ComponentPose=@n104
    bAlphaBoolEnabled:bool=@n103
    ConstraintWeights_0=0.8526
    ConstraintWeights_1:float=0.0499
    ConstraintWeights_2:float=0
    ConstraintWeights_3:float=0.0975
    ConstraintWeights_4:float=0
    ConstraintWeights_5:float=0
    ConstraintWeights_6:float=0
    ConstraintWeights_7:float=0
   ) -> (Pose) {
    alphaInputType=Bool
    alphaBoolBlend {blendInTime=0.1, blendOption=Linear, blendOutTime=0.1}
    boneToModify {boneName=dyn_knife_chain_b}
    constraintSetup [
     {targetBone {boneName=pelvis}}
     {targetBone {boneName=spine_01}}
     {targetBone {boneName=spine_03}}
     {targetBone {boneName=thigh_twist_01_l}}
     {targetBone {boneName=thigh_twist_01_r}}
     {targetBone {boneName=dyn_skirt_bk_r}}
     {targetBone {boneName=dyn_skilt_bk_l}}
     {targetBone {boneName=dyn_skirt_r}}
    ]
    constraintWeights [1,1,1,1,1,1,1,1]
   }
   n106 = Get IsBaseballSliding()
   n107 = Look At - Bone: dyn_skirt_bk_r(ComponentPose=@n105, Alpha=@n106) {
    alphaBoolBlend {blendOption=Linear}
    boneToModify {boneName=dyn_skirt_bk_r}
    lookAtLocation=<-28.1296 15.7547 5.3908>
    lookAtTarget {boneReference {boneName=pelvis}}
    lookAt_Axis {axis=<1 0>}
    lookUp_Axis {axis=<0 1>}
   }
   n108 = Look At - Bone: dyn_skilt_bk_l(ComponentPose=@n107, Alpha=@n106) {
    alphaBoolBlend {blendOption=Linear}
    boneToModify {boneName=dyn_skilt_bk_l}
    lookAtLocation=<-21.6336 -2.9358 -4.2201>
    lookAtTarget {boneReference {boneName=pelvis}}
    lookAt_Axis {axis=<1 0>}
    lookUp_Axis {axis=<0 1>}
   }
   n109 = "Constraint - dyn_shoulderpad_l (R)"(ComponentPose=@n108, Alpha=1, ConstraintWeights_0=1) {
    alphaBoolBlend {blendOption=Linear}
    boneToModify {boneName=dyn_shoulderpad_l}
    constraintSetup [{transformType=Rotation, targetBone {boneName=upperarm_twist_01_l}}]
    constraintWeights [1]
   }
   n110 = Look At - Bone: dyn_shoulderpad_l(ComponentPose=@n109, Alpha=1) {
    bUseLookUpAxis=true
    alphaBoolBlend {blendOption=Linear}
    boneToModify {boneName=dyn_shoulderpad_l}
    lookAtLocation {x=0}
    lookAtTarget {boneReference {boneName=lowerarm_l}}
    lookAt_Axis {axis=<1 0>}
    lookUp_Axis {axis=<0 -1>}
   }
   n111 = "Constraint - dyn_shoulderpad_l (R)"(ComponentPose=@n110, Alpha=1, ConstraintWeights_0=1) {
    alphaBoolBlend {blendOption=Linear}
    boneToModify {boneName=dyn_shoulderpad_l}
    constraintSetup [{transformType=Rotation, targetBone {boneName=dyn_sleeve_l_main}}]
    constraintWeights [1]
   }
   n112 = "Transform (Modify) Bone - Bone: dyn_shoulderpad_l"!(ComponentPose=@n111, Alpha:float=1, Translation="0.064,-0.263,-0.520", Rotation="2.154,0.811,-3.680") -> (Pose) {
    rotationMode=BMM_Additive
    rotationSpace=BCS_BoneSpace
    translationMode=BMM_Replace
    translationSpace=BCS_BoneSpace
    alphaBoolBlend {blendOption=Linear}
    boneToModify {boneName=dyn_shoulderpad_l}
    rotation=<rot 2.1541 0.811 -3.68>
    translation=<0.064 -0.263 -0.52>
   }
   n113 = Control Rig!(
    Source=Component To Local(ComponentPose=@n112)
    Alpha
    IsCrouching:double=Get IsCrouching_alpha()
    IsSliding:double=Get IsSliding_alpha()
    MelleePistolOtherAssaultRifle:int=Get MelleePistolOtherAssaultRifle() -> (MelleePistolOtherAssaultRifle:int=0)
    IsCrouchingBool:bool=Get bIsCrouching()
    SkyDiving:bool=Get bIsSkyDiving()
    IsBaseballSliding:double=Get IsBaseballSliding()
    Left:double=Get Left()
    Right:double=Get Right() -> (Right:double)
    Targeting:bool=Get bIsTargeting()
    static_crouching_rotation:bool=Get static_crouching_rotation() -> (static_crouching_rotation:bool)
    TacticalJump:bool=Get TacticalJump() -> (TacticalJump:bool)
    TacticalMovement:bool=Get bIsTacticalSprinting()
    IsInFrontEnd:bool=Get bIsInFrontEnd()
    IsJumping:bool=Get bIsJumping()
    IsOnGround:bool=Get bIsOnGround() -> (bIsOnGround:bool)
    IsFalling:bool=Get bIsFalling()
   ) -> (Pose) {
    alphaBoolBlend {blendOption=Linear}
    controlRigAssetReference {blueprintRigClass=&/M_MED_PaperBlaze_CtrlRig.M_MED_PaperBlaze_CtrlRig_C}
   }
   n114 = "Transform (Modify) Bone - Bone: dyn_skirt_main_l"!(
    ComponentPose=Local To Component(LocalPose=@n113)
    Alpha:float=Get IsCrouching_alpha()
    Translation="3.479,1.436,2.628"
    Rotation="-52.561,2.179,-31.738"
   ) -> (Pose) {
    rotationMode=BMM_Additive
    rotationSpace=BCS_BoneSpace
    translationMode=BMM_Replace
    translationSpace=BCS_BoneSpace
    alphaBoolBlend {blendOption=Linear}
    boneToModify {boneName=dyn_skirt_main_l}
    rotation=<rot -52.561 2.179 -31.738>
    translation=<3.4792 1.436 2.6278>
   }
   n115 = "Transform (Modify) Bone - Bone: dyn_skirt_main_l"(
    ComponentPose=@n114
    Alpha=Get MovingNotCrouching() -> (MovingNotCrouching:double)
    Translation="2.543,0.001,2.599"
    Rotation="-36.488,0.036,0.035"
   ) {
    rotationMode=BMM_Additive
    rotationSpace=BCS_BoneSpace
    scaleSpace=BCS_BoneSpace
    translationMode=BMM_Additive
    translationSpace=BCS_BoneSpace
    alphaBoolBlend {blendOption=Linear}
    alphaScaleBiasClamp {bInterpResult=true, interpSpeedDecreasing=5, interpSpeedIncreasing=5}
    boneToModify {boneName=dyn_skirt_main_l}
    rotation=<rot -36.488 0.036 0.035>
    translation=<2.543 0.001 2.599>
   }
   Save cached pose 'SP_PreRigidBody'(Pose=Component To Local(ComponentPose=@n115)) {cacheName=SP_PreRigidBody}
  ]}
 EventGraph {nodes [
   Event BlueprintUpdateAnimation()
   > n1 = Sequence!(execute) -> (then_0, then_1, then_2:Exec, then_3:Exec, then_4:Exec, then_5:Exec, then_6:Exec)
   n2 = OR Boolean(A=Get bIsCrouching(), B=Get bIsDBNO())
   > Set IsCrouching_alpha(IsCrouching_alpha:double=FInterpTo(
     Current=Get IsCrouching_alpha()
     Target="To Float (Boolean)"(InBool=@n2)
     DeltaTime=GetWorldDeltaSeconds!() -> (ReturnValue:double)
     InterpSpeed=25
    )) -> (Output_Get:double)
   n3 = FInterpTo(Current=Get IsSliding_alpha(), Target="To Float (Boolean)"(InBool=Get bIsKneeSliding()), DeltaTime=GetWorldDeltaSeconds(), InterpSpeed=50)
   > n4 = Set IsSliding_alpha(IsSliding_alpha:double=@n3) -> (Output_Get:double)
   n5 = Get bIsBaseballSliding()
   > n6 = Set IsBaseballSliding(IsBaseballSliding:double=FInterpTo(Current=Get IsBaseballSliding(), Target="To Float (Boolean)"(InBool=@n5), DeltaTime=GetWorldDeltaSeconds(), InterpSpeed=50)) -> (Output_Get:double)
   n7 = float * float(A=@n6) {is=title K2Node_PromotableOperator}
   n8 = "Equal (Float)"(A=@n6, B=1)
   > Set rigidbody_alpha(rigidbody_alpha:double=SelectFloat(A=float - float(A=1, B=@n3), B=SelectFloat(A=@n7, B=1, bPickA=@n8), bPickA="Equal (Float)"(A=@n4, B=1))) -> (Output_Get:double)
   n9 = Get bIsDBNO()
   n10 = Get bIsCrouching()
   n11 = OR Boolean(A=AND Boolean!(A=Get bIsMoving2D(), B=NOT Boolean(A=@n10), C:bool=NOT Boolean(A=@n9)) -> (ReturnValue), B=Get bIsPlayingEmote())
   > Set MovingNotCrouching(MovingNotCrouching:double=FInterpTo(Current=Get MovingNotCrouching(), Target="To Float (Boolean)"(InBool=@n11), DeltaTime=GetWorldDeltaSeconds(), InterpSpeed=15)) -> (Output_Get:double)
   > n12 = Switch on EFortWeaponCoreAnimation!(execute, Selection:EFortWeaponCoreAnimation Enum=Get WeaponCoreAnim!() -> (WeaponCoreAnim)) -> (
    Melee:Exec
    Pistol:Exec
    Shotgun:Exec
    PaperBlueprint:Exec
    Rifle:Exec
    MeleeOneHand:Exec
    MachinePistol:Exec
    RocketLauncher:Exec
    GrenadeLauncher:Exec
    GoingCommando:Exec
    AssaultRifle:Exec
    TacticalShotgun:Exec
    SniperRifle:Exec
    TrapPlacement:Exec
    ShoulderLauncher:Exec
    AbilityDecoTool:Exec
    Crossbow:Exec
    C4:Exec
    RemoteControl:Exec
    DualWield:Exec
    AR_BullPup:Exec
    AR_ForwardGrip:Exec
    MedPackPaddles:Exec
    SMG_P90:Exec
    AR_DrumGun:Exec
    Consumable_Small:Exec
    Consumable_Large:Exec
    Balloon:Exec
    MountedTurret:Exec
    CreativeTool:Exec
    ExplosiveBow:Exec
    AshtonIndigo:Exec
    AshtonChicago:Exec
    MeleeDualWield:Exec
    Unarmed:Exec
   )
   Set MelleePistolOtherAssaultRifle(
    execute=[
     @n12.AR_BullPup
     @n12.AR_DrumGun
     @n12.AR_ForwardGrip
     @n12.AbilityDecoTool
     @n12.AshtonChicago
     @n12.AshtonIndigo
     @n12.Consumable_Large
     @n12.Consumable_Small
     @n12.Crossbow
     @n12.DualWield
     @n12.ExplosiveBow
     @n12.GoingCommando
     @n12.GrenadeLauncher
     @n12.MachinePistol
     @n12.MedPackPaddles
     @n12.MountedTurret
     @n12.PaperBlueprint
     @n12.RemoteControl
     @n12.Rifle
     @n12.RocketLauncher
     @n12.SMG_P90
     @n12.Shotgun
     @n12.ShoulderLauncher
     @n12.SniperRifle
     @n12.TacticalShotgun
    ]
    MelleePistolOtherAssaultRifle:int=2
   ) -> (Output_Get:int=0)
   Set MelleePistolOtherAssaultRifle(execute=@n12.AssaultRifle, MelleePistolOtherAssaultRifle=3) -> (Output_Get=0)
   Set MelleePistolOtherAssaultRifle(
    execute=[@n12.Balloon, @n12.C4, @n12.CreativeTool, @n12.Melee, @n12.MeleeDualWield, @n12.MeleeOneHand, @n12.TrapPlacement, @n12.Unarmed]
    MelleePistolOtherAssaultRifle=0
   ) -> (Output_Get=0)
   Set MelleePistolOtherAssaultRifle(execute=@n12.Pistol, MelleePistolOtherAssaultRifle=1) -> (Output_Get=0)
   n13 = TryGetPawnOwner()
   Is Valid(exec=@n1.then_5, InputObject=@n13)
   > n14 = Cast To FortPlayerPawnAthena(Object=@n13)
   n15 = GetSkeletalMeshForPartType(self=@n14)
   > n16 = IsBoneHiddenByName!(execute, self:Skinned Mesh Component Object Reference=@n15, BoneName:Name=hand_r) -> (then, ReturnValue:bool)
   > n17 = IsBoneHiddenByName(self=@n15, BoneName=hand_l)
   n18 = "To Float (Boolean)"(InBool=NOT Boolean(A=@n16))
   > Set HideRightBraselet(HideRightBraselet:double=@n18) -> (Output_Get:double)
   n19 = "To Float (Boolean)"(InBool=NOT Boolean(A=@n17))
   > Set HideLeftBraselet(HideLeftBraselet:double=@n19) -> (Output_Get:double)
   n20 = NOT Boolean(A=Get bIsTargeting())
   > Set Stop wobling left(Stop wobling left:double=float * float(A=@n19, B:bool=@n20)) -> (Output_Get:double)
   > Set Stop wobling right(Stop wobling right:double=float * float(A=@n18, B=@n20)) -> (Output_Get:double)
   n21 = NOT Boolean(A=Get bIsDBNO())
   n22 = Get bIsCrouching()
   n23 = Get bIsMoving2D()
   Set L Skirt correction then moving and crouching(
    execute=@n1.then_1
    L Skirt correction then moving and crouching:float=float * float(
     A=GetAlphaFromReferenceTransformDelta(BoneName=thigh_l, BaseName=pelvis, DeltaComponentType=EulerY, InputRange=(X=-10,Y=40), TwistAxis="1,0,0")
     B=AND Boolean(A=@n23, B=@n22, C=@n21)
    )
   ) -> (Output_Get:float)
   n24 = GetAlphaFromReferenceTransformDelta(
    SkeletalMeshComponent=GetOwnerSkeletalMeshComponent()
    BoneName=spine_05
    BaseName=root
    DeltaComponentType=EulerX
    InputRange=Select!(
     Melee:Vector 2D Structure=(X=-5,Y=-37)
     Pistol:Vector 2D Structure=(X=15,Y=-95)
     Shotgun:Vector 2D Structure=(X=-10,Y=-116)
     PaperBlueprint:Vector 2D Structure
     Rifle:Vector 2D Structure
     MeleeOneHand:Vector 2D Structure
     MachinePistol:Vector 2D Structure
     RocketLauncher:Vector 2D Structure
     GrenadeLauncher:Vector 2D Structure
     GoingCommando:Vector 2D Structure
     AssaultRifle:Vector 2D Structure=(X=4,Y=-101)
     TacticalShotgun:Vector 2D Structure
     SniperRifle:Vector 2D Structure
     TrapPlacement:Vector 2D Structure
     ShoulderLauncher:Vector 2D Structure
     AbilityDecoTool:Vector 2D Structure
     Crossbow:Vector 2D Structure
     C4:Vector 2D Structure
     RemoteControl:Vector 2D Structure
     DualWield:Vector 2D Structure
     AR_BullPup:Vector 2D Structure
     AR_ForwardGrip:Vector 2D Structure
     MedPackPaddles:Vector 2D Structure
     SMG_P90:Vector 2D Structure
     AR_DrumGun:Vector 2D Structure
     Consumable_Small:Vector 2D Structure
     Consumable_Large:Vector 2D Structure
     Balloon:Vector 2D Structure
     MountedTurret:Vector 2D Structure
     CreativeTool:Vector 2D Structure
     ExplosiveBow:Vector 2D Structure
     AshtonIndigo:Vector 2D Structure
     AshtonChicago:Vector 2D Structure
     MeleeDualWield:Vector 2D Structure
     Unarmed:Vector 2D Structure
     Index:EFortWeaponCoreAnimation Enum=Get WeaponCoreAnim()
    ) -> (ReturnValue:Vector 2D Structure)
    TwistAxis="1,0,0"
   )
   n25 = Get bIsPlayingEmote()
   Set Left(
    execute=@n1.then_3
    Left:double=MapRangeClamped(Value=SelectFloat(A=0, B=@n24, bPickA=@n25), InRangeA=0, InRangeB=0.5, OutRangeA=1, OutRangeB=0)
   ) -> (Output_Get:double)
   > Set Right(Right:double=MapRangeClamped(Value=@n24, InRangeA=0.5, InRangeB=1)) -> (Output_Get:double)
   > Set Rifle Bracelet Correction Alpha(Rifle Bracelet Correction Alpha:float=Select(
     Melee:float=0
     Pistol:float=0
     Shotgun:float=0
     PaperBlueprint:float=0
     Rifle:float=0
     MeleeOneHand:float=0
     MachinePistol:float=0
     RocketLauncher:float=0
     GrenadeLauncher:float=0
     GoingCommando:float=0
     AssaultRifle:float=0.4
     TacticalShotgun:float=0
     SniperRifle:float=0
     TrapPlacement:float=0
     ShoulderLauncher:float=0
     AbilityDecoTool:float=0
     Crossbow:float=0
     C4:float=0
     RemoteControl:float=0
     DualWield:float=0
     AR_BullPup:float=0
     AR_ForwardGrip:float=0
     MedPackPaddles:float=0
     SMG_P90:float=0
     AR_DrumGun:float=0
     Consumable_Small:float=0
     Consumable_Large:float=0
     Balloon:float=0
     MountedTurret:float=0
     CreativeTool:float=0
     ExplosiveBow:float=0
     AshtonIndigo:float=0
     AshtonChicago:float=0
     MeleeDualWield:float=0
     Unarmed:float=0
     Index=Get WeaponCoreAnim()
    ) -> (ReturnValue:float)) -> (Output_Get:float)
   K2Node_Composite_1 = static_crouching_rotation_CollapseGraph_2!() -> (ReturnValue:bool=~) {is=title K2Node_Composite}
   Set static_crouching_rotation(execute=@n1.then_4, static_crouching_rotation:bool=@K2Node_Composite_1) -> (Output_Get:bool)
   n26 = OR Boolean(A=Get bIsJumping(), B=Get bIsFalling())
   > Set TacticalJump(TacticalJump:bool=AND Boolean!(A=Get bIsTacticalSprinting(), B=@n26) -> (ReturnValue)) -> (Output_Get:bool)
   n27 = Get bIsMoving2D()
   > Set JumpingOrFalling(JumpingOrFalling:bool=AND Boolean(A=@n26, B=NOT Boolean(A=@n27))) -> (Output_Get:bool)
   n28 = Get bIsPlayingEmote()
   > Set Is In Front End and not playing emote(Is In Front End and not playing emote:bool=AND Boolean(A=Get bIsInFrontEnd(), B=NOT Boolean(A=@n28))) -> (Output_Get:bool)
   n29 = Gravity Override State Function DEPRECATED(execute=@n1.then_2, CharacterPartAnimInstance=Self-Reference(), GameTimeInSeconds=GetGameTimeInSeconds())
   > Set gravity_override(gravity_override=@n29)
   n30 = "Equal (Float)"(A=Get IsBaseballSliding(), B=1)
   n31 = "Equal (Float)"(A=Get IsSliding_alpha(), B=1)
   > Set SlidingAndBaseballSliding(SlidingAndBaseballSliding:double=SelectFloat(A=1, B=0, bPickA=OR Boolean(A=@n31, B=@n30))) -> (Output_Get:double)
   n32 = Get CurrentWeapon() -> (CurrentWeapon:Fort Weapon Object Reference)
   Is Valid(exec=@n1.then_6, InputObject=@n32)
   > n33 = Set is whip(is whip:bool=false) -> (Output_Get:bool)
   > n34 = Branch(Condition=@n33)
   n35 = GetAnimInstance(self=Get WeaponMesh!(self:Fort Weapon Object Reference=@n32) -> (WeaponMesh:Skeletal Mesh Component Budgeted Object Reference))
   n36 = GetCurrentActiveMontage!(self:Anim Instance Object Reference=@n35) -> (ReturnValue:Anim Montage Object Reference)
   > n37 = Is Valid(InputObject=@n36)
   Set Whip(execute=[@n34.else, @n37.Is Not Valid], Whip:double=0) -> (Output_Get:double)
   n38 = MapRangeClamped(
    Value=Montage_GetPosition!(self:Anim Instance Object Reference=@n35, Montage:Anim Montage Object Reference=@n36) -> (ReturnValue:float)
    InRangeA=0
    InRangeB=GetPlayLength!(self:Animation Asset Object Reference=@n36) -> (ReturnValue:float)
    OutRangeA=1
    OutRangeB=0
   )
   Set Whip(execute=@n37.Is Valid, Whip=SelectFloat(A=0, B=1, bPickA="float < float"(A=@n38, B=0.4)))
  ]}
 static_crouching_rotation_CollapseGraph_2 {
  path=EventGraph.K2Node_Composite_1.static_crouching_rotation_CollapseGraph_2
  nodes [
   Inputs!()
   n1 = Get bIsMoving2D()
   n2 = "Absolute (Float)"(A=BreakRotator(InRot=Get PawnDeltaRotation() -> (PawnDeltaRotation:Rotator)).Yaw)
   Outputs!(ReturnValue:bool=AND Boolean!(A="float > float"(A=@n2, B=2), B=Get bIsCrouching(), C:bool=NOT Boolean(A=@n1)) -> (ReturnValue))
  ]
 }
]
variables [
 gravity_override {type=Vector}
 HideLeftBraselet {type=number}
 HideRightBraselet {type=number}
 Is In Front End and not playing emote {type=boolean}
 Is Playing Melee Anim {type=boolean}
 is whip {type=boolean}
 IsBaseballSliding {type=number}
 IsCrouching_alpha {type=number}
 IsSliding_alpha {type=number}
 JumpingOrFalling {type=boolean}
 L Skirt correction then moving and crouching {type=number}
 L_skirt_correction_alpha {type=number}
 Left {type=number}
 MelleePistolOtherAssaultRifle {type=integer}
 MovingNotCrouching {type=number}
 Rifle Bracelet Correction Alpha {type=number, default=0.4}
 Right {type=number}
 rigidbody_alpha {type=number}
 SlidingAndBaseballSliding {type=number}
 static_crouching_rotation {type=boolean}
 Stop wobling left {type=number}
 Stop wobling right {type=number}
 TacticalJump {type=boolean}
 Whip {type=number}
]
