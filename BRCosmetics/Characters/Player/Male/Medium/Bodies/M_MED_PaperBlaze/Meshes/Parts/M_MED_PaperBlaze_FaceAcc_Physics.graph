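# Physics asset for the PaperBlaze male medium character's FaceAcc part, defining bodies and joints for its dynamic hair bones.
# Short capsule chains hang off the head with limited swing and almost no twist, and cross-constraints tie two right-side chains together.
# Head, facial root, upper spine and right upper arm add larger body volumes.
# Back hair, facial root and arm joints rotate freely, while head-to-spine is limited.
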PhysicsAsset /BRCosmetics/Characters/Player/Male/Medium/Bodies/M_MED_PaperBlaze/Meshes/Parts/M_MED_PaperBlaze_FaceAcc_Physics
bodies {
 FACIAL_C_FacialRoot [{
   shapeType=Capsule
   length=5.2864
   radius=4.6249
   shapeName=FACIAL_C_FacialRoot_capsule
   center=<0 2.9639 2.9024>
   rotation=<rot 23.0882 -90 0>
  }]
 dyn_hair19 [{shapeType=Capsule, length=2.5574, radius=1.9245, shapeName=dyn_hair19_capsule, center=<1.7841 0 0>, rotation=<rot 90 0 0>}]
 dyn_hair20 [{shapeType=Capsule, length=2.7327, radius=1.2136, shapeName=dyn_hair20_capsule, center=<1.7841 0 0>, rotation=<rot 90 0 0>}]
 dyn_hair_b [{shapeType=Capsule, length=0.505, radius=0.505, shapeName=dyn_hair_b_capsule, center=<0.1093 0 0>, rotation=<rot 0 -90 0>}]
 dyn_hair_f_r_1 [{shapeType=Capsule, length=2.7613, radius=1.5068, shapeName=dyn_hair_f_r_1_capsule, center=<1.7841 0 0>, rotation=<rot 90 0 0>}]
 dyn_hair_f_r_2 [{
   shapeType=Capsule
   length=2.9339
   radius=1.4989
   shapeName=dyn_hair_f_r_2_capsule
   center=<1.7841 0 0>
   rotation=<rot 78.7498 -0.0002 -0.0002>
  }]
 dyn_hair_l_1 [{shapeType=Capsule, length=2.4652, radius=1.1681, shapeName=dyn_hair_l_1_capsule, center=<1.7841 0 0>, rotation=<rot 90 0 0>}]
 dyn_hair_l_2 [{shapeType=Capsule, length=2.0659, radius=0.8747, shapeName=dyn_hair_l_2_capsule, center=<1.7841 0 0.4638>, rotation=<rot 90 0 0>}]
 dyn_hair_l_3 [{shapeType=Capsule, length=2.7544, radius=0.8382, shapeName=dyn_hair_l_3_capsule, center=<1.7841 0 0>, rotation=<rot 90 0 0>}]
 dyn_hair_r_1 [{shapeType=Capsule, length=3.5381, radius=1.2874, shapeName=dyn_hair_r_1_capsule, center=<1.7841 0 0>, rotation=<rot 90 0 0>}]
 dyn_hair_r_2 [{shapeType=Capsule, length=4.9485, radius=1.2659, shapeName=dyn_hair_r_2_capsule, center=<1.7841 0 0>, rotation=<rot 90 0 0>}]
 dyn_hair_r_3 [{shapeType=Capsule, length=2.6026, radius=1.5753, shapeName=dyn_hair_r_3_capsule, center=<1.3238 -0.2771 0>, rotation=<rot 90 0 0>}]
 dyn_hair_r_4 [
  {
   shapeType=Capsule
   length=5.2608
   radius=1.0355
   shapeName=dyn_hair_r_4_capsule
   center=<2.4753 0.8864 -0.3347>
   rotation=<rot 82.2992 -0.0075 -24.6448>
  }
  {
   shapeType=Capsule
   length=2.8426
   radius=0.9995
   shapeName=dyn_hair_r_4_capsule
   center=<0.7595 0 0.8961>
   rotation=<rot 66.8448 -0.0001 -8.3681>
  }
 ]
 dyn_hair_r_f [{shapeType=Capsule, length=1.9891, radius=0.7562, shapeName=dyn_hair_r_f_capsule, center=<1.7841 0 0>, rotation=<rot 90 0 0>}]
 dyn_hair_r_f_1 [{shapeType=Capsule, length=2.1744, radius=2.1224, shapeName=dyn_hair_r_f_1_capsule, center=<1.7841 0 0>, rotation=<rot 90 0 0>}]
 dyn_hair_r_f_2 [{shapeType=Capsule, length=3.1972, radius=0.9608, shapeName=dyn_hair_r_f_2_capsule, center=<1.7841 0 0>, rotation=<rot 90 0 0>}]
 head [{shapeType=Capsule, length=7.095, radius=9.246, rotation=<rot 90 180 177.1876>}]
 spine_05 [
  {shapeType=Sphere, radius=29.709, shapeName=spine_05_sphere, center=<-8.0774 -11.871 0.8447>}
  {shapeType=Capsule, length=7.7364, radius=13.3266, center=<3.2583 -0.8489 6.5025>, rotation=<rot 90 0 66.5188>}
 ]
 upperarm_r [{shapeType=Capsule, length=18.5074, radius=7.6146, shapeName=upperarm_r_capsule, center=<-15.0413 0 0>, rotation=<rot 90 0 0>}]
}
constraints [
 {bone1Name=head, bone2Name=spine_05, swing1LimitDegrees=45, swing2LimitDegrees=45, twistLimitDegrees=45}
 {bone1Name=dyn_hair19, bone2Name=head, swing1LimitDegrees=15, swing2LimitDegrees=15, twistLimitDegrees=1}
 {bone1Name=dyn_hair20, bone2Name=dyn_hair19, swing1LimitDegrees=15, swing2LimitDegrees=15, twistLimitDegrees=1}
 {bone1Name=dyn_hair_f_r_1, bone2Name=head, swing1LimitDegrees=15, swing2LimitDegrees=15, twistLimitDegrees=1}
 {bone1Name=dyn_hair_f_r_2, bone2Name=dyn_hair_f_r_1, swing1LimitDegrees=15, swing2LimitDegrees=15, twistLimitDegrees=1}
 {bone1Name=dyn_hair_l_1, bone2Name=head, swing1LimitDegrees=30, swing2LimitDegrees=30, twistLimitDegrees=1}
 {bone1Name=dyn_hair_l_2, bone2Name=dyn_hair_l_1, swing1LimitDegrees=30, swing2LimitDegrees=30, twistLimitDegrees=1}
 {bone1Name=dyn_hair_l_3, bone2Name=head, swing1LimitDegrees=30, swing2LimitDegrees=30, twistLimitDegrees=1}
 {bone1Name=dyn_hair_r_1, bone2Name=head, swing1LimitDegrees=5, swing2LimitDegrees=20, twistLimitDegrees=1}
 {bone1Name=dyn_hair_r_2, bone2Name=dyn_hair_r_1, swing1LimitDegrees=20, swing2LimitDegrees=25, twistLimitDegrees=1}
 {bone1Name=dyn_hair_r_f, bone2Name=head, swing1LimitDegrees=15, swing2LimitDegrees=15, twistLimitDegrees=1}
 {bone1Name=dyn_hair_r_f_1, bone2Name=head, swing1LimitDegrees=15, swing2LimitDegrees=15, twistLimitDegrees=1}
 {bone1Name=dyn_hair_r_f_2, bone2Name=dyn_hair_r_f_1, swing1LimitDegrees=15, swing2LimitDegrees=15, twistLimitDegrees=1}
 {bone1Name=dyn_hair_r_3, bone2Name=head, swing1LimitDegrees=10, swing2LimitDegrees=40, twistLimitDegrees=1}
 {bone1Name=dyn_hair_r_4, bone2Name=dyn_hair_r_3, swing1LimitDegrees=30, swing2LimitDegrees=30, twistLimitDegrees=1}
 {
  bone1Name=dyn_hair_b
  bone2Name=head
  swing1LimitDegrees=45
  swing1Motion=Free
  swing2LimitDegrees=45
  swing2Motion=Free
  twistLimitDegrees=45
  twistMotion=Free
 }
 {
  bone1Name=upperarm_r
  bone2Name=spine_05
  swing1LimitDegrees=45
  swing1Motion=Free
  swing2LimitDegrees=45
  swing2Motion=Free
  twistLimitDegrees=45
  twistMotion=Free
 }
 {
  bone1Name=FACIAL_C_FacialRoot
  bone2Name=head
  swing1LimitDegrees=45
  swing1Motion=Free
  swing2LimitDegrees=45
  swing2Motion=Free
  twistLimitDegrees=45
  twistMotion=Free
 }
 {bone1Name=dyn_hair_r_2, bone2Name=dyn_hair_r_4, swing1LimitDegrees=15, swing2LimitDegrees=15, twistLimitDegrees=15}
 {bone1Name=dyn_hair_r_3, bone2Name=dyn_hair_r_1, swing1LimitDegrees=15, swing2LimitDegrees=15, twistLimitDegrees=15}
]
