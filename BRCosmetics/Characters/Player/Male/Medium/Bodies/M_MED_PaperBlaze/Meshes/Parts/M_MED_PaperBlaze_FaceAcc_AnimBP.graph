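# Anim Blueprint for the PaperBlaze face accessory part, driving its hood and hair over the character's pose.
# Blended parent and body mesh poses feed a Control Rig, then RigidBody physics with overridden gravity.
# Additive bone corrections follow for hood, back hair and chin.
# Head-to-spine_05 yaw routes to skydive, melee or default alphas, range-mapped and interpolated; an extra hood-correction chain is unconnected to the output.
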AnimBlueprint /BRCosmetics/Characters/Player/Male/Medium/Bodies/M_MED_PaperBlaze/Meshes/Parts/M_MED_PaperBlaze_FaceAcc_AnimBP
target_skeleton=&/M_MED_PaperBlaze_FaceAcc_Skeleton
graphs [
 AnimGraph {nodes [
   n1 = "Transform (Modify) Bone - Bone: dyn_hood"(
    Alpha="Get Return Value Rotation X (Roll)"() -> ("Return Value Rotation X (Roll)":float)
    Translation="-6.531,-11.210,8.130"
    Rotation="0.000,0.000,-19.192"
    Scale="1.488,1.488,2.137"
   ) {
    rotationMode=BMM_Additive
    scaleMode=BMM_Additive
    scaleSpace=BCS_BoneSpace
    translationMode=BMM_Additive
    alphaBoolBlend {blendOption=Linear}
    alphaScaleBiasClamp {bInterpResult=true, bMapRange=true, interpSpeedDecreasing=20, interpSpeedIncreasing=20, scale=0.4, inRange {max=-80, min=-15}}
    boneToModify {boneName=dyn_hood}
    rotation {roll=-19.192}
    scale=<1.488 1.488 2.137>
    translation=<-6.531 -11.21 8.13>
   }
   n2 = "Transform (Modify) Bone - Bone: dyn_hood"(
    ComponentPose=@n1
    Alpha=Get HoodCorrectionMellee() -> (HoodCorrectionMellee:double)
    Translation="-1.507,-8.316,6.143"
    Rotation="0.000,0.000,-6.481"
    Scale="1.488,1.494,2.137"
   ) {
    rotationMode=BMM_Additive
    scaleMode=BMM_Additive
    scaleSpace=BCS_BoneSpace
    translationMode=BMM_Additive
    alphaBoolBlend {blendOption=Linear}
    alphaScaleBiasClamp {bInterpResult=true, bMapRange=true, interpSpeedDecreasing=20, interpSpeedIncreasing=20, inRange {max=-80, min=-15}}
    boneToModify {boneName=dyn_hood}
    rotation {roll=-6.481}
    scale=<1.488 1.494 2.137>
    translation=<-1.507 -8.316 6.143>
   }
   "Transform (Modify) Bone - Bone: dyn_hood"(
    ComponentPose=@n2
    Alpha=Get Skydive() -> (Skydive:double)
    Translation="-1.507,-8.316,9.134"
    Rotation="-0.000,0.000,9.234"
    Scale="1.488,1.494,2.141"
   ) {
    rotationMode=BMM_Additive
    scaleMode=BMM_Additive
    scaleSpace=BCS_BoneSpace
    translationMode=BMM_Additive
    alphaBoolBlend {blendOption=Linear}
    alphaScaleBiasClamp {bInterpResult=true, bMapRange=true, interpSpeedDecreasing=20, interpSpeedIncreasing=20, inRange {max=-80, min=-15}}
    boneToModify {boneName=dyn_hood}
    rotation {roll=9.234}
    scale=<1.488 1.494 2.141>
    translation=<-1.507 -8.316 9.134>
   }
   n3 = Get BodyMesh() -> (BodyMesh:Skeletal Mesh Component Object Reference)
   n4 = Control Rig(Source=Layered blend per bone(
     BasePose=Copy Pose From Mesh() {bUseAttachedParent=true}
     BlendPoses_0=Copy Pose From Mesh(SourceMeshComponent=@n3) {bUseAttachedParent=true}
    ) {layerSetup [{branchFilters [{boneName=upperarm_l}, {blendDepth=1, boneName=upperarm_r}]}]}) {
    alphaBoolBlend {blendOption=Linear}
    controlRigAssetReference {blueprintRigClass=/BRCosmetics/Characters/Player/Male/Medium/Bodies/M_MED_PaperBlaze/Meshes/M_MED_PaperBlaze_CtrlRig.M_MED_PaperBlaze_CtrlRig_C}
   }
   n5 = RigidBody!(
    ComponentPose=Local To Component(LocalPose=@n4)
    Alpha
    OverrideWorldGravity:Vector=Get gravity_override() -> (gravity_override:Vector)
    ExternalForce
   ) -> (Pose) {
    bClampLinearTranslationLimitToRefPose=true
    bOverrideWorldGravity=true
    overridePhysicsAsset=&/M_MED_PaperBlaze_FaceAcc_Physics
    simulationSpace=BaseBoneSpace
    alphaBoolBlend {blendOption=Linear}
    alphaScaleBiasClamp {bInterpResult=true, interpSpeedDecreasing=5, interpSpeedIncreasing=5}
    baseBoneRef {boneName=neck_01}
    componentAppliedLinearAccClamp=<3000 3000 3000>
    componentLinearAccScale=<0.4 0.4 0.4>
    componentLinearVelScale=<0.6 0.6 0.6>
    overrideWorldGravity {z=-980}
    simSpaceSettings {
     maxAngularAcceleration=500
     maxAngularVelocity=3
     maxLinearAcceleration=400
     maxLinearVelocity=250
     velocityScaleZ=0
     worldAlpha=1
     externalLinearDragV=<0.1 0.1>
    }
   }
   n6 = "Transform (Modify) Bone - Bone: dyn_hood"(
    ComponentPose=@n5
    Alpha="Get Return Value Rotation X (Roll)"()
    Translation="-0.000,-15.918,10.975"
    Rotation="-0.000,0.000,10.598"
    Scale="1.338,1.031,1.052"
   ) {
    rotationMode=BMM_Additive
    scaleMode=BMM_Additive
    scaleSpace=BCS_BoneSpace
    translationMode=BMM_Additive
    alphaBoolBlend {blendOption=Linear}
    alphaScaleBiasClamp {bInterpResult=true, bMapRange=true, interpSpeedDecreasing=20, interpSpeedIncreasing=20, scale=0.4, inRange {max=-80, min=-15}}
    boneToModify {boneName=dyn_hood}
    rotation {roll=10.598}
    scale=<1.338 1.031 1.052>
    translation {y=-15.918, z=10.975}
   }
   n7 = Get HoodCorrectionMellee()
   n8 = "Transform (Modify) Bone - Bone: dyn_hood"(ComponentPose=@n6, Alpha=@n7, Translation="0.196,-4.565,0.869", Rotation="-0.000,0.000,9.595", Scale="1.516,1.122,1.091") {
    rotationMode=BMM_Additive
    scaleMode=BMM_Additive
    scaleSpace=BCS_BoneSpace
    translationMode=BMM_Additive
    alphaBoolBlend {blendOption=Linear}
    alphaScaleBiasClamp {bInterpResult=true, bMapRange=true, interpSpeedDecreasing=20, interpSpeedIncreasing=20, inRange {max=-60, min=-15}}
    boneToModify {boneName=dyn_hood}
    rotation {roll=9.595}
    scale=<1.516 1.122 1.091>
    translation=<0.196 -4.565 0.869>
   }
   n9 = "Transform (Modify) Bone - Bone: dyn_hood"(ComponentPose=@n8, Alpha=@n7, Translation="0.196,-4.565,0.869", Rotation="-0.000,0.000,9.595", Scale="1.516,1.122,1.091") {
    rotationMode=BMM_Additive
    scaleMode=BMM_Additive
    scaleSpace=BCS_BoneSpace
    translationMode=BMM_Additive
    alphaBoolBlend {blendOption=Linear}
    alphaScaleBiasClamp {bInterpResult=true, bMapRange=true, interpSpeedDecreasing=20, interpSpeedIncreasing=20, inRange {max=-80, min=-60}}
    boneToModify {boneName=dyn_hood}
    rotation {roll=9.595}
    scale=<1.516 1.122 1.091>
    translation=<0.196 -4.565 0.869>
   }
   n10 = "Transform (Modify) Bone - Bone: dyn_hood"(ComponentPose=@n9, Alpha=Get Skydive(), Translation="-1.507,-8.316,9.134", Rotation="-0.000,0.000,9.234", Scale="1.488,1.494,2.141") {
    rotationMode=BMM_Additive
    scaleMode=BMM_Additive
    scaleSpace=BCS_BoneSpace
    translationMode=BMM_Additive
    alphaBoolBlend {blendOption=Linear}
    alphaScaleBiasClamp {bInterpResult=true, bMapRange=true, interpSpeedDecreasing=20, interpSpeedIncreasing=20, inRange {max=-80, min=-15}}
    boneToModify {boneName=dyn_hood}
    rotation {roll=9.234}
    scale=<1.488 1.494 2.141>
    translation=<-1.507 -8.316 9.134>
   }
   n11 = "Transform (Modify) Bone - Bone: dyn_hair_b"(
    ComponentPose=@n10
    Alpha="Get Return Value Rotation X (Roll)"()
    Translation="2.819,-24.119,5.163"
    Rotation="-0.000,0.000,45.960"
    Scale="1,1,1"
   ) {
    rotationMode=BMM_Additive
    scaleMode=BMM_Additive
    scaleSpace=BCS_BoneSpace
    translationMode=BMM_Additive
    alphaBoolBlend {blendOption=Linear}
    alphaScaleBiasClamp {bInterpResult=true, bMapRange=true, interpSpeedDecreasing=20, interpSpeedIncreasing=20, inRange {max=-80, min=-15}}
    boneToModify {boneName=dyn_hair_b}
    rotation {roll=45.96}
    translation=<2.819 -24.119 5.163>
   }
   n12 = "Transform (Modify) Bone - Bone: FACIAL_C_Chin"(ComponentPose=@n11, Alpha=1, Translation="0.000,0.175,0.074", Rotation="-0.000,-0.000,-2.715", Scale="1,1,1") {
    rotationMode=BMM_Additive
    rotationSpace=BCS_BoneSpace
    scaleMode=BMM_Additive
    scaleSpace=BCS_BoneSpace
    translationMode=BMM_Additive
    translationSpace=BCS_BoneSpace
    alphaBoolBlend {blendOption=Linear}
    boneToModify {boneName=FACIAL_C_Chin}
    rotation=<rot 0 0 -2.715>
    translation=<0 0.175 0.0741>
   }
   Output Pose(Result=Component To Local(ComponentPose=@n12)) {layerGroup=None}
  ]}
 EventGraph {nodes [
   Event BlueprintUpdateAnimation()
   > n1 = Gravity Override State Function DEPRECATED(CharacterPartAnimInstance=Self-Reference(), GameTimeInSeconds=GetGameTimeInSeconds())
   > Set gravity_override(gravity_override=@n1)
   > n2 = Branch(Condition=Get bIsSkyDiving())
   > "Set Return Value Rotation X (Roll)"("Return Value Rotation X (Roll)":float=0) -> (Output_Get:float)
   > Set HoodCorrectionMellee(HoodCorrectionMellee:double=0) -> (Output_Get:double)
   n3 = GetDeltaTransformFromRefPose!(self=GetOwnerSkeletalMeshComponent(), BoneName=head, BaseName=spine_05) -> (
    ReturnValue_Location:Vector
    ReturnValue_Rotation_Roll:float
    ReturnValue_Rotation_Pitch:float
    ReturnValue_Rotation_Yaw:float
    ReturnValue_Scale:Vector
   )
   > Set Skydive(Skydive:double=@n3.ReturnValue_Rotation_Yaw) -> (Output_Get:double)
   n4 = Branch(
    execute=@n2.else
    Condition="Equal (Enum)"(A:EFortWeaponCoreAnimation Enum=Get WeaponCoreAnim!() -> (WeaponCoreAnim), B:EFortWeaponCoreAnimation Enum=Melee)
   )
   > "Set Return Value Rotation X (Roll)"("Return Value Rotation X (Roll)"=0)
   > Set HoodCorrectionMellee(HoodCorrectionMellee=@n3.ReturnValue_Rotation_Yaw)
   > Set Skydive(Skydive=0)
   "Set Return Value Rotation X (Roll)"(execute=@n4.else, "Return Value Rotation X (Roll)"=@n3.ReturnValue_Rotation_Yaw)
   > Set HoodCorrectionMellee(HoodCorrectionMellee=0)
   > Set Skydive(Skydive=0)
   Event BlueprintInitializeAnimation()
   > Set BodyMesh(BodyMesh=GetOwnerSkeletalMeshComponent())
  ]}
]
variables [
 BodyMesh {type=SkeletalMeshComponent}
 gravity_override {type=Vector}
 HoodCorrectionMellee {type=number}
 mesh {type=SkeletalMeshComponentBudgeted}
 "Return Value Rotation X (Roll)" {type=number}
 Skydive {type=number}
]
